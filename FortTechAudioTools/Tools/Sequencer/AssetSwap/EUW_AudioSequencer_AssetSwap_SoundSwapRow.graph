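# Editor utility widget for one row in the Sequencer audio asset swap tool, pairing an audio section's sound with a replacement.
# ApplyRowData caches the section's sound, track and frame range; clicking the source locates it in the Content Browser.
# Dropped sounds are broadcast with the row index, Ctrl-click toggles selection, and plain click assigns the Content Browser selection as Manual.
# SetReplacementState stores the replacement and score, updating icon, label and status tint.
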EditorUtilityWidgetBlueprint /FortTechAudioTools/Tools/Sequencer/AssetSwap/EUW_AudioSequencer_AssetSwap_SoundSwapRow
graphs [
 ApplyRowData {nodes [
   n1 = ApplyRowData!() -> (then, InRowIndex:int, InSourceAudioSection:Movie Scene Audio Section Object Reference, InFrameRate:Frame Rate Structure) {is=title K2Node_FunctionEntry}
   > Set RowIndex(RowIndex:int=@n1) -> (Output_Get:int=0)
   > n2 = Set SourceAudioSection(SourceAudioSection:Movie Scene Audio Section Object Reference=@n1.InSourceAudioSection) -> (Output_Get:Movie Scene Audio Section Object Reference)
   n3 = GetSound!(self:Movie Scene Audio Section Object Reference=@n2) -> (ReturnValue:Sound Base Object Reference)
   > Set SourceSoundAsset(SourceSoundAsset:Sound Base Object Reference=@n3) -> (Output_Get:Sound Base Object Reference)
   > n4 = GetIconStyleForSound!(execute, self, InSound:Sound Base Object Reference=@n3) -> (then, OutIconPath:Name=None, OutIconColor:Linear Color Structure)
   n5 = Get Image_RowSourceIcon() -> (Image_RowSourceIcon:Tech Audio Tools Editor Slate Icon Object Reference)
   > SetIconBrushByPath!(execute, self:Tech Audio Tools Editor Slate Icon Object Reference=@n5, InFullIconPath:Name=@n4) -> (then)
   > SetIconColor!(execute, self:Tech Audio Tools Editor Slate Icon Object Reference=@n5, InColor:Linear Color Structure=@n4.OutIconColor) -> (then)
   > Set SourceDisplayName(SourceDisplayName:String=GetDisplayName(Object=@n3)) -> (Output_Get:String)
   > n6 = Cast To MovieSceneAudioTrack!(execute, Object=GetOuterObject!(Object:Object Reference=@n2) -> (ReturnValue:Object Reference)) -> (then, CastFailed, AsMovie Scene Audio Track:Movie Scene Audio Track Object Reference) {is=title K2Node_DynamicCast}
   > n7 = Set AudioTrackName(AudioTrackName:String=GetDisplayName(Object=@n6)) -> (Output_Get:String)
   n8 = Get SourceAudioSection() -> (SourceAudioSection:Movie Scene Audio Section Object Reference)
   n9 = GetStartFrame!(execute=[@n7.then, @n6.CastFailed], Section:Movie Scene Section Object Reference=@n8) -> (then, ReturnValue:int=0)
   > Set StartFrame(StartFrame:int=@n9) -> (Output_Get:int=0)
   > n10 = GetEndFrame!(execute, Section:Movie Scene Section Object Reference=@n8) -> (then, ReturnValue:int=0)
   > Set EndFrame(EndFrame:int=@n10) -> (Output_Get:int=0)
   Get InFrameRate() -> (InFrameRate:Frame Rate Structure)
  ]}
 AssignSelectedReplacement {nodes [
   AssignSelectedReplacement() {is=title K2Node_FunctionEntry}
   > n1 = GetSelectedAssetsOfClass(AssetClass=/Script/Engine.SoundBase)
   > Branch(Condition=Is Valid Index!(TargetArray:Array of Object References=@n1, IndexToTest:int) -> (ReturnValue:bool))
   n2 = "Get (a copy)"(Array:Array of Object References=@n1) -> (Output:Object Reference)
   > n3 = Cast To SoundBase!(execute, Object=@n2) -> (then, CastFailed, AsSound Base:Sound Base Object Reference) {is=title K2Node_DynamicCast}
   > SetReplacementState!(
    execute
    self
    InReplacementDisplayName:String=GetDisplayName(Object=@n2)
    InStatusText:String=Get Manual() -> (Manual:String)
    InReplacementSound:Sound Base Object Reference=@n3
    InMatchScore:double=0
   ) -> (then)
  ]}
 EventGraph {nodes [
   n1 = "On Assets Dropped (AudioAssetDropTargetWidget_Row)"!() -> (OutputDelegate, then, DroppedAssets:Array of Object References) {is=title K2Node_ComponentBoundEvent}
   > OnDroppedAssets!(execute, self, DroppedAssets:Array of Object References=@n1.DroppedAssets) -> (then)
   n2 = "On Clicked (AudioAssetDropTargetWidget_Row)"!() -> (OutputDelegate, then, ModifierKeys:EModifierKeys Enum=None) {is=title K2Node_ComponentBoundEvent}
   n3 = "Equal (Enum)"(A:EModifierKeys Enum=@n2.ModifierKeys, B:EModifierKeys Enum=Control)
   > n4 = Branch(Condition=@n3)
   > WidgetedSelected!(execute, self, bIsCtrlPressed:bool=@n3) -> (then)
   AssignSelectedReplacement(execute=@n4.else) {is=title K2Node_CallFunction}
   > Call On Selection Changed!(
    execute:Exec
    self:EUW Audio Sequencer Asset Swap Sound Swap Row Object Reference
    RowIndex:int=Get RowIndex() -> (RowIndex:int=0)
    bIsSelected:bool
    bIsCtrlPressed:bool
   ) -> (then:Exec) {is=title K2Node_CallDelegate}
   "On Clicked (Button_SourceSound)"() {is=title K2Node_ComponentBoundEvent}
   > OnSourceSoundClicked()
  ]}
 GetIconStyleForSound {nodes [
   n1 = GetIconStyleForSound!() -> (then, InSound:Sound Base Object Reference) {is=title K2Node_FunctionEntry}
   > n2 = Construct Object from Class(Class=/Script/TechAudioToolsContentEditor.TechAudioToolsAssetViewModel) -> (ReturnValue:Tech Audio Tools Asset Viewmodel Object Reference)
   > SetAssetData!(
    execute
    self:Tech Audio Tools Asset Viewmodel Object Reference=@n2
    InAssetData:"Asset Data Structure (by ref)"=CreateAssetData(InAsset=@n1, bAllowBlueprintClass=false)
   ) -> (then)
   > Return Node!(
    execute
    OutIconPath:Name=GetClassIconBrushPath!(self:Tech Audio Tools Asset Viewmodel Object Reference=@n2) -> (ReturnValue:Name=None)
    OutIconColor:Linear Color Structure=GetClassColor!(self:Tech Audio Tools Asset Viewmodel Object Reference=@n2) -> (ReturnValue:Linear Color Structure)
   )
  ]}
 GetReplaceDisplayText {nodes [
   GetReplaceDisplayText() {is=title K2Node_FunctionEntry}
   > Return Node!(execute, ReturnValue:Text="To Text (String)"(InString=Get ReplacementDisplayName() -> (ReplacementDisplayName:String)))
  ]}
 GetTimelineDisplayString {nodes [
   n1 = GetTimelineDisplayString!() -> (then, InputPin:int) {is=title K2Node_FunctionEntry}
   n2 = QualifiedFrameTime To Seconds!(InFrameTime:"Qualified Frame Time Structure (by ref)"=MakeQualifiedFrameTime!(
     Frame:Frame Number Structure=Make Frame Number!(Value:int=@n1) -> (FrameNumber:Frame Number Structure) {is=title K2Node_MakeStruct}
     FrameRate:Frame Rate Structure
     SubFrame:float=0
    ) -> (ReturnValue:Qualified Frame Time Structure)) -> (ReturnValue:float)
   n3 = "% (Float)"!(A:double=@n2, B:double=60) -> (ReturnValue:double)
   n4 = float / float(A=@n2, B=3600)
   n5 = float / float(A="% (Float)"(A=@n2, B=3600), B=60)
   n6 = BreakFrameRate!(InFrameRate:"Frame Rate Structure (by ref)"=Get FrameRate() -> (FrameRate:Frame Rate Structure)) -> (Numerator:int=0, Denominator:int=0)
   > Return Node!(
    execute
    Result:Text=Format Text!(
     Format="LOCTEXT(\"{Hours:02d}:{Minutes:02d}:{Seconds:02d}:{Frames:02d}\")"
     "Hours:02d":int=Floor(A=@n4) {is=title K2Node_CallFunction}
     "Minutes:02d":int=Floor(A=@n5)
     "Seconds:02d":int=Floor(A=@n3)
     "Frames:02d":int="% (Integer)"(A=@n1, B=@n6) -> (ReturnValue=0)
    ) -> (Result)
   )
   Branch()
  ]}
 MatchedOrManual {nodes [
   MatchedOrManual() {is=title K2Node_FunctionEntry}
   n1 = Get Manual()
   n2 = Get StatusText() -> (StatusText:String)
   n3 = Get Matched() -> (Matched:String)
   > Return Node!(execute, bIsMatched:bool=OR Boolean(A="Equal Exactly (String)"(A=@n2, B=@n3), B="Equal Exactly (String)"(A=@n2, B=@n1)))
  ]}
 OnDroppedAssets {nodes [
   n1 = OnDroppedAssets!() -> (then, DroppedAssets:Array of Object References) {is=title K2Node_FunctionEntry}
   > n2 = For Each Loop(Array:Array of Object References=@n1) -> (Array Element:Object Reference)
   Call On Multiple Assets Dropped on Row!(
    execute:Exec=@n2.Completed
    self:EUW Audio Sequencer Asset Swap Sound Swap Row Object Reference
    StartingRowIndex:int=Get RowIndex() -> (RowIndex=0)
    "DroppedSounds ":Array of Sound Base Object References=Get DroppedSounds() -> (DroppedSounds:Array of Sound Base Object References)
   ) -> (then:Exec) {is=title K2Node_CallDelegate}
   n3 = Cast To SoundBase(execute=@n2.LoopBody, Object=@n2)
   > Add(TargetArray:Array of Sound Base Object References=Get DroppedSounds(), NewItem:"Sound Base Object Reference (by ref)"=@n3) {is=title K2Node_CallArrayFunction}
  ]}
 OnMultipleAssetsDroppedOnRow {nodes [OnMultipleAssetsDroppedOnRow!() -> (then, StartingRowIndex:int, "DroppedSounds ":Array of Sound Base Object References) {is=title K2Node_FunctionEntry}]}
 OnSelectionChanged {nodes [OnSelectionChanged!() -> (then, RowIndex:int, bIsSelected:bool=~, bIsCtrlPressed:bool=~) {is=title K2Node_FunctionEntry}]}
 OnSourceSoundClicked {nodes [
   OnSourceSoundClicked() {is=title K2Node_FunctionEntry}
   > K2Node_VariableGet_SourceAsset = Get!(execute:Exec) -> (then:Exec, else:Exec, SourceSoundAsset:Sound Base Object Reference)
   n1 = To Soft Object Reference(SoftObjectPath=GetSoftObjectPath(Object=@K2Node_VariableGet_SourceAsset))
   > SyncBrowserToObjects(AssetPaths=Make Array("[0]":String="To String (SoftObjectReference)"!(SoftObjectReference:"Soft Object Reference (by ref)"=@n1) -> (ReturnValue:String)) -> (Array:Array of Strings))
  ]}
 SetReplacementState {nodes [
   n1 = SetReplacementState!() -> (then, InReplacementDisplayName:String, InStatusText:String, InReplacementSound:Sound Base Object Reference, InMatchScore:double=~) {is=title K2Node_FunctionEntry}
   > Set ReplacementDisplayName(ReplacementDisplayName:String=@n1) -> (Output_Get:String)
   > Set StatusText(StatusText:String=@n1.InStatusText) -> (Output_Get:String)
   > Set ReplacementSound(ReplacementSound:Sound Base Object Reference=@n1.InReplacementSound) -> (Output_Get:Sound Base Object Reference)
   > n2 = Is Valid(InputObject=@n1.InReplacementSound)
   n3 = Get Image_RowReplacementIcon() -> (Image_RowReplacementIcon:Tech Audio Tools Editor Slate Icon Object Reference)
   n4 = SetVisibility(execute=@n2.Is Not Valid, self=@n3, InVisibility=Collapsed)
   n5 = GetIconStyleForSound(execute=[@n12.then, @n4.then], InSound=@n1.InReplacementSound) -> (OutIconPath=None) {is=title K2Node_CallFunction}
   > SetIconBrushByPath(self=@n3, InFullIconPath=@n5)
   > SetIconColor(self=@n3, InColor=@n5.OutIconColor)
   > Set MatchScore(MatchScore:double=@n1.InMatchScore) -> (Output_Get:double)
   n6 = Get MatchScore() -> (MatchScore:double)
   > n7 = Set MatchScoreToolTip(MatchScoreToolTip:String=Append!(
     A=Get ReplacementDisplayName()
     B=" "
     C:String=Get StatusText()
     D:String=" with a score of "
     E:String="To String (Float)"(InDouble=@n6)
     F:String
    ) -> (ReturnValue)) -> (Output_Get:String)
   > PrintString(InString=@n7)
   n8 = Get ReplacementDisplayName()
   > SetText(
    self:Audio Asset Drop Target Widget Object Reference=Get AudioAssetDropTargetWidget_Row() -> (AudioAssetDropTargetWidget_Row:Audio Asset Drop Target Widget Object Reference)
    InText="To Text (String)"(InString=Append!(A="      ", B=@n8) -> (ReturnValue))
   )
   > n9 = Branch(Condition="Equal Exactly (String)"(A=Get StatusText(), B=Get Matched()))
   > Set members in Slate Brush!(
    execute:Exec
    StructRef:"Slate Brush Structure (by ref)"=Get Background!(self:Border Object Reference=Get Border_RowStatus() -> (Border_RowStatus:Border Object Reference)) -> (Background:Slate Brush Structure)
    TintColor:Slate Color Structure=Make Slate Color(SpecifiedColor=(R=0.1459,G=0.5312,B=0.2002,A=1))
   ) -> (then:Exec, StructOut:"Slate Brush Structure (by ref)") {is=title K2Node_SetFieldsInStruct}
   n10 = Branch(execute=@n9.else, Condition="Equal Exactly (String)"(A=Get StatusText(), B=Get NeedsReview() -> (NeedsReview:String)))
   > Set members in Slate Brush(StructRef=Get Background(self=Get Border_RowStatus()), TintColor=Make Slate Color(SpecifiedColor=(R=0.6094,G=0.4032,B=0.0825,A=1)))
   n11 = Branch(execute=@n10.else, Condition="Equal Exactly (String)"(A=Get StatusText(), B=Get Manual()))
   > Set members in Slate Brush(StructRef=Get Background(self=Get Border_RowStatus()), TintColor=Make Slate Color(SpecifiedColor=(R=0.127,G=0.2971,B=0.6094,A=1)))
   Set members in Slate Brush(
    execute=@n11.else
    StructRef=Get Background(self=Get Border_RowStatus())
    TintColor=Make Slate Color(SpecifiedColor=(R=0.4948,G=0.0876,B=0.0876,A=1))
   )
   n12 = SetVisibility(execute=@n2.Is Valid, self=@n3, InVisibility=Visible)
  ]}
 SetRowIndexText {nodes [
   SetRowIndexText() {is=title K2Node_FunctionEntry}
   n1 = Get RowIndex() -> (RowIndex=0)
   > Return Node!(execute, ReturnValue:Text="To Text (Integer)"(Value=int + int(A=@n1), bUseGrouping=false, MinimumIntegralDigits=2))
  ]}
 SoundSourceToolTip {nodes [
   SoundSourceToolTip() {is=title K2Node_FunctionEntry}
   n1 = Get StartFrame() -> (StartFrame:int=0)
   n2 = Get EndFrame() -> (EndFrame:int=0)
   > Return Node(ReturnValue=Format Text!(
     Format="LOCTEXT(\"Click to navigate to this asset in the Content Browser.\\r\\nTimecode Range: [{TStart} - {TEnd}]\\r\\nFrames Range: [{FStart} - {FEnd}]\")"
     TStart:Text=GetTimelineDisplayString!(InputPin:int=@n1) -> (Result:Text) {is=title K2Node_CallFunction}
     TEnd:Text=GetTimelineDisplayString(InputPin=@n2)
     FStart:int=Get StartFrame() -> (StartFrame=0)
     FEnd:int=Get EndFrame() -> (EndFrame=0)
    ) -> (Result))
  ]}
 WidgetedSelected {nodes [
   n1 = WidgetedSelected!() -> (then, bIsCtrlPressed:bool=~) {is=title K2Node_FunctionEntry}
   > n2 = Branch(Condition=@n1)
   n3 = Get AudioAssetDropTargetWidget_Row()
   > n4 = SetSelected!(
    execute
    self:Audio Asset Drop Target Widget Object Reference=@n3
    bSelected:bool=NOT Boolean(A=GetIsSelected!(self:Audio Asset Drop Target Widget Object Reference=@n3) -> (ReturnValue:bool))
   ) -> (then)
   Call On Selection Changed(
    execute=[@n5.then, @n4.then]
    RowIndex=Get RowIndex() -> (RowIndex=0)
    bIsSelected=GetIsSelected(self=Get AudioAssetDropTargetWidget_Row())
    bIsCtrlPressed=Get bIsCtrlPressed() -> (bIsCtrlPressed:bool)
   )
   n5 = SetSelected(execute=@n2.else, self=Get AudioAssetDropTargetWidget_Row(), bSelected=true)
  ]}
]
variables [
 AudioTrackName {type=string, default=""}
 EndFrame {type=integer}
 FrameRate {type=FrameRate, default {denominator=1, numerator=60000}}
 Manual {type=string, default=Manual}
 Matched {type=string, default=Matched}
 MatchScore {type=number}
 MatchScoreToolTip {type=string, default=""}
 NeedsReview {type=string, default=Needs Review}
 OnMultipleAssetsDroppedOnRow {default=()}
 OnSelectionChanged {default=()}
 ReplacementDisplayName {type=string, default=Drop Asset Here}
 ReplacementSound {type=SoundBase, default=null}
 RowIndex {type=integer}
 SourceAudioSection {type=MovieSceneAudioSection, default=null}
 SourceDisplayName {type=string, default=""}
 SourceSoundAsset {type=SoundBase, default=null}
 StartFrame {type=integer}
 StatusText {type=string, default=Missing}
]
widgets {
 info {rootWidgetClass=&Border, widgetCount=18}
 widgets [
  Border_Row {widgetClassPath=&Border, widget=Border_Row, properties {brushColor=<rgba 0.0137 0.016 0.0212 1>}}
  SizeBox_Row {
   widgetClassPath=&SizeBox
   parent=Border_Row
   widget=SizeBox_Row
   properties {bOverride_HeightOverride=true, heightOverride=34}
   slot {class=BorderSlot}
  }
  HB_Row {widgetClassPath=&HorizontalBox, parent=SizeBox_Row, widget=HB_Row, slot {class=SizeBoxSlot}}
  SizeBox_41 {
   widgetClassPath=&SizeBox
   parent=HB_Row
   widget=SizeBox_41
   properties {bOverride_WidthOverride=true, widthOverride=10}
   slot {class=HorizontalBoxSlot, properties {verticalAlignment=VAlign_Center, padding=<ltrb 8 0 8 0>}}
  }
  Text_RowIndex {
   widgetClassPath=&UEFN_TextBlock
   bIsVariable=true
   parent=SizeBox_41
   widget=Text_RowIndex
   properties {
    justification=Center
    text=01
    colorAndOpacity {colorUseRule=UseColor_Specified, specifiedColor=<rgba 0.3968 0.4342 0.491 1>}
    font {
     bForceMonospaced=false
     bMaterialIsStencil=false
     fontMaterial=null
     fontObject=/LauncherSocial/BrutalType
     letterSpacing=0
     monospacedWidth=1
     size=9
     skewAmount=0
     typefaceFontName=Regular
     outlineSettings {
      bApplyOutlineToDropShadows=false
      bMiteredCorners=false
      bSeparateFillAlpha=false
      outlineMaterial=null
      outlineSize=0
      outlineColor=<rgba 0 0 0 1>
     }
    }
    shadowColorAndOpacity=<rgba 0 0 0 1>
   }
   slot {class=SizeBoxSlot}
  }
  Button_SourceSound {
   widgetClassPath=&EditorUtilityButton
   parent=HB_Row
   widget=Button_SourceSound
   properties {
    toolTipText=Click to navigate to this asset in the Content Browser
    widgetStyle {
     clickedSlateSound {resourceObject=null}
     disabled {
      bIsDynamicallyLoaded=false
      drawAs=RoundedBox
      imageType=NoImage
      mirroring=NoMirror
      resourceName=None
      resourceObject=null
      tiling=NoTile
      imageSize=<0 0>
      margin=<ltrb 0 0 0 0>
      outlineSettings {
       bUseBrushTransparency=false
       roundingType=FixedRadius
       width=1
       color {colorUseRule=UseColor_ColorTable, specifiedColor=<rgba 0 0 0 0>}
       cornerRadii=<4 4 4 4>
      }
      tintColor {colorUseRule=UseColor_Specified, specifiedColor=<rgba 0.0395 0.0395 0.0395 1>}
      uVRegion {bIsValid=false, max=<0 0>, min=<0 0>}
     }
     disabledForeground {colorUseRule=UseColor_ColorTable, specifiedColor=<rgba 0 0 0 0>}
     hovered {
      bIsDynamicallyLoaded=false
      drawAs=RoundedBox
      imageType=NoImage
      mirroring=NoMirror
      resourceName=None
      resourceObject=null
      tiling=NoTile
      imageSize=<32 32>
      margin=<ltrb 0 0 0 0>
      outlineSettings {
       bUseBrushTransparency=false
       roundingType=FixedRadius
       width=1
       color {colorUseRule=UseColor_ColorTable, specifiedColor=<rgba 0 0 0 0>}
       cornerRadii=<4 4 4 4>
      }
      tintColor {colorUseRule=UseColor_Specified, specifiedColor=<rgba 0.0273 0.314 0.8388 0.2941>}
      uVRegion {bIsValid=false, max=<0 0>, min=<0 0>}
     }
     hoveredForeground {colorUseRule=UseColor_ColorTable, specifiedColor=<rgba 0 0 0 0>}
     hoveredSlateSound {resourceObject=null}
     normal {
      bIsDynamicallyLoaded=false
      drawAs=RoundedBox
      imageType=NoImage
      mirroring=NoMirror
      resourceName=None
      resourceObject=null
      tiling=NoTile
      imageSize=<32 32>
      margin=<ltrb 0 0 0 0>
      outlineSettings {
       bUseBrushTransparency=false
       roundingType=FixedRadius
       width=1
       color {colorUseRule=UseColor_ColorTable, specifiedColor=<rgba 0 0 0 0>}
       cornerRadii=<4 4 4 4>
      }
      tintColor {colorUseRule=UseColor_Specified, specifiedColor=<rgba 0.0232 0.0262 0.0331 1>}
      uVRegion {bIsValid=false, max=<0 0>, min=<0 0>}
     }
     normalForeground {colorUseRule=UseColor_ColorTable, specifiedColor=<rgba 0 0 0 0>}
     normalPadding=<ltrb 8 6 8 6>
     pressed {
      bIsDynamicallyLoaded=false
      drawAs=RoundedBox
      imageType=NoImage
      mirroring=NoMirror
      resourceName=None
      resourceObject=null
      tiling=NoTile
      imageSize=<32 32>
      margin=<ltrb 0 0 0 0>
      outlineSettings {
       bUseBrushTransparency=false
       roundingType=FixedRadius
       width=1
       color {colorUseRule=UseColor_ColorTable, specifiedColor=<rgba 0 0 0 0>}
       cornerRadii=<4 4 4 4>
      }
      tintColor {colorUseRule=UseColor_Specified, specifiedColor=<rgba 0.008 0.0916 0.2448 0.2941>}
      uVRegion {bIsValid=false, max=<0 0>, min=<0 0>}
     }
     pressedForeground {colorUseRule=UseColor_ColorTable, specifiedColor=<rgba 0 0 0 0>}
     pressedPadding=<ltrb 8 6 8 6>
     pressedSlateSound {resourceObject=null}
    }
   }
   slot {class=HorizontalBoxSlot, properties {padding=<ltrb 0 0 8 0>, size {sizeRule=Fill, value=1}}}
  }
  HorizontalBox_0 {
   widgetClassPath=&HorizontalBox
   parent=Button_SourceSound
   widget=HorizontalBox_0
   slot {class=ButtonSlot, properties {horizontalAlignment=HAlign_Left, padding=<ltrb 0 0 0 0>}}
  }
  Image_RowSourceIcon {
   widgetClassPath=/Script/TechAudioToolsContentEditor.TechAudioToolsEditorSlateIcon
   bIsVariable=true
   parent=HorizontalBox_0
   widget=Image_RowSourceIcon
   properties {bOverride_Cursor=true}
   slot {class=HorizontalBoxSlot, properties {horizontalAlignment=HAlign_Center, verticalAlignment=VAlign_Center, padding=<ltrb 0 0 2 0>}}
  }
  VerticalBox_1 {
   widgetClassPath=&VerticalBox
   parent=HorizontalBox_0
   widget=VerticalBox_1
   properties {bOverride_Cursor=true}
   slot {class=HorizontalBoxSlot}
  }
  Text_RowSourceName {
   widgetClassPath=&UEFN_TextBlock
   bIsVariable=true
   parent=VerticalBox_1
   widget=Text_RowSourceName
   properties {
    minDesiredWidth=250
    text=Test
    colorAndOpacity {colorUseRule=UseColor_Specified, specifiedColor=<rgba 0.3968 0.4342 0.491 1>}
    font {
     bForceMonospaced=false
     bMaterialIsStencil=false
     fontMaterial=null
     fontObject=&BrutalType
     letterSpacing=0
     monospacedWidth=1
     size=10.5
     skewAmount=0
     typefaceFontName=Regular
     outlineSettings {
      bApplyOutlineToDropShadows=false
      bMiteredCorners=false
      bSeparateFillAlpha=false
      outlineMaterial=null
      outlineSize=0
      outlineColor=<rgba 0 0 0 1>
     }
    }
    shadowColorAndOpacity=<rgba 0 0 0 1>
   }
   slot {class=VerticalBoxSlot}
  }
  Text_RowArrow {
   widgetClassPath=&UEFN_TextBlock
   parent=HB_Row
   widget=Text_RowArrow
   properties {
    justification=Center
    text="->"
    colorAndOpacity {colorUseRule=UseColor_Specified, specifiedColor=<rgba 0.3968 0.4342 0.491 1>}
    font {
     bForceMonospaced=false
     bMaterialIsStencil=false
     fontMaterial=null
     fontObject=&BrutalType
     letterSpacing=0
     monospacedWidth=1
     size=9
     skewAmount=0
     typefaceFontName=Regular
     outlineSettings {
      bApplyOutlineToDropShadows=false
      bMiteredCorners=false
      bSeparateFillAlpha=false
      outlineMaterial=null
      outlineSize=0
      outlineColor=<rgba 0 0 0 1>
     }
    }
    shadowColorAndOpacity=<rgba 0 0 0 1>
   }
   slot {class=HorizontalBoxSlot, properties {horizontalAlignment=HAlign_Left, verticalAlignment=VAlign_Center, padding=<ltrb 0 0 8 0>}}
  }
  HorizontalBox_1 {
   widgetClassPath=&Overlay
   parent=HB_Row
   widget=HorizontalBox_1
   properties {bOverride_Cursor=true}
   slot {class=HorizontalBoxSlot, properties {size {sizeRule=Fill, value=1}}}
  }
  AudioAssetDropTargetWidget_Row {
   widgetClassPath=/Script/FortTechAudioToolsEditor.AudioAssetDropTargetWidget
   bIsVariable=true
   parent=HorizontalBox_1
   widget=AudioAssetDropTargetWidget_Row
   properties {
    text="      Drop Asset Here"
    textJustification=Left
    toolTipText=Drag audio assets here or click to assign from Content Browser selection
    allowedAssetClasses [&SoundBase]
    colorHovered {colorUseRule=UseColor_Specified, specifiedColor=<rgba 0.0273 0.314 0.8388 0.2941>}
    colorNormal {colorUseRule=UseColor_Specified, specifiedColor=<rgba 0.0232 0.0262 0.0331 1>}
    colorPressed {colorUseRule=UseColor_Specified, specifiedColor=<rgba 0.008 0.0916 0.2448 0.2941>}
    font {
     bForceMonospaced=false
     bMaterialIsStencil=false
     fontMaterial=null
     fontObject=&BrutalType
     letterSpacing=0
     monospacedWidth=1
     size=10.5
     skewAmount=0
     typefaceFontName=Regular
     outlineSettings {
      bApplyOutlineToDropShadows=false
      bMiteredCorners=false
      bSeparateFillAlpha=false
      outlineMaterial=null
      outlineSize=0
      outlineColor=<rgba 0 0 0 1>
     }
    }
    shadowColorAndOpacity=<rgba 0 0 0 1>
    textColor {colorUseRule=UseColor_Specified, specifiedColor=<rgba 0.3968 0.4342 0.491 1>}
   }
   slot {class=OverlaySlot, properties {horizontalAlignment=HAlign_Fill, verticalAlignment=VAlign_Fill, padding=<ltrb 0 0 2 0>}}
  }
  SizeBox_RowReplacementIcon {
   widgetClassPath=&SizeBox
   parent=HorizontalBox_1
   widget=SizeBox_RowReplacementIcon
   properties {bOverride_WidthOverride=true, widthOverride=30}
   slot {class=OverlaySlot, properties {verticalAlignment=VAlign_Fill}}
  }
  Image_RowReplacementIcon {
   widgetClassPath=&TechAudioToolsEditorSlateIcon
   bIsVariable=true
   parent=SizeBox_RowReplacementIcon
   widget=Image_RowReplacementIcon
   slot {class=SizeBoxSlot, properties {horizontalAlignment=HAlign_Center, verticalAlignment=VAlign_Center}}
  }
  SizeBox_0 {
   widgetClassPath=&SizeBox
   parent=HB_Row
   widget=SizeBox_0
   properties {bOverride_WidthOverride=true, widthOverride=60}
   slot {class=HorizontalBoxSlot}
  }
  Border_RowStatus {
   widgetClassPath=&Border
   bIsVariable=true
   parent=SizeBox_0
   widget=Border_RowStatus
   properties {horizontalAlignment=HAlign_Left, verticalAlignment=VAlign_Center, brushColor=<rgba 0.0232 0.0262 0.0331 1>, padding=<ltrb 8 6 8 6>}
   slot {class=SizeBoxSlot}
  }
  Text_RowStatus {
   widgetClassPath=&UEFN_TextBlock
   bIsVariable=true
   parent=Border_RowStatus
   widget=Text_RowStatus
   properties {
    justification=Center
    text=Missing
    colorAndOpacity {colorUseRule=UseColor_Specified, specifiedColor=<rgba 0.7913 0.7991 0.8228 1>}
    font {
     bForceMonospaced=false
     bMaterialIsStencil=false
     fontMaterial=null
     fontObject=&BrutalType
     letterSpacing=0
     monospacedWidth=1
     size=9
     skewAmount=0
     typefaceFontName=Regular
     outlineSettings {
      bApplyOutlineToDropShadows=false
      bMiteredCorners=false
      bSeparateFillAlpha=false
      outlineMaterial=null
      outlineSize=0
      outlineColor=<rgba 0 0 0 1>
     }
    }
    shadowColorAndOpacity=<rgba 0 0 0 1>
   }
   slot {class=BorderSlot, properties {horizontalAlignment=HAlign_Left, verticalAlignment=VAlign_Center, padding=<ltrb 8 6 8 6>}}
  }
 ]
}
