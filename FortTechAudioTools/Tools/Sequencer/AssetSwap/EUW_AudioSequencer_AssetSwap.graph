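# Editor utility widget that bulk-swaps sounds on a Level Sequence's audio sections.
# It builds a filterable, sortable row per section, with replacements dropped manually or auto-matched by weighted name-token scoring.
# An embedded base64 Python script creates SoundCues for sequence waves, plus DialogueVoices and DialogueWaves for VO_ waves, as match candidates.
# Executing applies replacements, auto-sizes eligible non-looping sections and warns when resized sections overlap others.
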EditorUtilityWidgetBlueprint /FortTechAudioTools/Tools/Sequencer/AssetSwap/EUW_AudioSequencer_AssetSwap
graphs [
 AddSoundRow {nodes [
   n1 = AddSoundRow!() -> (then, InRowIndex:int, InSourceAudioSection:Movie Scene Audio Section Object Reference) {is=title K2Node_FunctionEntry}
   > n2 = Create Widget!(
    execute:Exec
    WorldContextObject:Object Reference
    Class:User Widget Class Reference=&/EUW_AudioSequencer_AssetSwap_SoundSwapRow.EUW_AudioSequencer_AssetSwap_SoundSwapRow_C
    OwningPlayer:Player Controller Object Reference
   ) -> (then:Exec, ReturnValue:EUW Audio Sequencer Asset Swap Sound Swap Row Object Reference)
   > ApplyRowData!(
    execute
    self:EUW Audio Sequencer Asset Swap Sound Swap Row Object Reference=@n2
    InRowIndex:int=@n1
    InSourceAudioSection:Movie Scene Audio Section Object Reference=@n1.InSourceAudioSection
    InFrameRate:Frame Rate Structure
   ) -> (then)
   > AddChild(self=Get ScrollBox_SoundRows() -> (ScrollBox_SoundRows:Editor Utility Scroll Box Object Reference), Content=@n2)
   > Add Unique(
    TargetArray:Array of EUW Audio Sequencer Asset Swap Sound Swap Row Object References=Get GeneratedRowWidgets() -> (GeneratedRowWidgets:Array of EUW Audio Sequencer Asset Swap Sound Swap Row Object References)
    NewItem:"EUW Audio Sequencer Asset Swap Sound Swap Row Object Reference (by ref)"=@n2
   )
   > Bind Event to On Multiple Assets Dropped on Row(self:EUW Audio Sequencer Asset Swap Sound Swap Row Object Reference=@n2, Delegate=Create Event()) {is=title K2Node_AddDelegate}
   > Bind Event to On Selection Changed(self:EUW Audio Sequencer Asset Swap Sound Swap Row Object Reference=@n2, Delegate=Create Event()) {is=title K2Node_AddDelegate}
  ]}
 AutoMatchReplacements {nodes [
   AutoMatchReplacements() {is=title K2Node_FunctionEntry}
   > n1 = For Each Loop(Array:Array of EUW Audio Sequencer Asset Swap Sound Swap Row Object References=Get GeneratedRowWidgets()) -> (Array Element:EUW Audio Sequencer Asset Swap Sound Swap Row Object Reference)
   RefreshSummaryText(execute=@n1.Completed)
   > SortRows()
   n2 = Set CurrentRowWidget(execute=@n1.LoopBody, CurrentRowWidget:EUW Audio Sequencer Asset Swap Sound Swap Row Object Reference=@n1) -> (Output_Get:EUW Audio Sequencer Asset Swap Sound Swap Row Object Reference)
   > n3 = Branch(Condition=MatchedOrManual!(self:EUW Audio Sequencer Asset Swap Sound Swap Row Object Reference=@n2) -> (bIsMatched:bool))
   Set BestMatchScore(execute=@n3.else, BestMatchScore:double=0) -> (Output_Get:double)
   > Set BestMatchName(BestMatchName:String) -> (Output_Get:String)
   > Set BestMatchSound(BestMatchSound:Sound Base Object Reference) -> (Output_Get:Sound Base Object Reference)
   > n4 = For Each Loop(Array:Array of Sound Base Object References=Get ReplacementCandidates() -> (ReplacementCandidates:Array of Sound Base Object References)) -> (Array Element:Sound Base Object Reference)
   > n5 = Set CandidateDisplayName(CandidateDisplayName:String=GetDisplayName(Object=@n4)) -> (Output_Get:String)
   > n6 = GetAudioMatchScore!(
    execute
    self
    SourceName:String=Get SourceDisplayName!(self:EUW Audio Sequencer Asset Swap Sound Swap Row Object Reference=Get CurrentRowWidget() -> (CurrentRowWidget:EUW Audio Sequencer Asset Swap Sound Swap Row Object Reference)) -> (SourceDisplayName:String)
    CandidateName:String=@n5
   ) -> (then, Score:double)
   > n7 = Set CurrentScore(CurrentScore:double=@n6) -> (Output_Get:double)
   > Branch(Condition="float > float"(A=@n7, B=Get BestMatchScore() -> (BestMatchScore:double)))
   > Set BestMatchScore(BestMatchScore=Get CurrentScore() -> (CurrentScore:double))
   > Set BestMatchName(BestMatchName=Get CandidateDisplayName() -> (CandidateDisplayName:String))
   > Set BestMatchSound(BestMatchSound=@n4)
   n8 = Branch(execute=@n4.Completed, Condition="float >= float"(A=Get BestMatchScore(), B=0.75))
   n9 = Get CurrentRowWidget()
   > SetReplacementState!(
    execute
    self:EUW Audio Sequencer Asset Swap Sound Swap Row Object Reference=@n9
    InReplacementDisplayName:String=Get BestMatchName() -> (BestMatchName:String)
    InStatusText:String=Get Matched!(self:EUW Audio Sequencer Asset Swap Sound Swap Row Object Reference=@n9) -> (Matched:String)
    InReplacementSound:Sound Base Object Reference=Get BestMatchSound() -> (BestMatchSound:Sound Base Object Reference)
    InMatchScore:double=Get BestMatchScore()
   ) -> (then)
   n10 = Branch(execute=@n8.else, Condition="float >= float"(A=Get BestMatchScore(), B=60))
   n11 = Get CurrentRowWidget()
   > SetReplacementState(
    self=@n11
    InReplacementDisplayName=Get BestMatchName()
    InStatusText=Get NeedsReview!(self:EUW Audio Sequencer Asset Swap Sound Swap Row Object Reference=@n11) -> (NeedsReview:String)
    InReplacementSound=Get BestMatchSound()
    InMatchScore=Get BestMatchScore()
   )
   SetReplacementState(execute=@n10.else, self=Get CurrentRowWidget(), InReplacementDisplayName=Drop Asset Here, InStatusText=Missing, InMatchScore=0)
  ]}
 CheckForOverlap {nodes [
   CheckForOverlap!() -> (then, "ResizedSection ":Movie Scene Audio Section Object Reference, ResizedSections:Array of Movie Scene Audio Section Object References) {is=title K2Node_FunctionEntry}
   n1 = GetOuterObject!(Object:Object Reference="Get ResizedSection "() -> ("ResizedSection ":Movie Scene Audio Section Object Reference)) -> (ReturnValue:Object Reference)
   > n2 = Is Valid(InputObject=@n1)
   > n3 = Cast To MovieSceneAudioTrack!(execute, Object=@n1) -> (then, CastFailed, AsMovie Scene Audio Track:Movie Scene Audio Track Object Reference) {is=title K2Node_DynamicCast}
   > n4 = GetSections!(execute, Track:Movie Scene Track Object Reference=@n3) -> (then, ReturnValue:Array of Movie Scene Section Object References)
   > n5 = For Each Loop(Array:Array of Movie Scene Section Object References=@n4) -> (Array Element:Movie Scene Section Object Reference)
   > n6 = Branch(Condition="Equal (Object)"(A=@n5, B="Get ResizedSection "()))
   n7 = "Get ResizedSection "()
   Branch(
    execute=@n6.else
    Condition="Equal (Integer)"(A=GetRowIndex!(self:Movie Scene Section Object Reference=@n7) -> (ReturnValue:int=0), B=GetRowIndex(self=@n5) -> (ReturnValue=0))
   )
   > n8 = Cast To MovieSceneAudioSection!(execute, Object=@n5) -> (then, CastFailed, AsMovie Scene Audio Section:Movie Scene Audio Section Object Reference) {is=title K2Node_DynamicCast}
   > n9 = Branch(Condition=Contains Item(
     TargetArray:Array of Movie Scene Audio Section Object References=Get ResizedSections() -> (ResizedSections:Array of Movie Scene Audio Section Object References)
     ItemToFind:"Movie Scene Audio Section Object Reference (by ref)"=@n8
    ))
   n10 = GetEndFrame!(execute=@n9.else, Section:Movie Scene Section Object Reference="Get ResizedSection "()) -> (then, ReturnValue:int=0)
   > n11 = GetStartFrame!(execute, Section:Movie Scene Section Object Reference=@n8) -> (then, ReturnValue:int=0)
   > n12 = GetStartFrame(Section="Get ResizedSection "()) -> (ReturnValue=0)
   > n13 = GetEndFrame(Section=@n8) -> (ReturnValue=0)
   > Branch(Condition=AND Boolean(A="integer > integer"(A=@n13, B=@n12), B="integer > integer"(A=@n10, B=@n11)))
   > Return Node!(
    execute
    bHasOverlap:bool=true
    OverlapFrames:int=int - int(A=@n10, B=@n11)
    AsMovie Scene Audio Section:Movie Scene Audio Section Object Reference=@n8
   )
   Return Node(execute=[@n5.Completed, @n3.CastFailed, @n2.Is Not Valid], bHasOverlap=false, OverlapFrames=0)
  ]}
 ClearAllSelected {nodes [
   ClearAllSelected() {is=title K2Node_FunctionEntry}
   > n1 = For Each Loop(Array:Array of EUW Audio Sequencer Asset Swap Sound Swap Row Object References=Get SelectedRows() -> (SelectedRows:Array of EUW Audio Sequencer Asset Swap Sound Swap Row Object References)) -> (Array Element:EUW Audio Sequencer Asset Swap Sound Swap Row Object Reference)
   SortRows(execute=@n1.Completed)
   SetReplacementState(execute=@n1.LoopBody, self=@n1, InReplacementDisplayName=Drop Asset Here, InStatusText=Missing, InMatchScore=0)
  ]}
 ClearAllSelections {nodes [
   ClearAllSelections() {is=title K2Node_FunctionEntry}
   > n1 = For Each Loop(Array=Get SelectedRows())
   Clear(execute=@n1.Completed, TargetArray:Array of EUW Audio Sequencer Asset Swap Sound Swap Row Object References=Get SelectedRows())
   SetSelected!(
    execute=@n1.LoopBody
    self:Audio Asset Drop Target Widget Object Reference=Get AudioAssetDropTargetWidget_Row!(self:EUW Audio Sequencer Asset Swap Sound Swap Row Object Reference=@n1) -> (AudioAssetDropTargetWidget_Row:Audio Asset Drop Target Widget Object Reference)
    bSelected:bool=false
   ) -> (then)
  ]}
 CreateCuesAndAutoMatch {nodes [
   CreateCuesAndAutoMatch() {is=title K2Node_FunctionEntry}
   > Clear(TargetArray:Array of Sound Base Object References=Get ReplacementCandidates())
   > Set ReplacementCandidateCount(ReplacementCandidateCount:int=0) -> (Output_Get:int=0)
   n1 = Get SelectedLevelSequence() -> (SelectedLevelSequence:Level Sequence Object Reference)
   > n2 = Is Valid(InputObject=@n1)
   "SetText (Text)"(
    execute=@n2.Is Not Valid
    self=Get Text_ToolbarStatus() -> (Text_ToolbarStatus:Text Block Object Reference)
    InText=MakeLiteralText(Value="LOCTEXT(\"No sequence loaded. Drop a sequence first.\")")
   )
   n3 = MakeLiteralString(Value="CmltcG9ydCB1bnJlYWwKaW1wb3J0IHRpbWUKCmRlZiBnZXRfZm9sZGVyKHBrZyk6IHJldHVybiAnLycuam9pbihwa2cuc3BsaXQoJy8nKVs6LTFdKQpkZWYgZ2V0X25hbWUocGtnKTogcmV0dXJuIHBrZy5zcGxpdCgnLycpWy0xXQoKZGVmIGZpbmRfYXNzZXRfYnlfbmFtZV9hbmRfY2xhc3MoZm9sZGVyLCBhc3NldF9uYW1lLCBjbGFzc19uYW1lKToKICAgIGFyID0gdW5yZWFsLkFzc2V0UmVnaXN0cnlIZWxwZXJzLmdldF9hc3NldF9yZWdpc3RyeSgpCiAgICBhc3NldHMgPSBhci5nZXRfYXNzZXRzX2J5X3BhdGgoZm9sZGVyLCByZWN1cnNpdmU9RmFsc2UpCiAgICBmb3IgYSBpbiBhc3NldHM6CiAgICAgICAgaWYgc3RyKGEuYXNzZXRfbmFtZSkgPT0gYXNzZXRfbmFtZSBhbmQgc3RyKGEuYXNzZXRfY2xhc3NfcGF0aC5hc3NldF9uYW1lKSA9PSBjbGFzc19uYW1lOgogICAgICAgICAgICByZXR1cm4gc3RyKGEucGFja2FnZV9wYXRoKSArICcvJyArIHN0cihhLmFzc2V0X25hbWUpCiAgICByZXR1cm4gTm9uZQoKYXIgPSB1bnJlYWwuQXNzZXRSZWdpc3RyeUhlbHBlcnMuZ2V0X2Fzc2V0X3JlZ2lzdHJ5KCkKYXQgPSB1bnJlYWwuQXNzZXRUb29sc0hlbHBlcnMuZ2V0X2Fzc2V0X3Rvb2xzKCkKCnNlcSA9IHVucmVhbC5sb2FkX2Fzc2V0KFNFUV9QQVRIKQppZiBub3Qgc2VxOgogICAgcHJpbnQoZidFUlJPUjogQ291bGQgbm90IGxvYWQgc2VxdWVuY2UgYXQge1NFUV9QQVRIfScpCmVsc2U6CiAgICB3YXZlX3BhdGhzID0gW10KICAgIGZvciB0cmFjayBpbiBzZXEuZmluZF90cmFja3NfYnlfdHlwZSh1bnJlYWwuTW92aWVTY2VuZUF1ZGlvVHJhY2spOgogICAgICAgIGZvciBzZWN0aW9uIGluIHRyYWNrLmdldF9zZWN0aW9ucygpOgogICAgICAgICAgICBzb3VuZCA9IHNlY3Rpb24uZ2V0X2VkaXRvcl9wcm9wZXJ0eSgnc291bmQnKQogICAgICAgICAgICBpZiBzb3VuZCBhbmQgaXNpbnN0YW5jZShzb3VuZCwgdW5yZWFsLlNvdW5kV2F2ZSk6CiAgICAgICAgICAgICAgICBwa2cgPSBzb3VuZC5nZXRfcGF0aF9uYW1lKCkuc3BsaXQoJy4nKVswXQogICAgICAgICAgICAgICAgaWYgcGtnIG5vdCBpbiB3YXZlX3BhdGhzOgogICAgICAgICAgICAgICAgICAgIHdhdmVfcGF0aHMuYXBwZW5kKHBrZykKCiAgICBwcmludChmJ0ZvdW5kIHtsZW4od2F2ZV9wYXRocyl9IHVuaXF1ZSB3YXZlcycpCgogICAgIyA9PT09PT09PT09PT09PT09PT09PT09PT09PT09PT09PT09PT09PT09PT09PT09PT09PT09PT09PT0KICAgICMgU1RFUCAxOiBDcmVhdGUgRGlhbG9ndWVWb2ljZXMKICAgICMgPT09PT09PT09PT09PT09PT09PT09PT09PT09PT09PT09PT09PT09PT09PT09PT09PT09PT09PT09CiAgICBwcmludCgnLS0tIFN0ZXAgMTogQ3JlYXRpbmcgRGlhbG9ndWVWb2ljZXMgLS0tJykKICAgIGR2X21hcCA9IHt9CiAgICBmb3Igd2F2ZV9wa2cgaW4gd2F2ZV9wYXRoczoKICAgICAgICB3YXZlX2ZvbGRlciA9IGdldF9mb2xkZXIod2F2ZV9wa2cpCiAgICAgICAgd2F2ZV9uYW1lICAgPSBnZXRfbmFtZSh3YXZlX3BrZykKICAgICAgICBpZiBub3Qgd2F2ZV9uYW1lLnN0YXJ0c3dpdGgoJ1ZPXycpOgogICAgICAgICAgICBjb250aW51ZQogICAgICAgIHBhcnRzID0gd2F2ZV9uYW1lLnNwbGl0KCdfJykKICAgICAgICBjaGFyX25hbWUgPSBwYXJ0c1sxXSBpZiBsZW4ocGFydHMpID4gMSBlbHNlICdVbmtub3duJwogICAgICAgIGR2X25hbWUgPSBmJ0RWX3tjaGFyX25hbWV9JwogICAgICAgIGlmIGR2X25hbWUgaW4gZHZfbWFwOgogICAgICAgICAgICBjb250aW51ZQogICAgICAgIGFyLnNjYW5fcGF0aHNfc3luY2hyb25vdXMoW3dhdmVfZm9sZGVyXSwgZm9yY2VfcmVzY2FuPVRydWUpCiAgICAgICAgZXhpc3RpbmdfZHYgPSBmaW5kX2Fzc2V0X2J5X25hbWVfYW5kX2NsYXNzKHdhdmVfZm9sZGVyLCBkdl9uYW1lLCAnRGlhbG9ndWVWb2ljZScpCiAgICAgICAgaWYgZXhpc3RpbmdfZHY6CiAgICAgICAgICAgIHByaW50KGYnICBEViBleGlzdHM6IHtkdl9uYW1lfScpCiAgICAgICAgICAgIGR2X21hcFtkdl9uYW1lXSA9IGV4aXN0aW5nX2R2CiAgICAgICAgZWxzZToKICAgICAgICAgICAgZHYgPSBhdC5jcmVhdGVfYXNzZXQoZHZfbmFtZSwgd2F2ZV9mb2xkZXIsIHVucmVhbC5EaWFsb2d1ZVZvaWNlLCB1bnJlYWwuRGlhbG9ndWVWb2ljZUZhY3RvcnkoKSkKICAgICAgICAgICAgdW5yZWFsLkVkaXRvckFzc2V0TGlicmFyeS5zYXZlX2Fzc2V0KGR2LmdldF9wYXRoX25hbWUoKSkKICAgICAgICAgICAgYXIuc2Nhbl9wYXRoc19zeW5jaHJvbm91cyhbd2F2ZV9mb2xkZXJdLCBmb3JjZV9yZXNjYW49VHJ1ZSkKICAgICAgICAgICAgZHZfbWFwW2R2X25hbWVdID0gd2F2ZV9mb2xkZXIgKyAnLycgKyBkdl9uYW1lCiAgICAgICAgICAgIHByaW50KGYnICBEViBjcmVhdGVkOiB7ZHZfbmFtZX0nKQoKICAgIHByaW50KGYnU3RlcCAxIGRvbmUuIHtsZW4oZHZfbWFwKX0gRGlhbG9ndWVWb2ljZXMgcmVhZHkuJykKICAgIHRpbWUuc2xlZXAoMSkKCiAgICAjID09PT09PT09PT09PT09PT09PT09PT09PT09PT09PT09PT09PT09PT09PT09PT09PT09PT09PT09PQogICAgIyBTVEVQIDI6IENyZWF0ZSBEaWFsb2d1ZVdhdmVzIChiYXJlIC0gbm8gY29udGV4dCB5ZXQpCiAgICAjID09PT09PT09PT09PT09PT09PT09PT09PT09PT09PT09PT09PT09PT09PT09PT09PT09PT09PT09PQogICAgcHJpbnQoJy0tLSBTdGVwIDI6IENyZWF0aW5nIERpYWxvZ3VlV2F2ZXMgLS0tJykKICAgIGR3X21hcCA9IHt9CiAgICBmb3Igd2F2ZV9wa2cgaW4gd2F2ZV9wYXRoczoKICAgICAgICB3YXZlX2ZvbGRlciA9IGdldF9mb2xkZXIod2F2ZV9wa2cpCiAgICAgICAgd2F2ZV9uYW1lICAgPSBnZXRfbmFtZSh3YXZlX3BrZykKICAgICAgICBpZiBub3Qgd2F2ZV9uYW1lLnN0YXJ0c3dpdGgoJ1ZPXycpOgogICAgICAgICAgICBjb250aW51ZQogICAgICAgIGR3X25hbWUgPSBmJ3t3YXZlX25hbWV9X0RpYWxvZ3VlJwogICAgICAgIGFyLnNjYW5fcGF0aHNfc3luY2hyb25vdXMoW3dhdmVfZm9sZGVyXSwgZm9yY2VfcmVzY2FuPVRydWUpCiAgICAgICAgZXhpc3RpbmdfZHcgPSBmaW5kX2Fzc2V0X2J5X25hbWVfYW5kX2NsYXNzKHdhdmVfZm9sZGVyLCBkd19uYW1lLCAnRGlhbG9ndWVXYXZlJykKICAgICAgICBpZiBleGlzdGluZ19kdzoKICAgICAgICAgICAgcHJpbnQoZicgIERXIGV4aXN0czoge2R3X25hbWV9JykKICAgICAgICAgICAgZHdfbWFwW3dhdmVfcGtnXSA9IGV4aXN0aW5nX2R3CiAgICAgICAgZWxzZToKICAgICAgICAgICAgZHcgPSBhdC5jcmVhdGVfYXNzZXQoZHdfbmFtZSwgd2F2ZV9mb2xkZXIsIHVucmVhbC5EaWFsb2d1ZVdhdmUsIHVucmVhbC5EaWFsb2d1ZVdhdmVGYWN0b3J5KCkpCiAgICAgICAgICAgIHVucmVhbC5FZGl0b3JBc3NldExpYnJhcnkuc2F2ZV9hc3NldChkdy5nZXRfcGF0aF9uYW1lKCkpCiAgICAgICAgICAgIGFyLnNjYW5fcGF0aHNfc3luY2hyb25vdXMoW3dhdmVfZm9sZGVyXSwgZm9yY2VfcmVzY2FuPVRydWUpCiAgICAgICAgICAgIGR3X21hcFt3YXZlX3BrZ10gPSB3YXZlX2ZvbGRlciArICcvJyArIGR3X25hbWUKICAgICAgICAgICAgcHJpbnQoZicgIERXIGNyZWF0ZWQ6IHtkd19uYW1lfScpCgogICAgcHJpbnQoZidTdGVwIDIgZG9uZS4ge2xlbihkd19tYXApfSBEaWFsb2d1ZVdhdmVzIHJlYWR5LicpCiAgICB0aW1lLnNsZWVwKDEpCgogICAgIyA9PT09PT09PT09PT09PT09PT09PT09PT09PT09PT09PT09PT09PT09PT09PT09PT09PT09PT09PT0KICAgICMgU1RFUCAzOiBXaXJlIGVhY2ggRFcgLSBTcGVha2VyICsgU291bmRXYXZlLCBEaXJlY3RlZCBBdCBlbXB0eQogICAgIyBUb3VjaCBzcG9rZW5fdGV4dCB0byB0cmlnZ2VyIFBvc3RFZGl0Q2hhbmdlQ2hhaW5Qcm9wZXJ0eQogICAgIyAtPiBVcGRhdGVNYXBwaW5nUHJveHkgLT4gYnVpbGRzIFByb3h5IC0+IGVuYWJsZXMgcmlnaHQtY2xpY2sgUGxheQogICAgIyA9PT09PT09PT09PT09PT09PT09PT09PT09PT09PT09PT09PT09PT09PT09PT09PT09PT09PT09PT0KICAgIHByaW50KCctLS0gU3RlcCAzOiBXaXJpbmcgRGlhbG9ndWVXYXZlcyArIGJ1aWxkaW5nIFByb3h5IHZpYSBzcG9rZW5fdGV4dCAtLS0nKQogICAgZm9yIHdhdmVfcGtnLCBkd19wYXRoIGluIGR3X21hcC5pdGVtcygpOgogICAgICAgIHdhdmVfbmFtZSA9IGdldF9uYW1lKHdhdmVfcGtnKQogICAgICAgIHBhcnRzID0gd2F2ZV9uYW1lLnNwbGl0KCdfJykKICAgICAgICBjaGFyX25hbWUgPSBwYXJ0c1sxXSBpZiBsZW4ocGFydHMpID4gMSBlbHNlICdVbmtub3duJwogICAgICAgIGR2X25hbWUgPSBmJ0RWX3tjaGFyX25hbWV9JwoKICAgICAgICB3YXZlID0gdW5yZWFsLmxvYWRfYXNzZXQod2F2ZV9wa2cpCiAgICAgICAgZHcgICA9IHVucmVhbC5sb2FkX2Fzc2V0KGR3X3BhdGgpCiAgICAgICAgZHYgICA9IHVucmVhbC5sb2FkX2Fzc2V0KGR2X21hcFtkdl9uYW1lXSkgaWYgZHZfbmFtZSBpbiBkdl9tYXAgZWxzZSBOb25lCgogICAgICAgIGlmIG5vdCBkdyBvciBub3Qgd2F2ZSBvciBub3QgZHY6CiAgICAgICAgICAgIHByaW50KGYnICBTS0lQIChtaXNzaW5nIGFzc2V0KTogZHc9e2R3fSwgd2F2ZT17d2F2ZX0sIGR2PXtkdn0nKQogICAgICAgICAgICBjb250aW51ZQoKICAgICAgICBjdHhfbWFwcGluZyA9IHVucmVhbC5EaWFsb2d1ZUNvbnRleHRNYXBwaW5nKCkKICAgICAgICBjdHhfbWFwcGluZy5zZXRfZWRpdG9yX3Byb3BlcnR5KCdzb3VuZF93YXZlJywgd2F2ZSkKICAgICAgICBkaWFsb2d1ZV9jdHggPSB1bnJlYWwuRGlhbG9ndWVDb250ZXh0KCkKICAgICAgICBkaWFsb2d1ZV9jdHguc2V0X2VkaXRvcl9wcm9wZXJ0eSgnc3BlYWtlcicsIGR2KQogICAgICAgIGRpYWxvZ3VlX2N0eC5zZXRfZWRpdG9yX3Byb3BlcnR5KCd0YXJnZXRzJywgW10pCiAgICAgICAgY3R4X21hcHBpbmcuc2V0X2VkaXRvcl9wcm9wZXJ0eSgnY29udGV4dCcsIGRpYWxvZ3VlX2N0eCkKICAgICAgICBkdy5tb2RpZnkoKQogICAgICAgIGR3LnNldF9lZGl0b3JfcHJvcGVydHkoJ2NvbnRleHRfbWFwcGluZ3MnLCBbY3R4X21hcHBpbmddKQoKICAgICAgICAjIFRvdWNoIHNwb2tlbl90ZXh0IHRvIGZpcmUgUG9zdEVkaXRDaGFuZ2VDaGFpblByb3BlcnR5IC0+IFVwZGF0ZU1hcHBpbmdQcm94eQogICAgICAgIGN1cnJlbnRfdGV4dCA9IGR3LmdldF9lZGl0b3JfcHJvcGVydHkoJ3Nwb2tlbl90ZXh0JykKICAgICAgICBkdy5zZXRfZWRpdG9yX3Byb3BlcnR5KCdzcG9rZW5fdGV4dCcsIGN1cnJlbnRfdGV4dCBpZiBjdXJyZW50X3RleHQgZWxzZSAnICcpCgogICAgICAgIHVucmVhbC5FZGl0b3JBc3NldExpYnJhcnkuc2F2ZV9hc3NldChkdy5nZXRfcGF0aF9uYW1lKCkpCiAgICAgICAgcHJpbnQoZicgIFdpcmVkOiB7d2F2ZV9uYW1lfV9EaWFsb2d1ZSB8IHNwZWFrZXI9e2R2X25hbWV9LCB0YXJnZXRzPVtdJykKCiAgICBwcmludCgnU3RlcCAzIGRvbmUuJykKICAgIHRpbWUuc2xlZXAoMSkKCiAgICAjID09PT09PT09PT09PT09PT09PT09PT09PT09PT09PT09PT09PT09PT09PT09PT09PT09PT09PT09PQogICAgIyBTVEVQIDQ6IENyZWF0ZSBTb3VuZEN1ZXMgYW5kIHdpcmUgRGlhbG9ndWVXYXZlIGludG8gdGhlbQogICAgIyA9PT09PT09PT09PT09PT09PT09PT09PT09PT09PT09PT09PT09PT09PT09PT09PT09PT09PT09PT0KICAgIHByaW50KCctLS0gU3RlcCA0OiBDcmVhdGluZyBTb3VuZEN1ZXMgLS0tJykKICAgIGNyZWF0ZWRfY3VlcyA9IFtdCiAgICBmb3Igd2F2ZV9wa2cgaW4gd2F2ZV9wYXRoczoKICAgICAgICB3YXZlX2ZvbGRlciA9IGdldF9mb2xkZXIod2F2ZV9wa2cpCiAgICAgICAgd2F2ZV9uYW1lICAgPSBnZXRfbmFtZSh3YXZlX3BrZykKICAgICAgICBpc192byA9IHdhdmVfbmFtZS5zdGFydHN3aXRoKCdWT18nKQoKICAgICAgICBhci5zY2FuX3BhdGhzX3N5bmNocm9ub3VzKFt3YXZlX2ZvbGRlcl0sIGZvcmNlX3Jlc2Nhbj1UcnVlKQogICAgICAgIHdhdmUgPSB1bnJlYWwubG9hZF9hc3NldCh3YXZlX3BrZykKICAgICAgICBpZiBub3Qgd2F2ZToKICAgICAgICAgICAgcHJpbnQoZicgIFNLSVAgKG5vIHdhdmUpOiB7d2F2ZV9wa2d9JykKICAgICAgICAgICAgY29udGludWUKCiAgICAgICAgc2NfbmFtZSA9IGYnU0Nfe3dhdmVfbmFtZX0nCiAgICAgICAgc2NfcGtnICA9IHdhdmVfZm9sZGVyICsgJy8nICsgc2NfbmFtZQogICAgICAgIGV4aXN0aW5nX3NjID0gZmluZF9hc3NldF9ieV9uYW1lX2FuZF9jbGFzcyh3YXZlX2ZvbGRlciwgc2NfbmFtZSwgJ1NvdW5kQ3VlJykKICAgICAgICBpZiBleGlzdGluZ19zYzoKICAgICAgICAgICAgcHJpbnQoZicgIFNDIGV4aXN0czoge3NjX25hbWV9JykKICAgICAgICAgICAgY3JlYXRlZF9jdWVzLmFwcGVuZChzY19wa2cpCiAgICAgICAgICAgIGNvbnRpbnVlCgogICAgICAgIGZhY3RvcnkgPSB1bnJlYWwuU291bmRDdWVGYWN0b3J5TmV3KCkKICAgICAgICBpZiBpc192byBhbmQgd2F2ZV9wa2cgaW4gZHdfbWFwOgogICAgICAgICAgICBkdyA9IHVucmVhbC5sb2FkX2Fzc2V0KGR3X21hcFt3YXZlX3BrZ10pCiAgICAgICAgICAgIGlmIGR3OgogICAgICAgICAgICAgICAgZmFjdG9yeS5zZXRfZWRpdG9yX3Byb3BlcnR5KCdpbml0aWFsX2RpYWxvZ3VlX3dhdmVzJywgW2R3XSkKICAgICAgICAgICAgZWxzZToKICAgICAgICAgICAgICAgIGZhY3Rvcnkuc2V0X2VkaXRvcl9wcm9wZXJ0eSgnaW5pdGlhbF9zb3VuZF93YXZlcycsIFt3YXZlXSkKICAgICAgICBlbHNlOgogICAgICAgICAgICBmYWN0b3J5LnNldF9lZGl0b3JfcHJvcGVydHkoJ2luaXRpYWxfc291bmRfd2F2ZXMnLCBbd2F2ZV0pCgogICAgICAgIHNjID0gYXQuY3JlYXRlX2Fzc2V0KHNjX25hbWUsIHdhdmVfZm9sZGVyLCB1bnJlYWwuU291bmRDdWUsIGZhY3RvcnkpCiAgICAgICAgaWYgc2M6CiAgICAgICAgICAgIHVucmVhbC5FZGl0b3JBc3NldExpYnJhcnkuc2F2ZV9hc3NldChzYy5nZXRfcGF0aF9uYW1lKCkpCiAgICAgICAgICAgIGFyLnNjYW5fcGF0aHNfc3luY2hyb25vdXMoW3dhdmVfZm9sZGVyXSwgZm9yY2VfcmVzY2FuPVRydWUpCiAgICAgICAgICAgIGZuID0gc2MuZ2V0X2VkaXRvcl9wcm9wZXJ0eSgnZmlyc3Rfbm9kZScpCiAgICAgICAgICAgIHByaW50KGYnICBTQyBjcmVhdGVkOiB7c2NfbmFtZX0gfCBmaXJzdF9ub2RlPXt0eXBlKGZuKS5fX25hbWVfXyBpZiBmbiBlbHNlIE5vbmV9JykKICAgICAgICAgICAgY3JlYXRlZF9jdWVzLmFwcGVuZChzY19wa2cpCiAgICAgICAgZWxzZToKICAgICAgICAgICAgcHJpbnQoZicgIFNDIEZBSUxFRDoge3NjX25hbWV9JykKCiAgICBwcmludChmJ1N0ZXAgNCBkb25lLiB7bGVuKGNyZWF0ZWRfY3Vlcyl9IFNvdW5kQ3VlcyByZWFkeToge2NyZWF0ZWRfY3Vlc30nKQo=")
   n4 = ParseIntoArray(SourceString=Get Object Path String(Object=@n1), Delimiter=.)
   n5 = Append(
    A=Append(A=MakeLiteralString(Value="import base64; SEQ_PATH='"), B="Get (a copy)"(Array=@n4))
    B=MakeLiteralString(Value="'; exec(compile(base64.b64decode('")
   )
   Execute Python Script(execute=@n2.Is Valid, PythonScript=Append(A=Append(A=@n5, B=@n3), B=MakeLiteralString(Value="').decode('utf-8'), '<cc>', 'exec'))")))
   > n6 = For Each Loop(Array=Get GeneratedRowWidgets())
   n7 = Length(TargetArray:Array of Sound Base Object References=Get ReplacementCandidates())
   Set ReplacementCandidateCount(execute=@n6.Completed, ReplacementCandidateCount=@n7) -> (Output_Get=0)
   n8 = MakeLiteralInt!(Value:int=0) -> (ReturnValue:int=0)
   > "SetText (Text)"(
    self=Get Text_ToolbarStatus()
    InText=SelectText(
     A=MakeLiteralText(Value="LOCTEXT(\"SoundCues ready. Auto-matching...\")")
     B=MakeLiteralText(Value="LOCTEXT(\"No audio sections found — load a sequence first.\")")
     bPickA="integer > integer"(A=@n7, B=@n8)
    )
   )
   > AutoMatchReplacements() {is=title K2Node_CallFunction}
   n9 = Get SourceSoundAsset!(self:EUW Audio Sequencer Asset Swap Sound Swap Row Object Reference=@n6) -> (SourceSoundAsset:Sound Base Object Reference)
   Is Valid(exec=@n6.LoopBody, InputObject=@n9)
   n10 = "Get (a copy)"(Array=ParseIntoArray(SourceString=Get Object Path String(Object=@n9), Delimiter=.))
   n11 = ParseIntoArray(SourceString=@n10, Delimiter=/)
   n12 = "Get (a copy)"(Array=@n11, Dimension 1=Last Index!(TargetArray=@n11) -> (ReturnValue=0))
   n13 = Len(S=@n12)
   n14 = MakeLiteralInt(Value=1) -> (ReturnValue=0)
   n15 = Left!(SourceString:String=@n10, Count:int=int - int(A=Len(S=@n10), B=int + int(A=@n13, B=@n14))) -> (ReturnValue:String)
   > n16 = LoadAsset!(
    execute
    self:Editor Asset Subsystem Object Reference=Get EditorAssetSubsystem() -> (ReturnValue:Editor Asset Subsystem Object Reference) {is=title K2Node_GetEditorSubsystem}
    AssetPath=Append(A=Append(A=@n15, B=/), B=Append(A=SC_, B=@n12))
   ) -> (then, ReturnValue)
   > Is Valid(InputObject=@n16)
   > n17 = Cast To SoundCue!(execute, Object=@n16) -> (then, CastFailed, AsSound Cue:Sound Cue Object Reference) {is=title K2Node_DynamicCast}
   > Add Unique(TargetArray:Array of Sound Base Object References=Get ReplacementCandidates(), NewItem:"Sound Base Object Reference (by ref)"=@n17)
  ]}
 DeselectAll {nodes [
   DeselectAll() {is=title K2Node_FunctionEntry}
   > ClearAllSelections() {is=title K2Node_CallFunction}
   > UpdateToolbarButtonStates()
  ]}
 DoAutoSizing {nodes [
   n1 = DoAutoSizing!() -> (then, CurrentAudioSection:Movie Scene Audio Section Object Reference) {is=title K2Node_FunctionEntry}
   > n2 = GetAutoSizeHasEndFrame!(execute, Section:Movie Scene Section Object Reference=@n1) -> (then, ReturnValue:bool)
   > n3 = GetStartFrame(Section=Get CurrentAudioSection() -> (CurrentAudioSection:Movie Scene Audio Section Object Reference)) -> (ReturnValue=0)
   > "Set CurrentStartFrame "("CurrentStartFrame ":int=@n3) -> (Output_Get:int=0)
   > n4 = GetEndFrame(Section=Get CurrentAudioSection()) -> (ReturnValue=0)
   > "Set CurrentEndFrame "("CurrentEndFrame ":int=@n4) -> (Output_Get:int=0)
   n5 = GetLooping!(self:Movie Scene Audio Section Object Reference=@n1) -> (ReturnValue:bool)
   > n6 = Branch(Condition=AND Boolean(A=@n2, B=NOT Boolean(A=@n5)))
   n7 = Get CurrentAudioSection()
   > n8 = GetAutoSizeStartFrame!(execute, Section:Movie Scene Section Object Reference=@n7) -> (then, ReturnValue:int=0)
   > n9 = GetAutoSizeEndFrame!(execute, Section:Movie Scene Section Object Reference=@n7) -> (then, ReturnValue:int=0)
   > SetRange!(execute, Section:Movie Scene Section Object Reference=@n7, StartFrame:int=@n8, EndFrame:int=@n9) -> (then)
   n10 = "Get CurrentEndFrame "() -> ("CurrentEndFrame ":int=0)
   n11 = "Get CurrentStartFrame "() -> ("CurrentStartFrame ":int=0)
   > Return Node!(execute, bWasResized:bool=OR Boolean(A="Not Equal (Integer)"(A=@n11, B=@n8), B="Not Equal (Integer)"(A=@n10, B=@n9)))
   Return Node(execute=@n6.else, bWasResized=false)
  ]}
 EventGraph {nodes [
   "On Check State Changed (CheckBox_ShowUnmatchedOnly)"!() -> (OutputDelegate, then, bIsChecked:bool) {is=title K2Node_ComponentBoundEvent}
   > RefreshRowVisibility()
   "On Clicked (Button_Validate)"() {is=title K2Node_ComponentBoundEvent}
   > ValidateRows()
   "On Clicked (Button_SelectAll)"() {is=title K2Node_ComponentBoundEvent}
   > SelectAllVisible()
   "On Clicked (Button_DeselectAll)"() {is=title K2Node_ComponentBoundEvent}
   > DeselectAll() {is=title K2Node_CallFunction}
   n1 = "On Selection Changed (ComboBox_TrackFilter)"!() -> (OutputDelegate, then, SelectedItem:String, SelectionType:ESelectInfo Enum) {is=title K2Node_ComponentBoundEvent}
   > Set SelectedTrackFilter(SelectedTrackFilter:String=@n1.SelectedItem) -> (Output_Get:String)
   > RefreshRowVisibility()
   > RefreshSummaryText()
   n2 = "On Text Changed (TextBox_Search)"!() -> (OutputDelegate, then, Text:"Text (by ref)") {is=title K2Node_ComponentBoundEvent}
   > Set SearchText(SearchText:String="To String (Text)"(InText=@n2.Text)) -> (Output_Get:String)
   > RefreshRowVisibility()
   "On Clicked (Button_SeachFilterClear)"() {is=title K2Node_ComponentBoundEvent}
   > Set SearchText()
   > "SetText (Text Box)"!(
    execute
    self:Text Box Object Reference=Get TextBox_Search() -> (TextBox_Search:Editor Utility Editable Text Box Object Reference)
    InText:Text
   ) -> (then)
   > RefreshRowVisibility()
   n3 = "On Selection Changed (ComboBox_SortBy)"!() -> (OutputDelegate, then, SelectedItem:String, SelectionType:ESelectInfo Enum) {is=title K2Node_ComponentBoundEvent}
   > Set CurrentSortMode(CurrentSortMode:String=@n3.SelectedItem) -> (Output_Get:String)
   > SortRows()
   "On Clicked (AssetDropWidget_SelectedSequence)"!() -> (OutputDelegate, then, ModifierKeys:EModifierKeys Enum=None) {is=title K2Node_ComponentBoundEvent}
   > OnSequencedClicked()
   n4 = "On Assets Dropped (AssetDropWidget_SelectedSequence)"!() -> (OutputDelegate, then, DroppedAssets:Array of Object References) {is=title K2Node_ComponentBoundEvent}
   > n5 = Cast To LevelSequence!(execute, Object="Get (a copy)"(Array:Array of Object References=@n4.DroppedAssets) -> (Output:Object Reference)) -> (then, CastFailed, AsLevel Sequence:Level Sequence Object Reference) {is=title K2Node_DynamicCast}
   > Set SelectedLevelSequence(SelectedLevelSequence:Level Sequence Object Reference=@n5) -> (Output_Get:Level Sequence Object Reference)
   > RefreshSelectedSequenceDisplay()
   > RefreshSummaryText()
   > GetSequencerSounds()
   > SortRows()
   n6 = "On Clicked (Button_UseSelected)"() {is=title K2Node_ComponentBoundEvent}
   n7 = GetSelectedAssetsOfClass(execute=[@n6.then, @n16.then], AssetClass=/Script/LevelSequence)
   > n8 = For Each Loop(Array:Array of Object References=@n7) -> (Array Element:Object Reference)
   GetSequencerSounds(execute=@n8.Completed)
   > SortRows()
   n9 = Cast To LevelSequence(execute=@n8.LoopBody, Object=@n8)
   > n10 = Set SelectedLevelSequence(SelectedLevelSequence=@n9)
   RefreshSelectedSequenceDisplay(execute=[@n10.then, @n9.CastFailed])
   > RefreshSummaryText()
   "On Clicked (RefreshButton)"() {is=title K2Node_ComponentBoundEvent}
   > ReloadCurrentSequence()
   "On Clicked (Button_CreateCuesAndAutoMatch)"() {is=title K2Node_ComponentBoundEvent}
   > CreateCuesAndAutoMatch() {is=title K2Node_CallFunction}
   "On Clicked (Button_AutoMatch)"() {is=title K2Node_ComponentBoundEvent}
   > PullBrowserSelected()
   "On Clicked (Button_ExecuteSwap)"() {is=title K2Node_ComponentBoundEvent}
   > ExecuteSwap()
   "On Clicked (Button_SetAllSelected)"() {is=title K2Node_ComponentBoundEvent}
   > SetAllSelected()
   "On Clicked (Button_ClearAllSelected)"() {is=title K2Node_ComponentBoundEvent}
   > ClearAllSelected() {is=title K2Node_CallFunction}
   n11 = OnRowSelectionChanged!() -> (OutputDelegate, then, RowIndex:int, bIsSelected:bool=~, bIsCtrlPressed:bool=~) {is=title K2Node_CustomEvent}
   > HandleRowSelectionChanged!(execute, self, "RowIndex ":int=@n11.RowIndex, bIsSelected:bool=@n11.bIsSelected, bIsCtrlPressed:bool=@n11.bIsCtrlPressed) -> (then)
   n12 = OnMultiDropOnRow!() -> (OutputDelegate, then, StartingRowIndex:int, "DroppedSounds ":Array of Sound Base Object References) {is=title K2Node_CustomEvent}
   > n13 = For Each Loop(Array:Array of Sound Base Object References=@n12."DroppedSounds ") -> (Array Element:Sound Base Object Reference)
   SortRows(execute=@n13.Completed)
   n14 = Get GeneratedRowWidgets()
   n15 = int + int(A=@n13.Array Index, B=@n12.StartingRowIndex)
   Branch(
    execute=@n13.LoopBody
    Condition="integer < integer"(A=@n15, B=Length(TargetArray:Array of EUW Audio Sequencer Asset Swap Sound Swap Row Object References=@n14))
   )
   > SetReplacementState(
    self="Get (a copy)"(Array:Array of EUW Audio Sequencer Asset Swap Sound Swap Row Object References=@n14, Dimension 1=@n15) -> (Output:EUW Audio Sequencer Asset Swap Sound Swap Row Object Reference)
    InReplacementDisplayName=GetDisplayName(Object=@n13)
    InStatusText=Manual
    InReplacementSound=@n13
    InMatchScore=0
   )
   OnClicked_CreateCuesAndAutoMatch() {is=title K2Node_CustomEvent}
   > CreateCuesAndAutoMatch()
   n16 = OnSequenceEmptyClick() {is=title K2Node_CustomEvent}
   Event Construct()
   > RefreshSelectedSequenceDisplay()
   > SetKeyboardFocus(self:Widget Object Reference=Self-Reference())
  ]}
 ExecuteSwap {nodes [
   ExecuteSwap() {is=title K2Node_FunctionEntry}
   > Clear(TargetArray:Array of Movie Scene Audio Section Object References=Get ResizedSections())
   > n1 = For Each Loop(Array:Array of EUW Audio Sequencer Asset Swap Sound Swap Row Object References=Get GeneratedRowWidgets()) -> (Array Element:EUW Audio Sequencer Asset Swap Sound Swap Row Object Reference)
   GetSequencerSounds(execute=@n1.Completed)
   > RefreshSummaryText()
   > "SetText (Text)"(
    self=Get Text_ToolbarStatus()
    InText=Format Text!(
     Format="LOCTEXT(\"{Success} updated, {Skipped} skipped.\")"
     Success:int=Get SwapSuccessCount() -> (SwapSuccessCount:int=0)
     Skipped:int=Get SwapSkippedCount() -> (SwapSkippedCount:int=0)
    ) -> (Result)
   )
   > n2 = For Each Loop(Array:Array of Movie Scene Audio Section Object References=Get ResizedSections()) -> (Array Element:Movie Scene Audio Section Object Reference)
   > n3 = CheckForOverlap!(
    execute
    self
    "ResizedSection ":Movie Scene Audio Section Object Reference=@n2
    ResizedSections:Array of Movie Scene Audio Section Object References=Get ResizedSections()
   ) -> (then, bHasOverlap:bool, OverlapFrames:int=0, AsMovie Scene Audio Section:Movie Scene Audio Section Object Reference) {is=title K2Node_CallFunction}
   > Branch(Condition=@n3)
   > Show Message Dialog(
    Title=MakeLiteralText(Value="LOCTEXT(\"Overlapping Audio Tracks in Sequence\")")
    Message=Format Text!(
     Format="LOCTEXT(\"{FirstSection} is overlapping on audio track with {SecondSection}.\")"
     FirstSection:Sound Base Object Reference=Get Sound!(self:Movie Scene Audio Section Object Reference=@n2) -> (Sound:Sound Base Object Reference)
     SecondSection:Sound Base Object Reference=Get Sound(self=@n3.AsMovie Scene Audio Section)
    ) -> (Result)
    MessageType=Ok
    DefaultValue=No
    MessageCategory=Warning
   )
   n4 = Is Valid(exec=@n1.LoopBody, InputObject=@n1)
   n5 = Get SourceAudioSection!(self:EUW Audio Sequencer Asset Swap Sound Swap Row Object Reference=@n1) -> (SourceAudioSection:Movie Scene Audio Section Object Reference)
   > n6 = Is Valid(InputObject=@n5)
   n7 = Get ReplacementSound!(self:EUW Audio Sequencer Asset Swap Sound Swap Row Object Reference=@n1) -> (ReplacementSound:Sound Base Object Reference)
   > n8 = Is Valid(InputObject=@n7)
   > n9 = Branch(Condition=MatchedOrManual(self=@n1))
   > SetSound!(execute, self:Movie Scene Audio Section Object Reference=@n5, InSound:Sound Base Object Reference=@n7) -> (then)
   > n10 = DoAutoSizing!(execute, self, CurrentAudioSection:Movie Scene Audio Section Object Reference=@n5) -> (then, bWasResized:bool) {is=title K2Node_CallFunction}
   > n11 = Branch(Condition=@n10)
   > n12 = Add(
    TargetArray:Array of Movie Scene Audio Section Object References=Get ResizedSections()
    NewItem:"Movie Scene Audio Section Object Reference (by ref)"=@n5
   ) {is=title K2Node_CallArrayFunction}
   n13 = Increment Int(" "=[@n12.then, @n11.else], Value=Get SwapSuccessCount() -> (SwapSuccessCount=0))
   > Set SwapSuccessCount(SwapSuccessCount:int=@n13) -> (Output_Get:int=0)
   > n14 = Get!(execute:Exec) -> (then:Exec, else:Exec, SelectedLevelSequence:Level Sequence Object Reference)
   > TransactObject(Object=@n14)
   n15 = Increment Int(" "=[@n9.else, @n8.Is Not Valid, @n6.Is Not Valid, @n4.Is Not Valid], Value=Get SwapSkippedCount() -> (SwapSkippedCount=0))
   > Set SwapSkippedCount(SwapSkippedCount:int=@n15) -> (Output_Get:int=0)
  ]}
 ForEachAudioSection {nodes [
   n1 = Inputs!() -> (Exec:Exec, MovieSceneAudioTracks:Array of Movie Scene Audio Track Object References)
   > n2 = For Each Loop(Array:Array of Movie Scene Audio Track Object References=@n1) -> (Array Element:Movie Scene Audio Track Object Reference)
   > n3 = GetSections(Track=@n2)
   > n4 = For Each Loop(Array:Array of Movie Scene Section Object References=@n3) -> (Array Element:Movie Scene Section Object Reference)
   > n5 = Cast To MovieSceneAudioSection(Object=@n4)
   > Outputs!(
    LoopBody:Exec
    Movie Scene Audio Section:Movie Scene Audio Section Object Reference=@n5
    CurrentAudioTrack:Movie Scene Audio Track Object Reference=@n2
    Completed:Exec=@n2.Completed
   )
  ]}
 ForEachAudioTracksInSelectedSequences {nodes [
   Inputs!() -> (Exec)
   > n1 = Get()
   > n2 = FindTracksByType!(
    execute
    Sequence:Movie Scene Sequence Object Reference=@n1
    TrackType:Movie Scene Track Class Reference=/Script/MovieSceneTracks.MovieSceneAudioTrack
   ) -> (then, ReturnValue:Array of Movie Scene Audio Track Object References)
   > Outputs!(Out:Exec, AudioTracks:Array of Movie Scene Audio Track Object References=@n2)
  ]}
 GetAudioMatchScore {nodes [
   n1 = GetAudioMatchScore!() -> (then, SourceName:String, CandidateName:String) {is=title K2Node_FunctionEntry}
   > n2 = NormalizeAudioName!(execute, self, InName:String=@n1) -> (then, NormalizedName:String)
   > Set NormalizedSource(NormalizedSource:String=@n2) -> (Output_Get:String)
   > n3 = NormalizeAudioName(InName=@n1.CandidateName)
   > Set NormalizedCandidate(NormalizedCandidate:String=@n3) -> (Output_Get:String)
   > n4 = Branch(Condition="Equal Exactly (String)"(A=Get NormalizedSource() -> (NormalizedSource:String), B=Get NormalizedCandidate() -> (NormalizedCandidate:String)))
   > Return Node!(execute, Score:double=1)
   Set SourceTokens(execute=@n4.else, SourceTokens:Array of Strings=ParseIntoArray(SourceString=Get NormalizedSource(), Delimiter=" _")) -> (Output_Get:Array of Strings)
   > Set CandidateTokens(CandidateTokens:Array of Strings=ParseIntoArray(SourceString=Get NormalizedCandidate(), Delimiter=" _")) -> (Output_Get:Array of Strings)
   > Set TotalWeight(TotalWeight:double=0) -> (Output_Get:double)
   > Set MatchedWeight(MatchedWeight:double=0) -> (Output_Get:double)
   > Set SourceNumericCount(SourceNumericCount:int=0) -> (Output_Get:int=0)
   > Set MatchedNumericCount(MatchedNumericCount:int=0) -> (Output_Get:int=0)
   > n5 = For Each Loop(Array:Array of Strings=Get SourceTokens() -> (SourceTokens:Array of Strings)) -> (Array Element:String)
   n6 = Branch(execute=@n5.Completed, Condition="float <= float"(A=Get TotalWeight() -> (TotalWeight:double)) -> (ReturnValue=~))
   > Return Node(Score=0)
   Set BaseScore(execute=@n6.else, BaseScore:double=float / float(A=Get MatchedWeight() -> (MatchedWeight:double), B=Get TotalWeight())) -> (Output_Get:double)
   n7 = Get BaseScore() -> (BaseScore:double)
   n8 = Get SourceNumericCount() -> (SourceNumericCount:int=0)
   n9 = Get MatchedNumericCount() -> (MatchedNumericCount:int=0)
   > n10 = Branch(Condition=AND Boolean!(A="integer > integer"(A=@n8), B="Equal (Integer)"(A=@n9), C:bool="float > float"(A=@n7, B=0.59)) -> (ReturnValue))
   > Return Node(Score=0.59)
   Return Node(execute=@n10.else, Score=Get BaseScore())
   n11 = Sequence(execute=@n5.LoopBody)
   n12 = GetTokenWeight!(InToken:String=@n5) -> (Weight:double, bIsNumeric:bool)
   > Set CurrentWeight(CurrentWeight:double=@n12) -> (Output_Get:double)
   > Set bCurrentTokenIsNumeric(bCurrentTokenIsNumeric:bool=@n12.bIsNumeric) -> (Output_Get:bool)
   > Set TotalWeight(TotalWeight=float + float(A=Get TotalWeight(), B=Get CurrentWeight() -> (CurrentWeight:double)))
   > Branch(Condition=Get bCurrentTokenIsNumeric() -> (bCurrentTokenIsNumeric:bool))
   > n13 = Increment Int(Value=Get SourceNumericCount() -> (SourceNumericCount=0))
   > Set SourceNumericCount(SourceNumericCount=@n13) -> (Output_Get=0)
   Branch(
    execute=@n11.then_1
    Condition=Contains Item(TargetArray:Array of Strings=Get CandidateTokens() -> (CandidateTokens:Array of Strings), ItemToFind:"String (by ref)"=@n5)
   )
   > Set MatchedWeight(MatchedWeight=float + float(A=Get MatchedWeight(), B=Get CurrentWeight()))
   > Branch(Condition=Get bCurrentTokenIsNumeric())
   > n14 = Increment Int(Value=Get MatchedNumericCount() -> (MatchedNumericCount=0))
   > Set MatchedNumericCount(MatchedNumericCount=@n14) -> (Output_Get=0)
  ]}
 GetSequencerSounds {nodes [
   GetSequencerSounds() {is=title K2Node_FunctionEntry}
   > n1 = Get()
   > Set In Row Index(In Row Index:int=0) -> (Output_Get:int=0)
   > ClearChildren(self=Get ScrollBox_SoundRows())
   > Clear(TargetArray:Array of EUW Audio Sequencer Asset Swap Sound Swap Row Object References=Get GeneratedRowWidgets())
   > Clear(TargetArray:Array of Strings=Get UniqueTrackNames() -> (UniqueTrackNames:Array of Strings))
   > n2 = For Each Audio Tracks in Selected Sequences!(Exec:Exec) -> (Out:Exec, AudioTracks:Array of Movie Scene Audio Track Object References) {is=title K2Node_MacroInstance}
   > n3 = For Each Audio Section!(Exec:Exec, MovieSceneAudioTracks:Array of Movie Scene Audio Track Object References=@n2) -> (
    LoopBody:Exec
    Movie Scene Audio Section:Movie Scene Audio Section Object Reference
    CurrentAudioTrack:Movie Scene Audio Track Object Reference
    Completed:Exec
   ) {is=title K2Node_MacroInstance}
   RefreshSummaryText(execute=[@n7.then, @n3.Completed])
   > RefreshRowVisibility()
   > PopulateTrackFilter()
   > SortRows()
   > Return Node!(execute)
   n4 = GetDisplayName!(execute:Exec=@n3.LoopBody, Track:Movie Scene Track Object Reference=@n3.CurrentAudioTrack) -> (then:Exec, ReturnValue:Text)
   n5 = GetRowIndex(self=@n3) -> (ReturnValue=0)
   > Add Unique(
    TargetArray:Array of Strings=Get UniqueTrackNames()
    NewItem:"String (by ref)"="To String (Text)"(InText=Format Text!(Format="LOCTEXT(\"{Track} - Row {Row}\")", Track:Text=@n4, Row:int=@n5) -> (Result))
   )
   > AddSoundRow!(
    execute
    self
    InRowIndex:int=Get In Row Index() -> (In Row Index:int=0)
    InSourceAudioSection:Movie Scene Audio Section Object Reference=@n3
   ) -> (then) {is=title K2Node_CallFunction}
   > n6 = Increment Int(Value=Get In Row Index() -> (In Row Index=0))
   > Set In Row Index(In Row Index=@n6) -> (Output_Get=0)
   ClearChildren(execute=@n1.else, self=Get ScrollBox_SoundRows())
   > n7 = Clear(TargetArray:Array of EUW Audio Sequencer Asset Swap Sound Swap Row Object References=Get GeneratedRowWidgets())
  ]}
 GetSortKeyFromWidget {nodes [
   n1 = GetSortKeyFromWidget!() -> (then, RowWidget:EUW Audio Sequencer Asset Swap Sound Swap Row Object Reference, SortMode:String) {is=title K2Node_FunctionEntry}
   > n2 = Switch on String!(execute:Exec, Selection:String=@n1.SortMode) -> (Default:Exec, Row Index:Exec, Track Name:Exec, Source Sound Name:Exec, "Status (Matched First)":Exec, "Status (Unmatched First)":Exec) {pinNames [Row Index, Track Name, Source Sound Name, "Status (Matched First)", "Status (Unmatched First)"]}
   > Return Node!(execute, SortKey:String=MakeLiteralString())
   n3 = Get RowIndex!(self:EUW Audio Sequencer Asset Swap Sound Swap Row Object Reference=@n1) -> (RowIndex:int=0)
   Return Node(execute=@n2.Row Index, SortKey="To String (Text)"(InText="To Text (Integer)"(Value=@n3, bUseGrouping=false, MinimumIntegralDigits=4)))
   Return Node(execute=@n2.Source Sound Name, SortKey=Get SourceDisplayName(self=@n1))
   n4 = MatchedOrManual(self=@n1)
   n5 = MakeLiteralString(Value=1_UnMatched)
   n6 = MakeLiteralString(Value=0_Matched)
   Return Node(execute=@n2."Status (Matched First)", SortKey=SelectString(A=@n6, B=@n5, bPickA=@n4))
   Return Node(execute=@n2."Status (Unmatched First)", SortKey=SelectString(A=@n5, B=@n6, bPickA=@n4))
   Return Node(
    execute=@n2.Track Name
    SortKey=Get AudioTrackName!(self:EUW Audio Sequencer Asset Swap Sound Swap Row Object Reference=@n1) -> (AudioTrackName:String)
   )
  ]}
 GetTokenWeight {nodes [
   n1 = GetTokenWeight!() -> (then, InToken:String) {is=title K2Node_FunctionEntry}
   n2 = ToLower!(SourceString:String=@n1) -> (ReturnValue:String)
   > n3 = Set NormalizedToken(NormalizedToken:String=TrimTrailing!(SourceString:String=Trim!(SourceString:String=@n2) -> (ReturnValue:String)) -> (ReturnValue:String)) -> (Output_Get:String)
   > n4 = Branch(Condition=OR Boolean!(
     A="Equal Exactly (String)"(A=@n3, B=vo)
     B="Equal Exactly (String)"(A=@n3, B=ip)
     C:bool="Equal Exactly (String)"(A=@n3, B=temp)
     D:bool="Equal Exactly (String)"(A=@n3, B=cue)
     E:bool="Equal Exactly (String)"(A=@n3, B=dialogue)
     F:bool="Equal Exactly (String)"(A=@n3, B=gameplay)
     G:bool="Equal Exactly (String)"(A=@n3, B=mss)
     H:bool="Equal Exactly (String)"(A=@n3, B=sc)
    ) -> (ReturnValue))
   > Return Node!(execute, Weight:double=0.25, bIsNumeric:bool=false)
   n5 = IsNumericToken!(execute=@n4.else, self, InToken:String=Get NormalizedToken() -> (NormalizedToken:String)) -> (then, bIsNumeric:bool)
   > n6 = Branch(Condition=@n5)
   > Return Node(Weight=1.5, bIsNumeric=true)
   Return Node(execute=@n6.else, Weight=1, bIsNumeric=false)
  ]}
 HandleRowSelectionChanged {nodes [
   n1 = HandleRowSelectionChanged!() -> (then, "RowIndex ":int, bIsSelected:bool=~, bIsCtrlPressed:bool=~) {is=title K2Node_FunctionEntry}
   > n2 = Branch(Condition=@n1.bIsCtrlPressed)
   > n3 = Branch(Condition=Get bIsSelected() -> (bIsSelected:bool))
   n4 = Add(
    execute=[@n3.then, @n6.then]
    TargetArray:Array of EUW Audio Sequencer Asset Swap Sound Swap Row Object References=Get SelectedRows()
    NewItem:"EUW Audio Sequencer Asset Swap Sound Swap Row Object Reference (by ref)"="Get (a copy)"(Array=Get GeneratedRowWidgets(), Dimension 1="Get RowIndex "() -> ("RowIndex ":int=0))
   )
   UpdateToolbarButtonStates(execute=[@n4.then, @n5.then])
   n5 = Remove Item!(
    execute=@n3.else
    TargetArray:Array of EUW Audio Sequencer Asset Swap Sound Swap Row Object References=Get SelectedRows()
    Item:"EUW Audio Sequencer Asset Swap Sound Swap Row Object Reference (by ref)"="Get (a copy)"(Array=Get GeneratedRowWidgets(), Dimension 1="Get RowIndex "() -> ("RowIndex "=0))
   ) -> (then, ReturnValue:bool)
   ClearAllSelections(execute=@n2.else)
   > n6 = Branch(Condition=Get bIsSelected())
   UpdateToolbarButtonStates(execute=@n6.else)
  ]}
 IsNumericToken {nodes [
   n1 = IsNumericToken!() -> (then, InToken:String) {is=title K2Node_FunctionEntry}
   > n2 = Branch(Condition="integer <= integer"(A=Len(S=@n1)))
   > Return Node!(execute, bIsNumeric=false)
   Return Node(execute=@n2.else, bIsNumeric=IsNumeric!(SourceString:String=@n1) -> (ReturnValue:bool))
  ]}
 IsSearchBoxFiltered {nodes [
   IsSearchBoxFiltered!() -> (then, CurrentRow:EUW Audio Sequencer Asset Swap Sound Swap Row Object Reference) {is=title K2Node_FunctionEntry}
   n1 = ToLower(SourceString=Get ReplacementDisplayName!(self:EUW Audio Sequencer Asset Swap Sound Swap Row Object Reference=Get CurrentRow() -> (CurrentRow:EUW Audio Sequencer Asset Swap Sound Swap Row Object Reference)) -> (ReplacementDisplayName:String))
   n2 = Get SearchText() -> (SearchText:String)
   n3 = ToLower(SourceString=@n2)
   n4 = ToLower(SourceString=Get SourceDisplayName(self=Get CurrentRow()))
   n5 = Len(S=@n2)
   > Return Node!(
    execute
    ReturnValue:bool=OR Boolean!(A="Equal (Integer)"(A=@n5), B=Contains(SearchIn=@n4, Substring=@n3), C=Contains(SearchIn=@n1, Substring=@n3)) -> (ReturnValue)
   )
  ]}
 IsTrackBoxFiltered {nodes [
   n1 = IsTrackBoxFiltered!() -> (then, Current row:EUW Audio Sequencer Asset Swap Sound Swap Row Object Reference) {is=title K2Node_FunctionEntry}
   n2 = Get AudioTrackName(self=@n1)
   n3 = Get SelectedTrackFilter() -> (SelectedTrackFilter:String)
   n4 = Get SelectedTrackFilter()
   > Return Node(ReturnValue=OR Boolean!(A="Equal Exactly (String)"(A=@n4, B=All Tracks), B="Equal Exactly (String)"(A=@n2, B=@n3)) -> (ReturnValue))
  ]}
 NormalizeAudioName {nodes [
   n1 = NormalizeAudioName!() -> (then, InName:String) {is=title K2Node_FunctionEntry}
   n2 = ParseIntoArray(SourceString=ToLower(SourceString=@n1), Delimiter=/)
   n3 = "Get (a ref)"(Array:Array of Strings=@n2, Dimension 1=Last Index(TargetArray=@n2) -> (ReturnValue=0)) -> (Output:"String (by ref)")
   > PrintString(InString=@n3)
   n4 = ParseIntoArray(SourceString=@n3, Delimiter=.)
   n5 = "Get (a ref)"(Array=@n4, Dimension 1=Last Index(TargetArray=@n4) -> (ReturnValue=0))
   > PrintString(InString=@n5)
   n6 = Replace(SourceString=Replace(SourceString=Replace(SourceString=@n5, From=_cue), From=_dialogue), From=_gameplay)
   n7 = Replace(SourceString=Replace(SourceString=Replace(SourceString=@n6, From=sc_), From=mss_), From=__, To=_)
   > n8 = Branch(Condition=StartsWith(SourceString=@n7, InPrefix=_))
   > n9 = Set WorkingName(WorkingName:String=LeftChop(SourceString=@n7, Count=1)) -> (Output_Get:String)
   n10 = Branch(execute=[@n9.then, @n8.else], Condition=EndsWith(SourceString=@n7, InSuffix=_))
   > n11 = Set WorkingName(WorkingName=RightChop!(SourceString:String=@n7, Count:int=1) -> (ReturnValue:String))
   n12 = Set WorkingName(execute=[@n11.then, @n10.else], WorkingName=@n7)
   > Return Node!(execute, NormalizedName:String=@n12)
  ]}
 OnKeyDown {nodes [
   n1 = OnKeyDown!() -> (then, MyGeometry:Geometry Structure, InKeyEvent:Key Event Structure) {is=title K2Node_FunctionEntry}
   > Set Pressed Key(Pressed Key:Key Structure=GetKey!(Input:"Key Event Structure (by ref)"=@n1.InKeyEvent) -> (ReturnValue:Key Structure)) -> (Output_Get:Key Structure)
   n2 = GetModifierKeysState!() -> (ReturnValue:Modifier Keys State Structure)
   n3 = Get Pressed Key() -> (Pressed Key:Key Structure)
   > n4 = Branch(Condition=AND Boolean!(
     A="Equal (Key)"(A:Key Structure=@n3, B:Key Structure=A) -> (ReturnValue:bool=~) {is=title K2Node_PromotableOperator}
     B=Is Control Down!(KeysState:"Modifier Keys State Structure (by ref)"=@n2) -> (ReturnValue:bool)
    ) -> (ReturnValue))
   > SelectAllVisible()
   > Return Node!(execute, ReturnValue:Event Reply Structure=Handled())
   n5 = Is Control Down(KeysState=GetModifierKeysState())
   n6 = "Equal (Key)"(A=Get Pressed Key(), B=D) -> (ReturnValue=~)
   n7 = Get Pressed Key()
   n8 = Branch(execute=@n4.else, Condition=OR Boolean(A=AND Boolean(A=@n6, B=@n5), B="Equal (Key)"(A=@n7, B=Escape) -> (ReturnValue=~)))
   > DeselectAll()
   > Return Node(ReturnValue=Handled())
   n9 = Branch(execute=@n8.else, Condition="Equal (Key)"(A=Get Pressed Key(), B=Delete) -> (ReturnValue=~))
   > ClearAllSelected()
   > Return Node(ReturnValue=Handled())
   Return Node(execute=@n9.else, ReturnValue=Handled())
  ]}
 OnMouseButtonDown {nodes [
   OnMouseButtonDown!() -> (then, MyGeometry:Geometry Structure, MouseEvent:"Pointer Event Structure (by ref)") {is=title K2Node_FunctionEntry}
   > SetKeyboardFocus(self=Self-Reference())
   > Return Node(ReturnValue=Unhandled())
  ]}
 OnSequencedClicked {nodes [
   OnSequencedClicked() {is=title K2Node_FunctionEntry}
   > n1 = Get()
   n2 = To Soft Object Reference(SoftObjectPath=GetSoftObjectPath(Object=@n1))
   > SyncBrowserToObjects(AssetPaths=Make Array("[0]":String="To String (SoftObjectReference)"!(SoftObjectReference:"Soft Object Reference (by ref)"=@n2) -> (ReturnValue:String)) -> (Array:Array of Strings))
   OnSequenceEmptyClick(execute=@n1.else) {is=title K2Node_CallFunction}
  ]}
 PopulateTrackFilter {nodes [
   PopulateTrackFilter() {is=title K2Node_FunctionEntry}
   > Get()
   n1 = Get ComboBox_TrackFilter() -> (ComboBox_TrackFilter:Editor Utility Combo Box String Object Reference)
   > ClearOptions(self:"Combo Box (String) Object Reference"=@n1)
   > AddOption!(execute, self:"Combo Box (String) Object Reference"=@n1, Option:String=All Tracks) -> (then)
   > n2 = For Each Loop(Array=Get UniqueTrackNames())
   SetSelectedOption!(execute=@n2.Completed, self:"Combo Box (String) Object Reference"=Get ComboBox_TrackFilter(), Option:String=All Tracks) -> (then)
   AddOption(execute=@n2.LoopBody, self=Get ComboBox_TrackFilter(), Option=@n2)
  ]}
 PullBrowserSelected {nodes [
   PullBrowserSelected() {is=title K2Node_FunctionEntry}
   > Clear(TargetArray:Array of Sound Base Object References=Get ReplacementCandidates())
   > n1 = GetSelectedAssetsOfClass(AssetClass=/Script/Engine.SoundBase)
   > n2 = For Each Loop(Array:Array of Object References=@n1) -> (Array Element:Object Reference)
   n3 = Set ReplacementCandidateCount(execute=@n2.Completed, ReplacementCandidateCount=Length(TargetArray:Array of Sound Base Object References=Get ReplacementCandidates())) -> (Output_Get=0)
   > "SetText (Text)"(
    self=Get Text_ToolbarStatus()
    InText=SelectText(
     A=Format Text!(Format="LOCTEXT(\"{Count} replacement candidates loaded from the Content Browser.\")", Count:int=@n3) -> (Result)
     B="LOCTEXT(\"No SoundBase assets selected in the Content Browser.\")"
     bPickA="integer > integer"(A=@n3)
    )
   )
   > AutoMatchReplacements()
   n4 = Cast To SoundBase!(execute=@n2.LoopBody, Object=@n2) -> (then, CastFailed, AsSound Base:Sound Base Object Reference) {is=title K2Node_DynamicCast}
   > Add Unique(TargetArray:Array of Sound Base Object References=Get ReplacementCandidates(), NewItem:"Sound Base Object Reference (by ref)"=@n4)
  ]}
 RefreshRowVisibility {nodes [
   RefreshRowVisibility() {is=title K2Node_FunctionEntry}
   > n1 = Branch(Condition=NOT Boolean(A=Get IsRefreshing() -> (IsRefreshing:bool)))
   > Set IsRefreshing(IsRefreshing:bool=true) -> (Output_Get:bool)
   > n2 = For Each Loop(Array:Array of EUW Audio Sequencer Asset Swap Sound Swap Row Object References=Get GeneratedRowWidgets()) -> (Array Element:EUW Audio Sequencer Asset Swap Sound Swap Row Object Reference)
   RefreshSummaryText(execute=@n2.Completed)
   > Set IsRefreshing(IsRefreshing=false)
   > Return Node!(execute)
   Set CurrentRow(execute=@n2.LoopBody, CurrentRow:EUW Audio Sequencer Asset Swap Sound Swap Row Object Reference=@n2) -> (Output_Get:EUW Audio Sequencer Asset Swap Sound Swap Row Object Reference)
   n3 = GetCheckedState!(self:Check Box Object Reference=Get CheckBox_ShowUnmatchedOnly() -> (CheckBox_ShowUnmatchedOnly:Editor Utility Check Box Object Reference)) -> (ReturnValue:ECheckBoxState Enum=Unchecked)
   > n4 = Branch(Condition="Equal (Enum)"(A:ECheckBoxState Enum=@n3, B:ECheckBoxState Enum=Unchecked))
   n5 = Branch(
    execute=[@n8.else, @n4.then, @n7.else]
    Condition=IsTrackBoxFiltered!(Current row:EUW Audio Sequencer Asset Swap Sound Swap Row Object Reference=Get CurrentRow()) -> (ReturnValue:bool) {is=title K2Node_CallFunction}
   )
   > n6 = Branch(Condition=IsSearchBoxFiltered!(CurrentRow:EUW Audio Sequencer Asset Swap Sound Swap Row Object Reference=Get CurrentRow()) -> (ReturnValue:bool) {is=title K2Node_CallFunction})
   > SetVisibility(self=Get CurrentRow(), InVisibility=Visible)
   SetVisibility(execute=[@n6.else, @n8.then, @n5.else], self=Get CurrentRow(), InVisibility=Collapsed)
   n7 = Branch(execute=@n4.else, Condition="Equal (Enum)"(A=@n3, B=Checked))
   > n8 = Branch(Condition=MatchedOrManual(self=Get CurrentRow()))
   Set IsRefreshing(execute=@n1.else, IsRefreshing=false)
   > Return Node()
  ]}
 RefreshSelectedSequenceDisplay {nodes [
   RefreshSelectedSequenceDisplay() {is=title K2Node_FunctionEntry}
   n1 = Get SelectedLevelSequence()
   n2 = GetDisplayName!(Object:Object Reference=@n1) -> (ReturnValue:String)
   n3 = IsValid(Object=@n1)
   > SetText(
    self:Audio Asset Drop Target Widget Object Reference=Get AssetDropWidget_SelectedSequence() -> (AssetDropWidget_SelectedSequence:Audio Asset Drop Target Widget Object Reference)
    InText=SelectText(A="To Text (String)"(InString=@n2), B="LOCTEXT(\"No Level Sequence Selected\")", bPickA=@n3)
   )
   n4 = Get Text_SequenceStatus() -> (Text_SequenceStatus:Text Block Object Reference)
   > "SetText (Text)"(
    self=@n4
    InText=SelectText(A="LOCTEXT(\"Level Sequence Ready\")", B="LOCTEXT(\"Select a Level Sequence in the Content Browser, then click Use Selected\")", bPickA=@n3)
   )
   > SetColorAndOpacity(
    self=@n4
    InColorAndOpacity=Select(
     Option 0:Slate Color Structure=Make Slate Color(SpecifiedColor=(R=0.6795,G=0.3968,B=0.0482,A=1))
     Option 1:Slate Color Structure=Make Slate Color(SpecifiedColor=(R=0.0844,G=0.5395,B=0.1746,A=1))
     Index=@n3
    ) -> (ReturnValue:Slate Color Structure)
   )
  ]}
 RefreshSummaryText {nodes [
   RefreshSummaryText() {is=title K2Node_FunctionEntry}
   > Set TotalRows(TotalRows:int=Length(TargetArray:Array of EUW Audio Sequencer Asset Swap Sound Swap Row Object References=Get GeneratedRowWidgets())) -> (Output_Get:int=0)
   > Set CandidateCount(CandidateCount:int=Length(TargetArray:Array of Sound Base Object References=Get ReplacementCandidates())) -> (Output_Get:int=0)
   > n1 = For Each Loop(Array=Get GeneratedRowWidgets())
   "SetText (Text)"(
    execute=@n1.Completed
    self=Get Text_Summary() -> (Text_Summary:Text Block Object Reference)
    InText=Format Text!(
     Format="LOCTEXT(\"{Rows} source rows | {Candidates} replacement candidates | {Matched} matched | {Manual} manual | {Review} needs review | {Missing} missing\")"
     Rows:int=Get TotalRows() -> (TotalRows:int=0)
     Candidates:int=Get CandidateCount() -> (CandidateCount:int=0)
     Matched:int=Get MatchedCount() -> (MatchedCount:int=0)
     Manual:int=Get ManualCount() -> (ManualCount:int=0)
     Review:int=Get NeedsReviewCount() -> (NeedsReviewCount:int=0)
     Missing:int=Get MissingCount() -> (MissingCount:int=0)
    ) -> (Result)
   )
   n2 = Get StatusText!(self:EUW Audio Sequencer Asset Swap Sound Swap Row Object Reference=@n1) -> (StatusText:String)
   n3 = Branch(execute=@n1.LoopBody, Condition="Equal Exactly (String)"(A=@n2, B=Get Matched(self=@n1)))
   > n4 = Increment Int(Value=Get MatchedCount() -> (MatchedCount=0))
   > Set MatchedCount(MatchedCount:int=@n4) -> (Output_Get:int=0)
   n5 = Branch(execute=@n3.else, Condition="Equal Exactly (String)"(A=@n2, B=Get NeedsReview(self=@n1)))
   > n6 = Increment Int(Value=Get NeedsReviewCount() -> (NeedsReviewCount=0))
   > Set NeedsReviewCount(NeedsReviewCount:int=@n6) -> (Output_Get:int=0)
   n7 = Branch(
    execute=@n5.else
    Condition="Equal Exactly (String)"(A=@n2, B=Get Manual!(self:EUW Audio Sequencer Asset Swap Sound Swap Row Object Reference=@n1) -> (Manual:String))
   )
   > n8 = Increment Int(Value=Get ManualCount() -> (ManualCount=0))
   > Set ManualCount(ManualCount:int=@n8) -> (Output_Get:int=0)
   n9 = Increment Int(" "=@n7.else, Value=Get MissingCount() -> (MissingCount=0))
   > Set MissingCount(MissingCount:int=@n9) -> (Output_Get:int=0)
  ]}
 ReloadCurrentSequence {nodes [
   ReloadCurrentSequence() {is=title K2Node_FunctionEntry}
   > GetSequencerSounds() {is=title K2Node_CallFunction}
   > RefreshRowVisibility() {is=title K2Node_CallFunction}
   > RefreshSummaryText() {is=title K2Node_CallFunction}
   > RefreshSelectedSequenceDisplay() {is=title K2Node_CallFunction}
  ]}
 SelectAllVisible {nodes [
   SelectAllVisible() {is=title K2Node_FunctionEntry}
   > Clear(TargetArray:Array of EUW Audio Sequencer Asset Swap Sound Swap Row Object References=Get SelectedRows())
   > n1 = For Each Loop(Array=Get GeneratedRowWidgets())
   UpdateToolbarButtonStates(execute=@n1.Completed)
   Branch(
    execute=@n1.LoopBody
    Condition="Equal (Enum)"(
     A:ESlateVisibility Enum=GetVisibility!(self:Widget Object Reference=@n1) -> (ReturnValue:ESlateVisibility Enum=Visible)
     B:ESlateVisibility Enum=Visible
    )
   )
   > SetSelected(self=Get AudioAssetDropTargetWidget_Row(self=@n1), bSelected=true)
   > Add(TargetArray=Get SelectedRows(), NewItem=@n1)
  ]}
 SetAllSelected {nodes [
   SetAllSelected() {is=title K2Node_FunctionEntry}
   > n1 = For Each Loop(Array=Get SelectedRows())
   > AssignSelectedReplacement(self:EUW Audio Sequencer Asset Swap Sound Swap Row Object Reference=@n1)
  ]}
 SortRows {nodes [
   SortRows() {is=title K2Node_FunctionEntry}
   > Set ArrayLength(ArrayLength:int=Length(TargetArray:Array of EUW Audio Sequencer Asset Swap Sound Swap Row Object References=Get GeneratedRowWidgets())) -> (Output_Get:int=0)
   n1 = int - int(A=Get ArrayLength() -> (ArrayLength:int=0))
   > n2 = For Loop(LastIndex=@n1)
   ClearChildren(execute=@n2.Completed, self=Get ScrollBox_SoundRows())
   > n3 = For Each Loop(Array=Get GeneratedRowWidgets())
   n4 = For Each Loop(Exec=@n3.Completed, Array=Get GeneratedRowWidgets())
   Set Timer by Event(execute=@n4.Completed, Delegate=Create Event(), Time=0.2, bLooping=false)
   > Return Node()
   Set RowIndex!(execute=@n4.LoopBody, RowIndex:int=@n4.Array Index, self:EUW Audio Sequencer Asset Swap Sound Swap Row Object Reference=@n4) -> (then, Output_Get:int=0)
   AddChild(execute=@n3.LoopBody, self=Get ScrollBox_SoundRows(), Content=@n3)
   n5 = For Loop(execute=@n2.LoopBody, FirstIndex=int + int(A=@n2), LastIndex=@n1)
   n6 = Get GeneratedRowWidgets()
   n7 = "Get (a copy)"(Array=@n6, Dimension 1=@n2)
   > n8 = "Set  WidgetA"(" WidgetA":EUW Audio Sequencer Asset Swap Sound Swap Row Object Reference=@n7) -> (Output_Get:EUW Audio Sequencer Asset Swap Sound Swap Row Object Reference)
   > n9 = "Set  WidgetB"(" WidgetB":EUW Audio Sequencer Asset Swap Sound Swap Row Object Reference="Get (a copy)"(Array=@n6, Dimension 1=@n5)) -> (Output_Get:EUW Audio Sequencer Asset Swap Sound Swap Row Object Reference)
   n10 = Get CurrentSortMode() -> (CurrentSortMode:String)
   > n11 = Set SortKeyA(SortKeyA:String=GetSortKeyFromWidget!(RowWidget:EUW Audio Sequencer Asset Swap Sound Swap Row Object Reference=@n8, SortMode:String=@n10) -> (SortKey:String) {is=title K2Node_CallFunction}) -> (Output_Get:String)
   > n12 = Set SortKeyB(SortKeyB:String=GetSortKeyFromWidget(RowWidget=@n9, SortMode=@n10)) -> (Output_Get:String)
   > Branch(Condition="integer > integer"(A=CompareStrings!(LeftStr:String=@n11, RightStr:String=@n12) -> (ReturnValue:int=0)))
   > Set TempWidget(TempWidget:EUW Audio Sequencer Asset Swap Sound Swap Row Object Reference=@n7) -> (Output_Get:EUW Audio Sequencer Asset Swap Sound Swap Row Object Reference)
   n13 = Get GeneratedRowWidgets()
   > Set Array Elem!(
    execute
    TargetArray:Array of EUW Audio Sequencer Asset Swap Sound Swap Row Object References=@n13
    Index:int=@n2
    Item:"EUW Audio Sequencer Asset Swap Sound Swap Row Object Reference (by ref)"="Get  WidgetB"() -> (" WidgetB":EUW Audio Sequencer Asset Swap Sound Swap Row Object Reference)
    bSizeToFit:bool=false
   ) -> (then)
   > Set Array Elem(
    TargetArray=@n13
    Index=@n5
    Item=Get TempWidget() -> (TempWidget:EUW Audio Sequencer Asset Swap Sound Swap Row Object Reference)
    bSizeToFit=false
   )
   "Timespan > Timespan"(A:Wildcard, B:Wildcard) -> (ReturnValue:bool=~) {is=title K2Node_PromotableOperator}
  ]}
 UpdateSelectionCounter {nodes [
   UpdateSelectionCounter() {is=title K2Node_FunctionEntry}
   n1 = Length(TargetArray=Get SelectedRows())
   > Set Count(Count:int=@n1) -> (Output_Get:int=0)
   > Branch(Condition="Equal (Integer)"(A=@n1, B=1))
   Format Text!(Format="LOCTEXT(\"{0} row selected\\\"\")", 0:int=Get Count() -> (Count:int=0)) -> (Result)
  ]}
 UpdateToolbarButtonStates {nodes [
   UpdateToolbarButtonStates() {is=title K2Node_FunctionEntry}
   n1 = Get SelectedRows()
   > SetIsEnabled(
    self=[
     Get Button_SetAllSelected() -> (Button_SetAllSelected:Editor Utility Button Object Reference)
     Get Button_ClearAllSelected() -> (Button_ClearAllSelected:Editor Utility Button Object Reference)
    ]
    bInIsEnabled="integer > integer"(A=Length(TargetArray=@n1))
   )
   > RefreshSummaryText()
   > SortRows() {is=title K2Node_CallFunction}
  ]}
 ValidateRows {nodes [
   ValidateRows() {is=title K2Node_FunctionEntry}
   > n1 = For Each Loop(Array=Get GeneratedRowWidgets())
   > n2 = Is Valid(InputObject=Get SourceAudioSection(self=@n1))
   > n3 = Branch(Condition=MatchedOrManual(self=@n1))
   > n4 = Is Valid(InputObject=Get ReplacementSound(self=@n1))
   > n5 = Increment Int(Value=Get ReadyCount() -> (ReadyCount:int=0))
   > Set ReadyCount(ReadyCount:int=@n5) -> (Output_Get:int=0)
   n6 = Increment Int(" "=[@n4.Is Not Valid, @n2.Is Not Valid], Value=Get InvalidCount() -> (InvalidCount:int=0))
   > Set InvalidCount(InvalidCount:int=@n6) -> (Output_Get:int=0)
   n7 = Branch(execute=@n3.else, Condition="Equal Exactly (String)"(A=Get StatusText(self=@n1), B=Get NeedsReview(self=@n1)))
   > n8 = Increment Int(Value=Get ReviewCount() -> (ReviewCount:int=0))
   > Set ReviewCount(ReviewCount:int=@n8) -> (Output_Get:int=0)
   n9 = Increment Int(" "=@n7.else, Value=Get MissingCount() -> (MissingCount=0))
   > Set MissingCount(MissingCount=@n9) -> (Output_Get=0)
   "SetText (Text)"(
    execute=@n1.Completed
    self=Get Text_ToolbarStatus()
    InText=Format Text!(
     Format="LOCTEXT(\"{Ready} Ready, {Invalid} Invalid, {Review} Needs Review, {Missing} Missing. \")"
     Ready:int=Get ReadyCount() -> (ReadyCount=0)
     Invalid:int=Get InvalidCount() -> (InvalidCount=0)
     Review:int=Get ReviewCount() -> (ReviewCount=0)
     Missing:int=Get MissingCount() -> (MissingCount=0)
    ) -> (Result)
   )
  ]}
]
variables [
 CurrentSortMode {type=string, default=Row Index}
 GeneratedRowWidgets {type="array<EUW_AudioSequencer_AssetSwap_SoundSwapRow_C>", default []}
 IsRefreshing {type=boolean}
 ReplacementCandidateCount {type=integer}
 ReplacementCandidates {type="array<SoundBase>", default []}
 SearchText {type=string, default=""}
 SelectedLevelSequence {type=LevelSequence, default=null}
 SelectedRows {type="array<EUW_AudioSequencer_AssetSwap_SoundSwapRow_C>", default []}
 SelectedSequenceDisplayName {type=string, default=No Level Sequence Selected}
 SelectedTrackFilter {type=string, default=All Tracks}
 UniqueTrackNames {type="array<string>", default []}
]
widgets {
 info {rootWidgetClass=&Border, widgetCount=78}
 widgets [
  Border_MainPanel {widgetClassPath=&Border, widget=Border_MainPanel, properties {brushColor=<rgba 0.011 0.013 0.0176 1>, padding=<ltrb 20 20 20 20>}}
  VerticalBox_Root {
   widgetClassPath=&VerticalBox
   parent=Border_MainPanel
   widget=VerticalBox_Root
   slot {class=BorderSlot, properties {padding=<ltrb 20 20 20 20>}}
  }
  Text_Title {
   widgetClassPath=&TextBlock
   parent=VerticalBox_Root
   widget=Text_Title
   properties {
    text=Sound Swap Tool
    colorAndOpacity {colorUseRule=UseColor_Specified, specifiedColor=<rgba 0.7913 0.7991 0.8228 1>}
    font {
     bForceMonospaced=false
     bMaterialIsStencil=false
     fontMaterial=null
     fontObject=&Roboto
     letterSpacing=0
     monospacedWidth=1
     size=18
     skewAmount=0
     typefaceFontName=Bold
     outlineSettings {
      bApplyOutlineToDropShadows=false
      bMiteredCorners=false
      bSeparateFillAlpha=false
      outlineMaterial=null
      outlineSize=0
      outlineColor=<rgba 0 0 0 1>
     }
    }
   }
   slot {class=VerticalBoxSlot}
  }
  Spacer_154 {widgetClassPath=&Spacer, parent=VerticalBox_Root, widget=Spacer_154, properties {size=<1 10>}, slot {class=VerticalBoxSlot}}
  Border_Header {
   widgetClassPath=&Border
   parent=VerticalBox_Root
   widget=Border_Header
   properties {
    background {
     bIsDynamicallyLoaded=false
     drawAs=Border
     imageType=NoImage
     mirroring=NoMirror
     resourceName=None
     resourceObject=null
     tiling=NoTile
     imageSize=<32 32>
     margin=<ltrb 0 0 0 0>
     outlineSettings {
      bUseBrushTransparency=false
      roundingType=HalfHeightRadius
      width=0
      color {colorUseRule=UseColor_Specified, specifiedColor=<rgba 0 0 0 0>}
      cornerRadii=<0 0 0 0>
     }
     tintColor {colorUseRule=UseColor_Specified, specifiedColor=<rgba 1 1 1 1>}
     uVRegion {bIsValid=false, max=<0 0>, min=<0 0>}
    }
    brushColor=<rgba 0.0185 0.0203 0.0242 1>
    padding=<ltrb 12 12 12 12>
   }
   slot {class=VerticalBoxSlot}
  }
  VB_Header {
   widgetClassPath=&VerticalBox
   parent=Border_Header
   widget=VB_Header
   slot {class=BorderSlot, properties {padding=<ltrb 12 12 12 12>}}
  }
  Spacer_249 {widgetClassPath=&Spacer, parent=VB_Header, widget=Spacer_249, slot {class=VerticalBoxSlot, properties {padding=<ltrb 8 8 8 8>}}}
  HB_SequencerRow {
   widgetClassPath=&HorizontalBox
   parent=VB_Header
   widget=HB_SequencerRow
   slot {class=VerticalBoxSlot, properties {horizontalAlignment=HAlign_Center, verticalAlignment=VAlign_Center}}
  }
  HorizontalBox_639 {widgetClassPath=&HorizontalBox, parent=HB_SequencerRow, widget=HorizontalBox_639, slot {class=HorizontalBoxSlot}}
  Text_Sequencer {
   widgetClassPath=&TextBlock
   parent=HorizontalBox_639
   widget=Text_Sequencer
   properties {
    text=Sequencer:
    colorAndOpacity {colorUseRule=UseColor_Specified, specifiedColor=<rgba 0.7913 0.7991 0.8228 1>}
    font {
     bForceMonospaced=false
     bMaterialIsStencil=false
     fontMaterial=null
     fontObject=&Roboto
     letterSpacing=0
     monospacedWidth=1
     size=9.75
     skewAmount=0
     typefaceFontName=Bold
     outlineSettings {
      bApplyOutlineToDropShadows=false
      bMiteredCorners=false
      bSeparateFillAlpha=false
      outlineMaterial=null
      outlineSize=0
      outlineColor=<rgba 0 0 0 1>
     }
    }
   }
   slot {class=HorizontalBoxSlot, properties {verticalAlignment=VAlign_Center, padding=<ltrb 1 1 1 1>}}
  }
  SizeBox_2 {
   widgetClassPath=&SizeBox
   parent=HorizontalBox_639
   widget=SizeBox_2
   properties {bOverride_WidthOverride=true, widthOverride=300}
   slot {class=HorizontalBoxSlot, properties {horizontalAlignment=HAlign_Left, verticalAlignment=VAlign_Center, padding=<ltrb 0 6 0 6>}}
  }
  Border_0 {
   widgetClassPath=&Border
   parent=SizeBox_2
   widget=Border_0
   properties {brushColor=<rgba 0.0052 0.0052 0.0052 1>, padding=<ltrb 1 1 1 1>}
   slot {class=SizeBoxSlot, properties {padding=<ltrb 0 0 6 0>}}
  }
  AssetDropWidget_SelectedSequence {
   widgetClassPath=/Script/FortTechAudioToolsEditor.AudioAssetDropTargetWidget
   bIsVariable=true
   parent=Border_0
   widget=AssetDropWidget_SelectedSequence
   properties {
    bOverride_Cursor=true
    text=No Level Sequence Selected
    textJustification=Left
    allowedAssetClasses [&LevelSequence]
    colorHovered {colorUseRule=UseColor_Specified, specifiedColor=<rgba 0.0232 0.0262 0.0331 1>}
    colorNormal {colorUseRule=UseColor_Specified, specifiedColor=<rgba 0.0232 0.0262 0.0331 1>}
    colorPressed {colorUseRule=UseColor_Specified, specifiedColor=<rgba 0.0232 0.1462 0.0331 1>}
    font {
     bForceMonospaced=false
     bMaterialIsStencil=false
     fontMaterial=null
     fontObject=/LauncherSocial/BrutalType
     letterSpacing=0
     monospacedWidth=1
     size=9
     skewAmount=0
     typefaceFontName=Regular
     outlineSettings {
      bApplyOutlineToDropShadows=false
      bMiteredCorners=false
      bSeparateFillAlpha=false
      outlineMaterial=null
      outlineSize=0
      outlineColor=<rgba 0 0 0 1>
     }
    }
    shadowColorAndOpacity=<rgba 0 0 0 1>
    textColor {colorUseRule=UseColor_Specified, specifiedColor=<rgba 0.7913 0.7991 0.8228 1>}
   }
   slot {class=BorderSlot, properties {padding=<ltrb 1 1 1 1>}}
  }
  Button_UseSelected {
   widgetClassPath=&EditorUtilityButton
   parent=HB_SequencerRow
   widget=Button_UseSelected
   properties {
    toolTipText=Reset the source reference back to default asset.
    backgroundColor=<rgba 0.0452 0.2051 0.7011 1>
    widgetStyle {
     clickedSlateSound {resourceObject=null}
     disabled {
      bIsDynamicallyLoaded=false
      drawAs=RoundedBox
      imageType=NoImage
      mirroring=NoMirror
      resourceName=None
      resourceObject=null
      tiling=NoTile
      imageSize=<0 0>
      margin=<ltrb 0 0 0 0>
      outlineSettings {
       bUseBrushTransparency=false
       roundingType=FixedRadius
       width=1
       color {colorUseRule=UseColor_ColorTable, specifiedColor=<rgba 0 0 0 0>}
       cornerRadii=<4 4 4 4>
      }
      tintColor {colorUseRule=UseColor_Specified, specifiedColor=<rgba 0.0395 0.0395 0.0395 1>}
      uVRegion {bIsValid=false, max=<0 0>, min=<0 0>}
     }
     disabledForeground {colorUseRule=UseColor_ColorTable, specifiedColor=<rgba 0 0 0 0>}
     hovered {
      bIsDynamicallyLoaded=false
      drawAs=RoundedBox
      imageType=NoImage
      mirroring=NoMirror
      resourceName=None
      resourceObject=null
      tiling=NoTile
      imageSize=<32 32>
      margin=<ltrb 0 0 0 0>
      outlineSettings {
       bUseBrushTransparency=false
       roundingType=FixedRadius
       width=1
       color {colorUseRule=UseColor_ColorTable, specifiedColor=<rgba 0 0 0 0>}
       cornerRadii=<4 4 4 4>
      }
      tintColor {colorUseRule=UseColor_Specified, specifiedColor=<rgba 0.0685 0.2542 0.807 1>}
      uVRegion {bIsValid=false, max=<0 0>, min=<0 0>}
     }
     hoveredForeground {colorUseRule=UseColor_ColorTable, specifiedColor=<rgba 0 0 0 0>}
     hoveredSlateSound {resourceObject=null}
     normal {
      bIsDynamicallyLoaded=false
      drawAs=RoundedBox
      imageType=NoImage
      mirroring=NoMirror
      resourceName=None
      resourceObject=null
      tiling=NoTile
      imageSize=<32 32>
      margin=<ltrb 0 0 0 0>
      outlineSettings {
       bUseBrushTransparency=false
       roundingType=FixedRadius
       width=1
       color {colorUseRule=UseColor_ColorTable, specifiedColor=<rgba 0 0 0 0>}
       cornerRadii=<4 4 4 4>
      }
      tintColor {colorUseRule=UseColor_Specified, specifiedColor=<rgba 0.0395 0.0395 0.0395 1>}
      uVRegion {bIsValid=false, max=<0 0>, min=<0 0>}
     }
     normalForeground {colorUseRule=UseColor_ColorTable, specifiedColor=<rgba 0 0 0 0>}
     normalPadding=<ltrb 12 1.5 12 1.5>
     pressed {
      bIsDynamicallyLoaded=false
      drawAs=RoundedBox
      imageType=NoImage
      mirroring=NoMirror
      resourceName=None
      resourceObject=null
      tiling=NoTile
      imageSize=<32 32>
      margin=<ltrb 0 0 0 0>
      outlineSettings {
       bUseBrushTransparency=false
       roundingType=FixedRadius
       width=1
       color {colorUseRule=UseColor_ColorTable, specifiedColor=<rgba 0 0 0 0>}
       cornerRadii=<4 4 4 4>
      }
      tintColor {colorUseRule=UseColor_Specified, specifiedColor=<rgba 0.0284 0.0284 0.0284 1>}
      uVRegion {bIsValid=false, max=<0 0>, min=<0 0>}
     }
     pressedForeground {colorUseRule=UseColor_ColorTable, specifiedColor=<rgba 0 0 0 0>}
     pressedPadding=<ltrb 12 2.5 12 0.5>
     pressedSlateSound {resourceObject=null}
    }
   }
   slot {class=HorizontalBoxSlot, properties {verticalAlignment=VAlign_Center, padding=<ltrb 8 0 0 0>}}
  }
  Button_UseSelected_Text {
   widgetClassPath=&UEFN_TextBlock
   parent=Button_UseSelected
   widget=Button_UseSelected_Text
   properties {
    justification=Center
    text=Use Selected
    colorAndOpacity {colorUseRule=UseColor_Specified, specifiedColor=<rgba 0.7913 0.7991 0.8228 1>}
    font {
     bForceMonospaced=false
     bMaterialIsStencil=false
     fontMaterial=null
     fontObject=&BurbankSmall
     letterSpacing=0
     monospacedWidth=1
     size=9
     skewAmount=0
     typefaceFontName=Regular
     outlineSettings {
      bApplyOutlineToDropShadows=false
      bMiteredCorners=false
      bSeparateFillAlpha=false
      outlineMaterial=null
      outlineSize=0
      outlineColor=<rgba 0 0 0 1>
     }
    }
    shadowColorAndOpacity=<rgba 0 0 0 1>
   }
   slot {class=ButtonSlot}
  }
  RefreshButton {
   widgetClassPath=&EditorUtilityButton
   parent=HB_SequencerRow
   widget=RefreshButton
   properties {
    toolTipText=Reload audio from the selected Level Sequence
    backgroundColor=<rgba 0.0423 0.0497 0.063 1>
    widgetStyle {
     clickedSlateSound {resourceObject=null}
     disabled {
      bIsDynamicallyLoaded=false
      drawAs=RoundedBox
      imageType=NoImage
      mirroring=NoMirror
      resourceName=None
      resourceObject=null
      tiling=NoTile
      imageSize=<0 0>
      margin=<ltrb 0 0 0 0>
      outlineSettings {
       bUseBrushTransparency=false
       roundingType=FixedRadius
       width=1
       color {colorUseRule=UseColor_ColorTable, specifiedColor=<rgba 0 0 0 0>}
       cornerRadii=<4 4 4 4>
      }
      tintColor {colorUseRule=UseColor_Specified, specifiedColor=<rgba 0.0395 0.0395 0.0395 1>}
      uVRegion {bIsValid=false, max=<0 0>, min=<0 0>}
     }
     disabledForeground {colorUseRule=UseColor_ColorTable, specifiedColor=<rgba 0 0 0 0>}
     hovered {
      bIsDynamicallyLoaded=false
      drawAs=RoundedBox
      imageType=NoImage
      mirroring=NoMirror
      resourceName=None
      resourceObject=null
      tiling=NoTile
      imageSize=<32 32>
      margin=<ltrb 0 0 0 0>
      outlineSettings {
       bUseBrushTransparency=false
       roundingType=FixedRadius
       width=1
       color {colorUseRule=UseColor_ColorTable, specifiedColor=<rgba 0 0 0 0>}
       cornerRadii=<4 4 4 4>
      }
      tintColor {colorUseRule=UseColor_Specified, specifiedColor=<rgba 0.0685 0.2542 0.807 1>}
      uVRegion {bIsValid=false, max=<0 0>, min=<0 0>}
     }
     hoveredForeground {colorUseRule=UseColor_ColorTable, specifiedColor=<rgba 0 0 0 0>}
     hoveredSlateSound {resourceObject=null}
     normal {
      bIsDynamicallyLoaded=false
      drawAs=RoundedBox
      imageType=NoImage
      mirroring=NoMirror
      resourceName=None
      resourceObject=null
      tiling=NoTile
      imageSize=<32 32>
      margin=<ltrb 0 0 0 0>
      outlineSettings {
       bUseBrushTransparency=false
       roundingType=FixedRadius
       width=1
       color {colorUseRule=UseColor_ColorTable, specifiedColor=<rgba 0 0 0 0>}
       cornerRadii=<4 4 4 4>
      }
      tintColor {colorUseRule=UseColor_Specified, specifiedColor=<rgba 0.0395 0.0395 0.0395 1>}
      uVRegion {bIsValid=false, max=<0 0>, min=<0 0>}
     }
     normalForeground {colorUseRule=UseColor_ColorTable, specifiedColor=<rgba 0 0 0 0>}
     normalPadding=<ltrb 12 1.5 12 1.5>
     pressed {
      bIsDynamicallyLoaded=false
      drawAs=RoundedBox
      imageType=NoImage
      mirroring=NoMirror
      resourceName=None
      resourceObject=null
      tiling=NoTile
      imageSize=<32 32>
      margin=<ltrb 0 0 0 0>
      outlineSettings {
       bUseBrushTransparency=false
       roundingType=FixedRadius
       width=1
       color {colorUseRule=UseColor_ColorTable, specifiedColor=<rgba 0 0 0 0>}
       cornerRadii=<4 4 4 4>
      }
      tintColor {colorUseRule=UseColor_Specified, specifiedColor=<rgba 0.0284 0.0284 0.0284 1>}
      uVRegion {bIsValid=false, max=<0 0>, min=<0 0>}
     }
     pressedForeground {colorUseRule=UseColor_ColorTable, specifiedColor=<rgba 0 0 0 0>}
     pressedPadding=<ltrb 12 2.5 12 0.5>
     pressedSlateSound {resourceObject=null}
    }
   }
   slot {class=HorizontalBoxSlot, properties {verticalAlignment=VAlign_Center, padding=<ltrb 8 0 0 0>}}
  }
  RefreshText {
   widgetClassPath=&UEFN_TextBlock
   parent=RefreshButton
   widget=RefreshText
   properties {
    justification=Center
    text=Reload
    font {
     bForceMonospaced=false
     bMaterialIsStencil=false
     fontMaterial=null
     fontObject=&BurbankSmall
     letterSpacing=0
     monospacedWidth=1
     size=9.75
     skewAmount=0
     typefaceFontName=Regular
     outlineSettings {
      bApplyOutlineToDropShadows=false
      bMiteredCorners=false
      bSeparateFillAlpha=false
      outlineMaterial=null
      outlineSize=0
      outlineColor=<rgba 0 0 0 1>
     }
    }
    shadowColorAndOpacity=<rgba 0 0 0 1>
   }
   slot {class=ButtonSlot}
  }
  Text_SequenceStatus {
   widgetClassPath=&UEFN_TextBlock
   bIsVariable=true
   parent=VB_Header
   widget=Text_SequenceStatus
   properties {
    justification=Center
    text=No valid Level Sequence selected
    colorAndOpacity {colorUseRule=UseColor_Specified, specifiedColor=<rgba 0.6795 0.3968 0.0482 1>}
    font {
     bForceMonospaced=false
     bMaterialIsStencil=false
     fontMaterial=null
     fontObject=&BurbankSmall
     letterSpacing=0
     monospacedWidth=1
     size=9
     skewAmount=0
     typefaceFontName=Regular
     outlineSettings {
      bApplyOutlineToDropShadows=false
      bMiteredCorners=false
      bSeparateFillAlpha=false
      outlineMaterial=null
      outlineSize=0
      outlineColor=<rgba 0 0 0 1>
     }
    }
    shadowColorAndOpacity=<rgba 0 0 0 1>
   }
   slot {class=VerticalBoxSlot, properties {horizontalAlignment=HAlign_Center, verticalAlignment=VAlign_Center, padding=<ltrb 4 2 4 2>}}
  }
  Text_SourceMode {
   widgetClassPath=&UEFN_TextBlock
   parent=VB_Header
   widget=Text_SourceMode
   properties {
    justification=Center
    text=Use Selected: Auto-pull from sequencer audio tracks
    colorAndOpacity {colorUseRule=UseColor_Specified, specifiedColor=<rgba 0.3968 0.4342 0.491 1>}
    font {
     bForceMonospaced=false
     bMaterialIsStencil=false
     fontMaterial=null
     fontObject=&BrutalType
     letterSpacing=0
     monospacedWidth=1
     size=9
     skewAmount=0
     typefaceFontName=Regular
     outlineSettings {
      bApplyOutlineToDropShadows=false
      bMiteredCorners=false
      bSeparateFillAlpha=false
      outlineMaterial=null
      outlineSize=0
      outlineColor=<rgba 0 0 0 1>
     }
    }
    shadowColorAndOpacity=<rgba 0 0 0 1>
   }
   slot {class=VerticalBoxSlot, properties {horizontalAlignment=HAlign_Center, verticalAlignment=VAlign_Center, padding=<ltrb 4 2 4 2>}}
  }
  Text_ReplacementMode {
   widgetClassPath=&UEFN_TextBlock
   parent=VB_Header
   widget=Text_ReplacementMode
   properties {
    justification=Center
    text="Auto Match: Auto-fill from Content Browser, then manual review"
    colorAndOpacity {colorUseRule=UseColor_Specified, specifiedColor=<rgba 0.3968 0.4342 0.491 1>}
    font {
     bForceMonospaced=false
     bMaterialIsStencil=false
     fontMaterial=null
     fontObject=&BrutalType
     letterSpacing=0
     monospacedWidth=1
     size=9
     skewAmount=0
     typefaceFontName=Regular
     outlineSettings {
      bApplyOutlineToDropShadows=false
      bMiteredCorners=false
      bSeparateFillAlpha=false
      outlineMaterial=null
      outlineSize=0
      outlineColor=<rgba 0 0 0 1>
     }
    }
    shadowColorAndOpacity=<rgba 0 0 0 1>
   }
   slot {class=VerticalBoxSlot, properties {horizontalAlignment=HAlign_Center, verticalAlignment=VAlign_Center, padding=<ltrb 4 2 4 2>}}
  }
  Spacer_642 {
   widgetClassPath=&Spacer
   parent=VerticalBox_Root
   widget=Spacer_642
   slot {class=VerticalBoxSlot, properties {padding=<ltrb 0 12 0 0>}}
  }
  Border_Toolbar {
   widgetClassPath=&Border
   parent=VerticalBox_Root
   widget=Border_Toolbar
   properties {horizontalAlignment=HAlign_Center, verticalAlignment=VAlign_Center, brushColor=<rgba 0.0423 0.0497 0.063 1>}
   slot {class=VerticalBoxSlot}
  }
  VB_Toolbar {
   widgetClassPath=&VerticalBox
   parent=Border_Toolbar
   widget=VB_Toolbar
   slot {class=BorderSlot, properties {horizontalAlignment=HAlign_Center, verticalAlignment=VAlign_Center}}
  }
  HB_SelectedBar {
   widgetClassPath=&HorizontalBox
   parent=VB_Toolbar
   widget=HB_SelectedBar
   slot {class=VerticalBoxSlot, properties {horizontalAlignment=HAlign_Center, verticalAlignment=VAlign_Center}}
  }
  Button_AutoMatch {
   widgetClassPath=&EditorUtilityButton
   parent=HB_SelectedBar
   widget=Button_AutoMatch
   properties {
    toolTipText="Automatically match sounds by name from selected assets\""
    backgroundColor=<rgba 0.0452 0.2051 0.7011 1>
    widgetStyle {
     clickedSlateSound {resourceObject=null}
     disabled {
      bIsDynamicallyLoaded=false
      drawAs=RoundedBox
      imageType=NoImage
      mirroring=NoMirror
      resourceName=None
      resourceObject=null
      tiling=NoTile
      imageSize=<0 0>
      margin=<ltrb 0 0 0 0>
      outlineSettings {
       bUseBrushTransparency=false
       roundingType=FixedRadius
       width=1
       color {colorUseRule=UseColor_ColorTable, specifiedColor=<rgba 0 0 0 0>}
       cornerRadii=<4 4 4 4>
      }
      tintColor {colorUseRule=UseColor_Specified, specifiedColor=<rgba 0.0395 0.0395 0.0395 1>}
      uVRegion {bIsValid=false, max=<0 0>, min=<0 0>}
     }
     disabledForeground {colorUseRule=UseColor_ColorTable, specifiedColor=<rgba 0 0 0 0>}
     hovered {
      bIsDynamicallyLoaded=false
      drawAs=RoundedBox
      imageType=NoImage
      mirroring=NoMirror
      resourceName=None
      resourceObject=null
      tiling=NoTile
      imageSize=<32 32>
      margin=<ltrb 0 0 0 0>
      outlineSettings {
       bUseBrushTransparency=false
       roundingType=FixedRadius
       width=1
       color {colorUseRule=UseColor_ColorTable, specifiedColor=<rgba 0 0 0 0>}
       cornerRadii=<4 4 4 4>
      }
      tintColor {colorUseRule=UseColor_Specified, specifiedColor=<rgba 0.0685 0.2542 0.807 1>}
      uVRegion {bIsValid=false, max=<0 0>, min=<0 0>}
     }
     hoveredForeground {colorUseRule=UseColor_ColorTable, specifiedColor=<rgba 0 0 0 0>}
     hoveredSlateSound {resourceObject=null}
     normal {
      bIsDynamicallyLoaded=false
      drawAs=RoundedBox
      imageType=NoImage
      mirroring=NoMirror
      resourceName=None
      resourceObject=null
      tiling=NoTile
      imageSize=<32 32>
      margin=<ltrb 0 0 0 0>
      outlineSettings {
       bUseBrushTransparency=false
       roundingType=FixedRadius
       width=1
       color {colorUseRule=UseColor_ColorTable, specifiedColor=<rgba 0 0 0 0>}
       cornerRadii=<4 4 4 4>
      }
      tintColor {colorUseRule=UseColor_Specified, specifiedColor=<rgba 0.0395 0.0395 0.0395 1>}
      uVRegion {bIsValid=false, max=<0 0>, min=<0 0>}
     }
     normalForeground {colorUseRule=UseColor_ColorTable, specifiedColor=<rgba 0 0 0 0>}
     normalPadding=<ltrb 12 1.5 12 1.5>
     pressed {
      bIsDynamicallyLoaded=false
      drawAs=RoundedBox
      imageType=NoImage
      mirroring=NoMirror
      resourceName=None
      resourceObject=null
      tiling=NoTile
      imageSize=<32 32>
      margin=<ltrb 0 0 0 0>
      outlineSettings {
       bUseBrushTransparency=false
       roundingType=FixedRadius
       width=1
       color {colorUseRule=UseColor_ColorTable, specifiedColor=<rgba 0 0 0 0>}
       cornerRadii=<4 4 4 4>
      }
      tintColor {colorUseRule=UseColor_Specified, specifiedColor=<rgba 0.0284 0.0284 0.0284 1>}
      uVRegion {bIsValid=false, max=<0 0>, min=<0 0>}
     }
     pressedForeground {colorUseRule=UseColor_ColorTable, specifiedColor=<rgba 0 0 0 0>}
     pressedPadding=<ltrb 12 2.5 12 0.5>
     pressedSlateSound {resourceObject=null}
    }
   }
   slot {class=HorizontalBoxSlot, properties {verticalAlignment=VAlign_Center, padding=<ltrb 8 0 12 0>}}
  }
  Button_AutoMatch_Text {
   widgetClassPath=&UEFN_TextBlock
   parent=Button_AutoMatch
   widget=Button_AutoMatch_Text
   properties {
    justification=Center
    text=Auto Match
    colorAndOpacity {colorUseRule=UseColor_Specified, specifiedColor=<rgba 0.7913 0.7991 0.8228 1>}
    font {
     bForceMonospaced=false
     bMaterialIsStencil=false
     fontMaterial=null
     fontObject=&BurbankSmall
     letterSpacing=0
     monospacedWidth=1
     size=9
     skewAmount=0
     typefaceFontName=Regular
     outlineSettings {
      bApplyOutlineToDropShadows=false
      bMiteredCorners=false
      bSeparateFillAlpha=false
      outlineMaterial=null
      outlineSize=0
      outlineColor=<rgba 0 0 0 1>
     }
    }
    shadowColorAndOpacity=<rgba 0 0 0 1>
   }
   slot {class=ButtonSlot}
  }
  Button_CreateCuesAndAutoMatch {
   widgetClassPath=&EditorUtilityButton
   parent=HB_SelectedBar
   widget=Button_CreateCuesAndAutoMatch
   properties {
    toolTipText="Automatically match sounds by name from selected assets\""
    backgroundColor=<rgba 0.0452 0.2051 0.7011 1>
    widgetStyle {
     clickedSlateSound {resourceObject=null}
     disabled {
      bIsDynamicallyLoaded=false
      drawAs=RoundedBox
      imageType=NoImage
      mirroring=NoMirror
      resourceName=None
      resourceObject=null
      tiling=NoTile
      imageSize=<0 0>
      margin=<ltrb 0 0 0 0>
      outlineSettings {
       bUseBrushTransparency=false
       roundingType=FixedRadius
       width=1
       color {colorUseRule=UseColor_ColorTable, specifiedColor=<rgba 0 0 0 0>}
       cornerRadii=<4 4 4 4>
      }
      tintColor {colorUseRule=UseColor_Specified, specifiedColor=<rgba 0.0395 0.0395 0.0395 1>}
      uVRegion {bIsValid=false, max=<0 0>, min=<0 0>}
     }
     disabledForeground {colorUseRule=UseColor_ColorTable, specifiedColor=<rgba 0 0 0 0>}
     hovered {
      bIsDynamicallyLoaded=false
      drawAs=RoundedBox
      imageType=NoImage
      mirroring=NoMirror
      resourceName=None
      resourceObject=null
      tiling=NoTile
      imageSize=<32 32>
      margin=<ltrb 0 0 0 0>
      outlineSettings {
       bUseBrushTransparency=false
       roundingType=FixedRadius
       width=1
       color {colorUseRule=UseColor_ColorTable, specifiedColor=<rgba 0 0 0 0>}
       cornerRadii=<4 4 4 4>
      }
      tintColor {colorUseRule=UseColor_Specified, specifiedColor=<rgba 0.0685 0.2542 0.807 1>}
      uVRegion {bIsValid=false, max=<0 0>, min=<0 0>}
     }
     hoveredForeground {colorUseRule=UseColor_ColorTable, specifiedColor=<rgba 0 0 0 0>}
     hoveredSlateSound {resourceObject=null}
     normal {
      bIsDynamicallyLoaded=false
      drawAs=RoundedBox
      imageType=NoImage
      mirroring=NoMirror
      resourceName=None
      resourceObject=null
      tiling=NoTile
      imageSize=<32 32>
      margin=<ltrb 0 0 0 0>
      outlineSettings {
       bUseBrushTransparency=false
       roundingType=FixedRadius
       width=1
       color {colorUseRule=UseColor_ColorTable, specifiedColor=<rgba 0 0 0 0>}
       cornerRadii=<4 4 4 4>
      }
      tintColor {colorUseRule=UseColor_Specified, specifiedColor=<rgba 0.0395 0.0395 0.0395 1>}
      uVRegion {bIsValid=false, max=<0 0>, min=<0 0>}
     }
     normalForeground {colorUseRule=UseColor_ColorTable, specifiedColor=<rgba 0 0 0 0>}
     normalPadding=<ltrb 12 1.5 12 1.5>
     pressed {
      bIsDynamicallyLoaded=false
      drawAs=RoundedBox
      imageType=NoImage
      mirroring=NoMirror
      resourceName=None
      resourceObject=null
      tiling=NoTile
      imageSize=<32 32>
      margin=<ltrb 0 0 0 0>
      outlineSettings {
       bUseBrushTransparency=false
       roundingType=FixedRadius
       width=1
       color {colorUseRule=UseColor_ColorTable, specifiedColor=<rgba 0 0 0 0>}
       cornerRadii=<4 4 4 4>
      }
      tintColor {colorUseRule=UseColor_Specified, specifiedColor=<rgba 0.0284 0.0284 0.0284 1>}
      uVRegion {bIsValid=false, max=<0 0>, min=<0 0>}
     }
     pressedForeground {colorUseRule=UseColor_ColorTable, specifiedColor=<rgba 0 0 0 0>}
     pressedPadding=<ltrb 12 2.5 12 0.5>
     pressedSlateSound {resourceObject=null}
    }
   }
   slot {class=HorizontalBoxSlot, properties {verticalAlignment=VAlign_Center, padding=<ltrb 8 0 12 0>}}
  }
  Button_CCAutoMatch_Text {
   widgetClassPath=&UEFN_TextBlock
   parent=Button_CreateCuesAndAutoMatch
   widget=Button_CCAutoMatch_Text
   properties {
    justification=Center
    text=Create Cues + AutoMatch
    colorAndOpacity {colorUseRule=UseColor_Specified, specifiedColor=<rgba 0.7913 0.7991 0.8228 1>}
    font {
     bForceMonospaced=false
     bMaterialIsStencil=false
     fontMaterial=null
     fontObject=&BurbankSmall
     letterSpacing=0
     monospacedWidth=1
     size=9
     skewAmount=0
     typefaceFontName=Regular
     outlineSettings {
      bApplyOutlineToDropShadows=false
      bMiteredCorners=false
      bSeparateFillAlpha=false
      outlineMaterial=null
      outlineSize=0
      outlineColor=<rgba 0 0 0 1>
     }
    }
    shadowColorAndOpacity=<rgba 0 0 0 1>
   }
   slot {class=ButtonSlot}
  }
  Button_SelectAll {
   widgetClassPath=&EditorUtilityButton
   parent=HB_SelectedBar
   widget=Button_SelectAll
   properties {
    backgroundColor=<rgba 0.0423 0.0497 0.063 1>
    widgetStyle {
     clickedSlateSound {resourceObject=null}
     disabled {
      bIsDynamicallyLoaded=false
      drawAs=RoundedBox
      imageType=NoImage
      mirroring=NoMirror
      resourceName=None
      resourceObject=null
      tiling=NoTile
      imageSize=<0 0>
      margin=<ltrb 0 0 0 0>
      outlineSettings {
       bUseBrushTransparency=false
       roundingType=FixedRadius
       width=1
       color {colorUseRule=UseColor_ColorTable, specifiedColor=<rgba 0 0 0 0>}
       cornerRadii=<4 4 4 4>
      }
      tintColor {colorUseRule=UseColor_Specified, specifiedColor=<rgba 0.0395 0.0395 0.0395 1>}
      uVRegion {bIsValid=false, max=<0 0>, min=<0 0>}
     }
     disabledForeground {colorUseRule=UseColor_ColorTable, specifiedColor=<rgba 0 0 0 0>}
     hovered {
      bIsDynamicallyLoaded=false
      drawAs=RoundedBox
      imageType=NoImage
      mirroring=NoMirror
      resourceName=None
      resourceObject=null
      tiling=NoTile
      imageSize=<32 32>
      margin=<ltrb 0 0 0 0>
      outlineSettings {
       bUseBrushTransparency=false
       roundingType=FixedRadius
       width=1
       color {colorUseRule=UseColor_ColorTable, specifiedColor=<rgba 0 0 0 0>}
       cornerRadii=<4 4 4 4>
      }
      tintColor {colorUseRule=UseColor_Specified, specifiedColor=<rgba 0.0685 0.2542 0.807 1>}
      uVRegion {bIsValid=false, max=<0 0>, min=<0 0>}
     }
     hoveredForeground {colorUseRule=UseColor_ColorTable, specifiedColor=<rgba 0 0 0 0>}
     hoveredSlateSound {resourceObject=null}
     normal {
      bIsDynamicallyLoaded=false
      drawAs=RoundedBox
      imageType=NoImage
      mirroring=NoMirror
      resourceName=None
      resourceObject=null
      tiling=NoTile
      imageSize=<32 32>
      margin=<ltrb 0 0 0 0>
      outlineSettings {
       bUseBrushTransparency=false
       roundingType=FixedRadius
       width=1
       color {colorUseRule=UseColor_ColorTable, specifiedColor=<rgba 0 0 0 0>}
       cornerRadii=<4 4 4 4>
      }
      tintColor {colorUseRule=UseColor_Specified, specifiedColor=<rgba 0.0395 0.0395 0.0395 1>}
      uVRegion {bIsValid=false, max=<0 0>, min=<0 0>}
     }
     normalForeground {colorUseRule=UseColor_ColorTable, specifiedColor=<rgba 0 0 0 0>}
     normalPadding=<ltrb 12 1.5 12 1.5>
     pressed {
      bIsDynamicallyLoaded=false
      drawAs=RoundedBox
      imageType=NoImage
      mirroring=NoMirror
      resourceName=None
      resourceObject=null
      tiling=NoTile
      imageSize=<32 32>
      margin=<ltrb 0 0 0 0>
      outlineSettings {
       bUseBrushTransparency=false
       roundingType=FixedRadius
       width=1
       color {colorUseRule=UseColor_ColorTable, specifiedColor=<rgba 0 0 0 0>}
       cornerRadii=<4 4 4 4>
      }
      tintColor {colorUseRule=UseColor_Specified, specifiedColor=<rgba 0.0284 0.0284 0.0284 1>}
      uVRegion {bIsValid=false, max=<0 0>, min=<0 0>}
     }
     pressedForeground {colorUseRule=UseColor_ColorTable, specifiedColor=<rgba 0 0 0 0>}
     pressedPadding=<ltrb 12 2.5 12 0.5>
     pressedSlateSound {resourceObject=null}
    }
   }
   slot {class=HorizontalBoxSlot, properties {verticalAlignment=VAlign_Center, padding=<ltrb 8 0 8 0>}}
  }
  Text_SelectAll {
   widgetClassPath=&UEFN_TextBlock
   parent=Button_SelectAll
   widget=Text_SelectAll
   properties {
    justification=Center
    text=Select All
    colorAndOpacity {colorUseRule=UseColor_Specified, specifiedColor=<rgba 0.7913 0.7991 0.8228 1>}
    font {
     bForceMonospaced=false
     bMaterialIsStencil=false
     fontMaterial=null
     fontObject=&BurbankSmall
     letterSpacing=0
     monospacedWidth=1
     size=9
     skewAmount=0
     typefaceFontName=Regular
     outlineSettings {
      bApplyOutlineToDropShadows=false
      bMiteredCorners=false
      bSeparateFillAlpha=false
      outlineMaterial=null
      outlineSize=0
      outlineColor=<rgba 0 0 0 1>
     }
    }
    shadowColorAndOpacity=<rgba 0 0 0 1>
   }
   slot {class=ButtonSlot}
  }
  Button_DeselectAll {
   widgetClassPath=&EditorUtilityButton
   parent=HB_SelectedBar
   widget=Button_DeselectAll
   properties {
    backgroundColor=<rgba 0.0423 0.0497 0.063 1>
    widgetStyle {
     clickedSlateSound {resourceObject=null}
     disabled {
      bIsDynamicallyLoaded=false
      drawAs=RoundedBox
      imageType=NoImage
      mirroring=NoMirror
      resourceName=None
      resourceObject=null
      tiling=NoTile
      imageSize=<0 0>
      margin=<ltrb 0 0 0 0>
      outlineSettings {
       bUseBrushTransparency=false
       roundingType=FixedRadius
       width=1
       color {colorUseRule=UseColor_ColorTable, specifiedColor=<rgba 0 0 0 0>}
       cornerRadii=<4 4 4 4>
      }
      tintColor {colorUseRule=UseColor_Specified, specifiedColor=<rgba 0.0395 0.0395 0.0395 1>}
      uVRegion {bIsValid=false, max=<0 0>, min=<0 0>}
     }
     disabledForeground {colorUseRule=UseColor_ColorTable, specifiedColor=<rgba 0 0 0 0>}
     hovered {
      bIsDynamicallyLoaded=false
      drawAs=RoundedBox
      imageType=NoImage
      mirroring=NoMirror
      resourceName=None
      resourceObject=null
      tiling=NoTile
      imageSize=<32 32>
      margin=<ltrb 0 0 0 0>
      outlineSettings {
       bUseBrushTransparency=false
       roundingType=FixedRadius
       width=1
       color {colorUseRule=UseColor_ColorTable, specifiedColor=<rgba 0 0 0 0>}
       cornerRadii=<4 4 4 4>
      }
      tintColor {colorUseRule=UseColor_Specified, specifiedColor=<rgba 0.0685 0.2542 0.807 1>}
      uVRegion {bIsValid=false, max=<0 0>, min=<0 0>}
     }
     hoveredForeground {colorUseRule=UseColor_ColorTable, specifiedColor=<rgba 0 0 0 0>}
     hoveredSlateSound {resourceObject=null}
     normal {
      bIsDynamicallyLoaded=false
      drawAs=RoundedBox
      imageType=NoImage
      mirroring=NoMirror
      resourceName=None
      resourceObject=null
      tiling=NoTile
      imageSize=<32 32>
      margin=<ltrb 0 0 0 0>
      outlineSettings {
       bUseBrushTransparency=false
       roundingType=FixedRadius
       width=1
       color {colorUseRule=UseColor_ColorTable, specifiedColor=<rgba 0 0 0 0>}
       cornerRadii=<4 4 4 4>
      }
      tintColor {colorUseRule=UseColor_Specified, specifiedColor=<rgba 0.0395 0.0395 0.0395 1>}
      uVRegion {bIsValid=false, max=<0 0>, min=<0 0>}
     }
     normalForeground {colorUseRule=UseColor_ColorTable, specifiedColor=<rgba 0 0 0 0>}
     normalPadding=<ltrb 12 1.5 12 1.5>
     pressed {
      bIsDynamicallyLoaded=false
      drawAs=RoundedBox
      imageType=NoImage
      mirroring=NoMirror
      resourceName=None
      resourceObject=null
      tiling=NoTile
      imageSize=<32 32>
      margin=<ltrb 0 0 0 0>
      outlineSettings {
       bUseBrushTransparency=false
       roundingType=FixedRadius
       width=1
       color {colorUseRule=UseColor_ColorTable, specifiedColor=<rgba 0 0 0 0>}
       cornerRadii=<4 4 4 4>
      }
      tintColor {colorUseRule=UseColor_Specified, specifiedColor=<rgba 0.0284 0.0284 0.0284 1>}
      uVRegion {bIsValid=false, max=<0 0>, min=<0 0>}
     }
     pressedForeground {colorUseRule=UseColor_ColorTable, specifiedColor=<rgba 0 0 0 0>}
     pressedPadding=<ltrb 12 2.5 12 0.5>
     pressedSlateSound {resourceObject=null}
    }
   }
   slot {class=HorizontalBoxSlot, properties {verticalAlignment=VAlign_Center, padding=<ltrb 8 0 8 0>}}
  }
  Text_DeselectAll {
   widgetClassPath=&UEFN_TextBlock
   parent=Button_DeselectAll
   widget=Text_DeselectAll
   properties {
    justification=Center
    text=Deselect All
    colorAndOpacity {colorUseRule=UseColor_Specified, specifiedColor=<rgba 0.7913 0.7991 0.8228 1>}
    font {
     bForceMonospaced=false
     bMaterialIsStencil=false
     fontMaterial=null
     fontObject=&BurbankSmall
     letterSpacing=0
     monospacedWidth=1
     size=9
     skewAmount=0
     typefaceFontName=Regular
     outlineSettings {
      bApplyOutlineToDropShadows=false
      bMiteredCorners=false
      bSeparateFillAlpha=false
      outlineMaterial=null
      outlineSize=0
      outlineColor=<rgba 0 0 0 1>
     }
    }
    shadowColorAndOpacity=<rgba 0 0 0 1>
   }
   slot {class=ButtonSlot}
  }
  Button_ClearAllSelected {
   widgetClassPath=&EditorUtilityButton
   parent=HB_SelectedBar
   widget=Button_ClearAllSelected
   properties {
    bIsEnabled=false
    backgroundColor=<rgba 0.0423 0.0497 0.063 1>
    widgetStyle {
     clickedSlateSound {resourceObject=null}
     disabled {
      bIsDynamicallyLoaded=false
      drawAs=RoundedBox
      imageType=NoImage
      mirroring=NoMirror
      resourceName=None
      resourceObject=null
      tiling=NoTile
      imageSize=<0 0>
      margin=<ltrb 0 0 0 0>
      outlineSettings {
       bUseBrushTransparency=false
       roundingType=FixedRadius
       width=1
       color {colorUseRule=UseColor_ColorTable, specifiedColor=<rgba 0 0 0 0>}
       cornerRadii=<4 4 4 4>
      }
      tintColor {colorUseRule=UseColor_Specified, specifiedColor=<rgba 0.0395 0.0395 0.0395 1>}
      uVRegion {bIsValid=false, max=<0 0>, min=<0 0>}
     }
     disabledForeground {colorUseRule=UseColor_ColorTable, specifiedColor=<rgba 0 0 0 0>}
     hovered {
      bIsDynamicallyLoaded=false
      drawAs=RoundedBox
      imageType=NoImage
      mirroring=NoMirror
      resourceName=None
      resourceObject=null
      tiling=NoTile
      imageSize=<32 32>
      margin=<ltrb 0 0 0 0>
      outlineSettings {
       bUseBrushTransparency=false
       roundingType=FixedRadius
       width=1
       color {colorUseRule=UseColor_ColorTable, specifiedColor=<rgba 0 0 0 0>}
       cornerRadii=<4 4 4 4>
      }
      tintColor {colorUseRule=UseColor_Specified, specifiedColor=<rgba 0.0685 0.2542 0.807 1>}
      uVRegion {bIsValid=false, max=<0 0>, min=<0 0>}
     }
     hoveredForeground {colorUseRule=UseColor_ColorTable, specifiedColor=<rgba 0 0 0 0>}
     hoveredSlateSound {resourceObject=null}
     normal {
      bIsDynamicallyLoaded=false
      drawAs=RoundedBox
      imageType=NoImage
      mirroring=NoMirror
      resourceName=None
      resourceObject=null
      tiling=NoTile
      imageSize=<32 32>
      margin=<ltrb 0 0 0 0>
      outlineSettings {
       bUseBrushTransparency=false
       roundingType=FixedRadius
       width=1
       color {colorUseRule=UseColor_ColorTable, specifiedColor=<rgba 0 0 0 0>}
       cornerRadii=<4 4 4 4>
      }
      tintColor {colorUseRule=UseColor_Specified, specifiedColor=<rgba 0.0395 0.0395 0.0395 1>}
      uVRegion {bIsValid=false, max=<0 0>, min=<0 0>}
     }
     normalForeground {colorUseRule=UseColor_ColorTable, specifiedColor=<rgba 0 0 0 0>}
     normalPadding=<ltrb 12 1.5 12 1.5>
     pressed {
      bIsDynamicallyLoaded=false
      drawAs=RoundedBox
      imageType=NoImage
      mirroring=NoMirror
      resourceName=None
      resourceObject=null
      tiling=NoTile
      imageSize=<32 32>
      margin=<ltrb 0 0 0 0>
      outlineSettings {
       bUseBrushTransparency=false
       roundingType=FixedRadius
       width=1
       color {colorUseRule=UseColor_ColorTable, specifiedColor=<rgba 0 0 0 0>}
       cornerRadii=<4 4 4 4>
      }
      tintColor {colorUseRule=UseColor_Specified, specifiedColor=<rgba 0.0284 0.0284 0.0284 1>}
      uVRegion {bIsValid=false, max=<0 0>, min=<0 0>}
     }
     pressedForeground {colorUseRule=UseColor_ColorTable, specifiedColor=<rgba 0 0 0 0>}
     pressedPadding=<ltrb 12 2.5 12 0.5>
     pressedSlateSound {resourceObject=null}
    }
   }
   slot {class=HorizontalBoxSlot, properties {verticalAlignment=VAlign_Center, padding=<ltrb 8 0 8 0>}}
  }
  Text_ClearAllSelected {
   widgetClassPath=&UEFN_TextBlock
   parent=Button_ClearAllSelected
   widget=Text_ClearAllSelected
   properties {
    justification=Center
    text=Clear All Selected
    colorAndOpacity {colorUseRule=UseColor_Specified, specifiedColor=<rgba 0.7913 0.7991 0.8228 1>}
    font {
     bForceMonospaced=false
     bMaterialIsStencil=false
     fontMaterial=null
     fontObject=&BurbankSmall
     letterSpacing=0
     monospacedWidth=1
     size=9
     skewAmount=0
     typefaceFontName=Regular
     outlineSettings {
      bApplyOutlineToDropShadows=false
      bMiteredCorners=false
      bSeparateFillAlpha=false
      outlineMaterial=null
      outlineSize=0
      outlineColor=<rgba 0 0 0 1>
     }
    }
    shadowColorAndOpacity=<rgba 0 0 0 1>
   }
   slot {class=ButtonSlot}
  }
  Button_SetAllSelected {
   widgetClassPath=&EditorUtilityButton
   parent=HB_SelectedBar
   widget=Button_SetAllSelected
   properties {
    bIsEnabled=false
    backgroundColor=<rgba 0.0423 0.0497 0.063 1>
    widgetStyle {
     clickedSlateSound {resourceObject=null}
     disabled {
      bIsDynamicallyLoaded=false
      drawAs=RoundedBox
      imageType=NoImage
      mirroring=NoMirror
      resourceName=None
      resourceObject=null
      tiling=NoTile
      imageSize=<0 0>
      margin=<ltrb 0 0 0 0>
      outlineSettings {
       bUseBrushTransparency=false
       roundingType=FixedRadius
       width=1
       color {colorUseRule=UseColor_ColorTable, specifiedColor=<rgba 0 0 0 0>}
       cornerRadii=<4 4 4 4>
      }
      tintColor {colorUseRule=UseColor_Specified, specifiedColor=<rgba 0.0395 0.0395 0.0395 1>}
      uVRegion {bIsValid=false, max=<0 0>, min=<0 0>}
     }
     disabledForeground {colorUseRule=UseColor_ColorTable, specifiedColor=<rgba 0 0 0 0>}
     hovered {
      bIsDynamicallyLoaded=false
      drawAs=RoundedBox
      imageType=NoImage
      mirroring=NoMirror
      resourceName=None
      resourceObject=null
      tiling=NoTile
      imageSize=<32 32>
      margin=<ltrb 0 0 0 0>
      outlineSettings {
       bUseBrushTransparency=false
       roundingType=FixedRadius
       width=1
       color {colorUseRule=UseColor_ColorTable, specifiedColor=<rgba 0 0 0 0>}
       cornerRadii=<4 4 4 4>
      }
      tintColor {colorUseRule=UseColor_Specified, specifiedColor=<rgba 0.0685 0.2542 0.807 1>}
      uVRegion {bIsValid=false, max=<0 0>, min=<0 0>}
     }
     hoveredForeground {colorUseRule=UseColor_ColorTable, specifiedColor=<rgba 0 0 0 0>}
     hoveredSlateSound {resourceObject=null}
     normal {
      bIsDynamicallyLoaded=false
      drawAs=RoundedBox
      imageType=NoImage
      mirroring=NoMirror
      resourceName=None
      resourceObject=null
      tiling=NoTile
      imageSize=<32 32>
      margin=<ltrb 0 0 0 0>
      outlineSettings {
       bUseBrushTransparency=false
       roundingType=FixedRadius
       width=1
       color {colorUseRule=UseColor_ColorTable, specifiedColor=<rgba 0 0 0 0>}
       cornerRadii=<4 4 4 4>
      }
      tintColor {colorUseRule=UseColor_Specified, specifiedColor=<rgba 0.0395 0.0395 0.0395 1>}
      uVRegion {bIsValid=false, max=<0 0>, min=<0 0>}
     }
     normalForeground {colorUseRule=UseColor_ColorTable, specifiedColor=<rgba 0 0 0 0>}
     normalPadding=<ltrb 12 1.5 12 1.5>
     pressed {
      bIsDynamicallyLoaded=false
      drawAs=RoundedBox
      imageType=NoImage
      mirroring=NoMirror
      resourceName=None
      resourceObject=null
      tiling=NoTile
      imageSize=<32 32>
      margin=<ltrb 0 0 0 0>
      outlineSettings {
       bUseBrushTransparency=false
       roundingType=FixedRadius
       width=1
       color {colorUseRule=UseColor_ColorTable, specifiedColor=<rgba 0 0 0 0>}
       cornerRadii=<4 4 4 4>
      }
      tintColor {colorUseRule=UseColor_Specified, specifiedColor=<rgba 0.0284 0.0284 0.0284 1>}
      uVRegion {bIsValid=false, max=<0 0>, min=<0 0>}
     }
     pressedForeground {colorUseRule=UseColor_ColorTable, specifiedColor=<rgba 0 0 0 0>}
     pressedPadding=<ltrb 12 2.5 12 0.5>
     pressedSlateSound {resourceObject=null}
    }
   }
   slot {class=HorizontalBoxSlot, properties {verticalAlignment=VAlign_Center, padding=<ltrb 8 0 8 0>}}
  }
  Text_SetAllSelected {
   widgetClassPath=&UEFN_TextBlock
   parent=Button_SetAllSelected
   widget=Text_SetAllSelected
   properties {
    justification=Center
    text=Set All Selected
    colorAndOpacity {colorUseRule=UseColor_Specified, specifiedColor=<rgba 0.7913 0.7991 0.8228 1>}
    font {
     bForceMonospaced=false
     bMaterialIsStencil=false
     fontMaterial=null
     fontObject=&BurbankSmall
     letterSpacing=0
     monospacedWidth=1
     size=9
     skewAmount=0
     typefaceFontName=Regular
     outlineSettings {
      bApplyOutlineToDropShadows=false
      bMiteredCorners=false
      bSeparateFillAlpha=false
      outlineMaterial=null
      outlineSize=0
      outlineColor=<rgba 0 0 0 1>
     }
    }
    shadowColorAndOpacity=<rgba 0 0 0 1>
   }
   slot {class=ButtonSlot}
  }
  Spacer_ToolbarFill_1 {
   widgetClassPath=&Spacer
   parent=HB_SelectedBar
   widget=Spacer_ToolbarFill_1
   slot {class=HorizontalBoxSlot, properties {padding=<ltrb 12 12 12 12>, size {sizeRule=Fill, value=1}}}
  }
  Button_Validate {
   widgetClassPath=&EditorUtilityButton
   parent=HB_SelectedBar
   widget=Button_Validate
   properties {
    toolTipText=Reset the source reference back to default asset.
    backgroundColor=<rgba 0.0423 0.0497 0.063 1>
    widgetStyle {
     clickedSlateSound {resourceObject=null}
     disabled {
      bIsDynamicallyLoaded=false
      drawAs=RoundedBox
      imageType=NoImage
      mirroring=NoMirror
      resourceName=None
      resourceObject=null
      tiling=NoTile
      imageSize=<0 0>
      margin=<ltrb 0 0 0 0>
      outlineSettings {
       bUseBrushTransparency=false
       roundingType=FixedRadius
       width=1
       color {colorUseRule=UseColor_ColorTable, specifiedColor=<rgba 0 0 0 0>}
       cornerRadii=<4 4 4 4>
      }
      tintColor {colorUseRule=UseColor_Specified, specifiedColor=<rgba 0.0395 0.0395 0.0395 1>}
      uVRegion {bIsValid=false, max=<0 0>, min=<0 0>}
     }
     disabledForeground {colorUseRule=UseColor_ColorTable, specifiedColor=<rgba 0 0 0 0>}
     hovered {
      bIsDynamicallyLoaded=false
      drawAs=RoundedBox
      imageType=NoImage
      mirroring=NoMirror
      resourceName=None
      resourceObject=null
      tiling=NoTile
      imageSize=<32 32>
      margin=<ltrb 0 0 0 0>
      outlineSettings {
       bUseBrushTransparency=false
       roundingType=FixedRadius
       width=1
       color {colorUseRule=UseColor_ColorTable, specifiedColor=<rgba 0 0 0 0>}
       cornerRadii=<4 4 4 4>
      }
      tintColor {colorUseRule=UseColor_Specified, specifiedColor=<rgba 0.0685 0.2542 0.807 1>}
      uVRegion {bIsValid=false, max=<0 0>, min=<0 0>}
     }
     hoveredForeground {colorUseRule=UseColor_ColorTable, specifiedColor=<rgba 0 0 0 0>}
     hoveredSlateSound {resourceObject=null}
     normal {
      bIsDynamicallyLoaded=false
      drawAs=RoundedBox
      imageType=NoImage
      mirroring=NoMirror
      resourceName=None
      resourceObject=null
      tiling=NoTile
      imageSize=<32 32>
      margin=<ltrb 0 0 0 0>
      outlineSettings {
       bUseBrushTransparency=false
       roundingType=FixedRadius
       width=1
       color {colorUseRule=UseColor_ColorTable, specifiedColor=<rgba 0 0 0 0>}
       cornerRadii=<4 4 4 4>
      }
      tintColor {colorUseRule=UseColor_Specified, specifiedColor=<rgba 0.0395 0.0395 0.0395 1>}
      uVRegion {bIsValid=false, max=<0 0>, min=<0 0>}
     }
     normalForeground {colorUseRule=UseColor_ColorTable, specifiedColor=<rgba 0 0 0 0>}
     normalPadding=<ltrb 12 1.5 12 1.5>
     pressed {
      bIsDynamicallyLoaded=false
      drawAs=RoundedBox
      imageType=NoImage
      mirroring=NoMirror
      resourceName=None
      resourceObject=null
      tiling=NoTile
      imageSize=<32 32>
      margin=<ltrb 0 0 0 0>
      outlineSettings {
       bUseBrushTransparency=false
       roundingType=FixedRadius
       width=1
       color {colorUseRule=UseColor_ColorTable, specifiedColor=<rgba 0 0 0 0>}
       cornerRadii=<4 4 4 4>
      }
      tintColor {colorUseRule=UseColor_Specified, specifiedColor=<rgba 0.0284 0.0284 0.0284 1>}
      uVRegion {bIsValid=false, max=<0 0>, min=<0 0>}
     }
     pressedForeground {colorUseRule=UseColor_ColorTable, specifiedColor=<rgba 0 0 0 0>}
     pressedPadding=<ltrb 12 2.5 12 0.5>
     pressedSlateSound {resourceObject=null}
    }
   }
   slot {class=HorizontalBoxSlot, properties {verticalAlignment=VAlign_Center}}
  }
  Text_Validate {
   widgetClassPath=&UEFN_TextBlock
   parent=Button_Validate
   widget=Text_Validate
   properties {
    justification=Center
    text=Validate
    colorAndOpacity {colorUseRule=UseColor_Specified, specifiedColor=<rgba 0.7913 0.7991 0.8228 1>}
    font {
     bForceMonospaced=false
     bMaterialIsStencil=false
     fontMaterial=null
     fontObject=&BurbankSmall
     letterSpacing=0
     monospacedWidth=1
     size=9
     skewAmount=0
     typefaceFontName=Regular
     outlineSettings {
      bApplyOutlineToDropShadows=false
      bMiteredCorners=false
      bSeparateFillAlpha=false
      outlineMaterial=null
      outlineSize=0
      outlineColor=<rgba 0 0 0 1>
     }
    }
    shadowColorAndOpacity=<rgba 0 0 0 1>
   }
   slot {class=ButtonSlot}
  }
  Text_ToolbarStatus {
   widgetClassPath=&UEFN_TextBlock
   bIsVariable=true
   parent=VB_Toolbar
   widget=Text_ToolbarStatus
   properties {
    justification=Center
    text="Load sequencer sounds first, then pull replacement candidates from the Content Browser."
    colorAndOpacity {colorUseRule=UseColor_Specified, specifiedColor=<rgba 0.3968 0.4342 0.491 1>}
    font {
     bForceMonospaced=false
     bMaterialIsStencil=false
     fontMaterial=null
     fontObject=&BurbankSmall
     letterSpacing=0
     monospacedWidth=1
     size=9.75
     skewAmount=0
     typefaceFontName=Regular
     outlineSettings {
      bApplyOutlineToDropShadows=false
      bMiteredCorners=false
      bSeparateFillAlpha=false
      outlineMaterial=null
      outlineSize=0
      outlineColor=<rgba 0 0 0 1>
     }
    }
    shadowColorAndOpacity=<rgba 0 0 0 1>
   }
   slot {class=VerticalBoxSlot, properties {padding=<ltrb 0 8 0 8>, size {sizeRule=Fill, value=1}}}
  }
  HB_ToolbarActions {
   widgetClassPath=&HorizontalBox
   parent=VB_Toolbar
   widget=HB_ToolbarActions
   slot {class=VerticalBoxSlot, properties {horizontalAlignment=HAlign_Center, verticalAlignment=VAlign_Center}}
  }
  HorizontalBox_641 {
   widgetClassPath=&HorizontalBox
   parent=HB_ToolbarActions
   widget=HorizontalBox_641
   slot {class=HorizontalBoxSlot, properties {verticalAlignment=VAlign_Center, padding=<ltrb 0 2 8 2>}}
  }
  CheckBox_ShowUnmatchedOnly {
   widgetClassPath=&EditorUtilityCheckBox
   bIsVariable=true
   parent=HorizontalBox_641
   widget=CheckBox_ShowUnmatchedOnly
   properties {toolTipText=Hide rows that already have replacement sounds assigned}
   slot {class=HorizontalBoxSlot}
  }
  Text_ShowUnmatchedOnly {
   widgetClassPath=&UEFN_TextBlock
   parent=HorizontalBox_641
   widget=Text_ShowUnmatchedOnly
   properties {
    justification=Center
    text=Show Unmatched Only
    toolTipText=Hide rows that already have replacement sounds assigned
    colorAndOpacity {colorUseRule=UseColor_Specified, specifiedColor=<rgba 0.3968 0.4342 0.491 1>}
    font {
     bForceMonospaced=false
     bMaterialIsStencil=false
     fontMaterial=null
     fontObject=&BurbankSmall
     letterSpacing=0
     monospacedWidth=1
     size=9
     skewAmount=0
     typefaceFontName=Regular
     outlineSettings {
      bApplyOutlineToDropShadows=false
      bMiteredCorners=false
      bSeparateFillAlpha=false
      outlineMaterial=null
      outlineSize=0
      outlineColor=<rgba 0 0 0 1>
     }
    }
    shadowColorAndOpacity=<rgba 0 0 0 1>
   }
   slot {class=HorizontalBoxSlot}
  }
  HorizontalBox {
   widgetClassPath=&HorizontalBox
   parent=HB_ToolbarActions
   widget=HorizontalBox
   slot {class=HorizontalBoxSlot, properties {padding=<ltrb 8 0 8 0>}}
  }
  Text_SortBy {
   widgetClassPath=&UEFN_TextBlock
   parent=HorizontalBox
   widget=Text_SortBy
   properties {
    justification=Center
    text=Sort By:
    toolTipText=Sort rows by the selected criteria
    colorAndOpacity {colorUseRule=UseColor_Specified, specifiedColor=<rgba 0.3968 0.4342 0.491 1>}
    font {
     bForceMonospaced=false
     bMaterialIsStencil=false
     fontMaterial=null
     fontObject=&BurbankSmall
     letterSpacing=0
     monospacedWidth=1
     size=9
     skewAmount=0
     typefaceFontName=Regular
     outlineSettings {
      bApplyOutlineToDropShadows=false
      bMiteredCorners=false
      bSeparateFillAlpha=false
      outlineMaterial=null
      outlineSize=0
      outlineColor=<rgba 0 0 0 1>
     }
    }
    shadowColorAndOpacity=<rgba 0 0 0 1>
   }
   slot {class=HorizontalBoxSlot, properties {verticalAlignment=VAlign_Center, padding=<ltrb 0 2 1 2>}}
  }
  SizeBox {
   widgetClassPath=&SizeBox
   parent=HorizontalBox
   widget=SizeBox
   properties {bOverride_WidthOverride=true, widthOverride=150}
   slot {class=HorizontalBoxSlot}
  }
  ComboBox_SortBy {
   widgetClassPath=/Script/Blutility.EditorUtilityComboBoxString
   bIsVariable=true
   parent=SizeBox
   widget=ComboBox_SortBy
   properties {
    selectedOption=Row Index
    toolTipText=Sort rows by the selected criteria
    defaultOptions [Row Index, Track Name, Source Sound Name, "Status (Matched First)", "Status (Unmatched First)"]
    font {
     bForceMonospaced=false
     bMaterialIsStencil=false
     fontMaterial=null
     fontObject=&BurbankSmall
     letterSpacing=0
     monospacedWidth=1
     size=9
     skewAmount=0
     typefaceFontName=Regular
     outlineSettings {
      bApplyOutlineToDropShadows=false
      bMiteredCorners=false
      bSeparateFillAlpha=false
      outlineMaterial=null
      outlineSize=0
      outlineColor=<rgba 0 0 0 1>
     }
    }
    foregroundColor {colorUseRule=UseColor_Specified, specifiedColor=<rgba 0.7913 0.7991 0.8228 1>}
   }
   slot {class=SizeBoxSlot}
  }
  HorizontalBox_0 {
   widgetClassPath=&HorizontalBox
   parent=HB_ToolbarActions
   widget=HorizontalBox_0
   slot {class=HorizontalBoxSlot, properties {padding=<ltrb 8 0 8 0>}}
  }
  Text_ShowUnmatchedOnly_1 {
   widgetClassPath=&UEFN_TextBlock
   parent=HorizontalBox_0
   widget=Text_ShowUnmatchedOnly_1
   properties {
    justification=Center
    text=Track Filtering:
    toolTipText=Filter rows by audio track and row index
    colorAndOpacity {colorUseRule=UseColor_Specified, specifiedColor=<rgba 0.3968 0.4342 0.491 1>}
    font {
     bForceMonospaced=false
     bMaterialIsStencil=false
     fontMaterial=null
     fontObject=&BurbankSmall
     letterSpacing=0
     monospacedWidth=1
     size=9
     skewAmount=0
     typefaceFontName=Regular
     outlineSettings {
      bApplyOutlineToDropShadows=false
      bMiteredCorners=false
      bSeparateFillAlpha=false
      outlineMaterial=null
      outlineSize=0
      outlineColor=<rgba 0 0 0 1>
     }
    }
    shadowColorAndOpacity=<rgba 0 0 0 1>
   }
   slot {class=HorizontalBoxSlot, properties {verticalAlignment=VAlign_Center, padding=<ltrb 0 2 1 2>}}
  }
  SizeBox_0 {
   widgetClassPath=&SizeBox
   parent=HorizontalBox_0
   widget=SizeBox_0
   properties {bOverride_WidthOverride=true, widthOverride=150}
   slot {class=HorizontalBoxSlot}
  }
  ComboBox_TrackFilter {
   widgetClassPath=&EditorUtilityComboBoxString
   bIsVariable=true
   parent=SizeBox_0
   widget=ComboBox_TrackFilter
   properties {
    toolTipText=Filter rows by audio track and row index
    font {
     bForceMonospaced=false
     bMaterialIsStencil=false
     fontMaterial=null
     fontObject=&BurbankSmall
     letterSpacing=0
     monospacedWidth=1
     size=9
     skewAmount=0
     typefaceFontName=Regular
     outlineSettings {
      bApplyOutlineToDropShadows=false
      bMiteredCorners=false
      bSeparateFillAlpha=false
      outlineMaterial=null
      outlineSize=0
      outlineColor=<rgba 0 0 0 1>
     }
    }
    foregroundColor {colorUseRule=UseColor_Specified, specifiedColor=<rgba 0.7913 0.7991 0.8228 1>}
   }
   slot {class=SizeBoxSlot}
  }
  HorizontalBox_1 {
   widgetClassPath=&HorizontalBox
   parent=HB_ToolbarActions
   widget=HorizontalBox_1
   slot {class=HorizontalBoxSlot, properties {padding=<ltrb 8 0 0 0>}}
  }
  Text_SearchFilter {
   widgetClassPath=&UEFN_TextBlock
   parent=HorizontalBox_1
   widget=Text_SearchFilter
   properties {
    justification=Center
    text=Filter:
    colorAndOpacity {colorUseRule=UseColor_Specified, specifiedColor=<rgba 0.3968 0.4342 0.491 1>}
    font {
     bForceMonospaced=false
     bMaterialIsStencil=false
     fontMaterial=null
     fontObject=&BurbankSmall
     letterSpacing=0
     monospacedWidth=1
     size=9
     skewAmount=0
     typefaceFontName=Regular
     outlineSettings {
      bApplyOutlineToDropShadows=false
      bMiteredCorners=false
      bSeparateFillAlpha=false
      outlineMaterial=null
      outlineSize=0
      outlineColor=<rgba 0 0 0 1>
     }
    }
    shadowColorAndOpacity=<rgba 0 0 0 1>
   }
   slot {class=HorizontalBoxSlot, properties {horizontalAlignment=HAlign_Center, verticalAlignment=VAlign_Center, padding=<ltrb 0 2 1 2>}}
  }
  SizeBox_3 {
   widgetClassPath=&SizeBox
   parent=HorizontalBox_1
   widget=SizeBox_3
   properties {bOverride_MaxDesiredWidth=true, maxDesiredWidth=150}
   slot {class=HorizontalBoxSlot}
  }
  TextBox_Search {
   widgetClassPath=&EditorUtilityEditableTextBox
   bIsVariable=true
   parent=SizeBox_3
   widget=TextBox_Search
   properties {hintText=Search sounds..., minimumDesiredWidth=150}
   slot {class=SizeBoxSlot, properties {padding=<ltrb 0 1 0 1>}}
  }
  Button_SeachFilterClear {
   widgetClassPath=&EditorUtilityButton
   parent=HorizontalBox_1
   widget=Button_SeachFilterClear
   properties {
    toolTipText=Reset the source reference back to default asset.
    backgroundColor=<rgba 0.0423 0.0497 0.063 1>
    widgetStyle {
     clickedSlateSound {resourceObject=null}
     disabled {
      bIsDynamicallyLoaded=false
      drawAs=RoundedBox
      imageType=NoImage
      mirroring=NoMirror
      resourceName=None
      resourceObject=null
      tiling=NoTile
      imageSize=<0 0>
      margin=<ltrb 0 0 0 0>
      outlineSettings {
       bUseBrushTransparency=false
       roundingType=FixedRadius
       width=1
       color {colorUseRule=UseColor_ColorTable, specifiedColor=<rgba 0 0 0 0>}
       cornerRadii=<4 4 4 4>
      }
      tintColor {colorUseRule=UseColor_Specified, specifiedColor=<rgba 0.0395 0.0395 0.0395 1>}
      uVRegion {bIsValid=false, max=<0 0>, min=<0 0>}
     }
     disabledForeground {colorUseRule=UseColor_ColorTable, specifiedColor=<rgba 0 0 0 0>}
     hovered {
      bIsDynamicallyLoaded=false
      drawAs=RoundedBox
      imageType=NoImage
      mirroring=NoMirror
      resourceName=None
      resourceObject=null
      tiling=NoTile
      imageSize=<32 32>
      margin=<ltrb 0 0 0 0>
      outlineSettings {
       bUseBrushTransparency=false
       roundingType=FixedRadius
       width=1
       color {colorUseRule=UseColor_ColorTable, specifiedColor=<rgba 0 0 0 0>}
       cornerRadii=<4 4 4 4>
      }
      tintColor {colorUseRule=UseColor_Specified, specifiedColor=<rgba 0.0685 0.2542 0.807 1>}
      uVRegion {bIsValid=false, max=<0 0>, min=<0 0>}
     }
     hoveredForeground {colorUseRule=UseColor_ColorTable, specifiedColor=<rgba 0 0 0 0>}
     hoveredSlateSound {resourceObject=null}
     normal {
      bIsDynamicallyLoaded=false
      drawAs=RoundedBox
      imageType=NoImage
      mirroring=NoMirror
      resourceName=None
      resourceObject=null
      tiling=NoTile
      imageSize=<32 32>
      margin=<ltrb 0 0 0 0>
      outlineSettings {
       bUseBrushTransparency=false
       roundingType=FixedRadius
       width=1
       color {colorUseRule=UseColor_ColorTable, specifiedColor=<rgba 0 0 0 0>}
       cornerRadii=<4 4 4 4>
      }
      tintColor {colorUseRule=UseColor_Specified, specifiedColor=<rgba 0.0395 0.0395 0.0395 1>}
      uVRegion {bIsValid=false, max=<0 0>, min=<0 0>}
     }
     normalForeground {colorUseRule=UseColor_ColorTable, specifiedColor=<rgba 0 0 0 0>}
     normalPadding=<ltrb 12 1.5 12 1.5>
     pressed {
      bIsDynamicallyLoaded=false
      drawAs=RoundedBox
      imageType=NoImage
      mirroring=NoMirror
      resourceName=None
      resourceObject=null
      tiling=NoTile
      imageSize=<32 32>
      margin=<ltrb 0 0 0 0>
      outlineSettings {
       bUseBrushTransparency=false
       roundingType=FixedRadius
       width=1
       color {colorUseRule=UseColor_ColorTable, specifiedColor=<rgba 0 0 0 0>}
       cornerRadii=<4 4 4 4>
      }
      tintColor {colorUseRule=UseColor_Specified, specifiedColor=<rgba 0.0284 0.0284 0.0284 1>}
      uVRegion {bIsValid=false, max=<0 0>, min=<0 0>}
     }
     pressedForeground {colorUseRule=UseColor_ColorTable, specifiedColor=<rgba 0 0 0 0>}
     pressedPadding=<ltrb 12 2.5 12 0.5>
     pressedSlateSound {resourceObject=null}
    }
   }
   slot {class=HorizontalBoxSlot, properties {horizontalAlignment=HAlign_Left, verticalAlignment=VAlign_Center}}
  }
  Text_Validate_1 {
   widgetClassPath=&UEFN_TextBlock
   parent=Button_SeachFilterClear
   widget=Text_Validate_1
   properties {
    justification=Center
    text=Clear
    colorAndOpacity {colorUseRule=UseColor_Specified, specifiedColor=<rgba 0.7913 0.7991 0.8228 1>}
    font {
     bForceMonospaced=false
     bMaterialIsStencil=false
     fontMaterial=null
     fontObject=&BurbankSmall
     letterSpacing=0
     monospacedWidth=1
     size=9
     skewAmount=0
     typefaceFontName=Regular
     outlineSettings {
      bApplyOutlineToDropShadows=false
      bMiteredCorners=false
      bSeparateFillAlpha=false
      outlineMaterial=null
      outlineSize=0
      outlineColor=<rgba 0 0 0 1>
     }
    }
    shadowColorAndOpacity=<rgba 0 0 0 1>
   }
   slot {class=ButtonSlot}
  }
  Spacer_ToolbarFill {
   widgetClassPath=&Spacer
   parent=HB_ToolbarActions
   widget=Spacer_ToolbarFill
   slot {class=HorizontalBoxSlot, properties {padding=<ltrb 12 12 12 12>, size {sizeRule=Fill, value=1}}}
  }
  Border_1 {
   widgetClassPath=&Border
   parent=VerticalBox_Root
   widget=Border_1
   properties {brushColor=<rgba 0 0 0 1>, padding=<ltrb 0 0 0 0>}
   slot {class=VerticalBoxSlot, properties {horizontalAlignment=HAlign_Center, size {sizeRule=Fill, value=1}}}
  }
  Border_Table {
   widgetClassPath=&Border
   parent=Border_1
   widget=Border_Table
   properties {horizontalAlignment=HAlign_Center, brushColor=<rgba 0.0185 0.0203 0.0242 1>, padding=<ltrb 4 12 4 2>}
   slot {class=BorderSlot, properties {padding=<ltrb 0 0 0 0>}}
  }
  SizeBox_1 {
   widgetClassPath=&SizeBox
   parent=Border_Table
   widget=SizeBox_1
   properties {bOverride_HeightOverride=true, bOverride_WidthOverride=true, heightOverride=500, widthOverride=1200}
   slot {class=BorderSlot, properties {horizontalAlignment=HAlign_Center, padding=<ltrb 4 12 4 2>}}
  }
  VB_Table {widgetClassPath=&VerticalBox, parent=SizeBox_1, widget=VB_Table, slot {class=SizeBoxSlot}}
  Border_TableHeader {
   widgetClassPath=&Border
   parent=VB_Table
   widget=Border_TableHeader
   properties {brushColor=<rgba 0.0252 0.0284 0.0369 1>, padding=<ltrb 8 8 8 8>}
   slot {class=VerticalBoxSlot}
  }
  HB_TableHeaderRow {
   widgetClassPath=&HorizontalBox
   parent=Border_TableHeader
   widget=HB_TableHeaderRow
   slot {class=BorderSlot, properties {padding=<ltrb 8 8 8 8>}}
  }
  Text_HeaderIndex {
   widgetClassPath=&UEFN_TextBlock
   parent=HB_TableHeaderRow
   widget=Text_HeaderIndex
   properties {
    justification=Center
    text="##"
    colorAndOpacity {colorUseRule=UseColor_Specified, specifiedColor=<rgba 0.7913 0.7991 0.8228 1>}
    font {
     bForceMonospaced=false
     bMaterialIsStencil=false
     fontMaterial=null
     fontObject=&BrutalType
     letterSpacing=0
     monospacedWidth=1
     size=9
     skewAmount=0
     typefaceFontName=Bold
     outlineSettings {
      bApplyOutlineToDropShadows=false
      bMiteredCorners=false
      bSeparateFillAlpha=false
      outlineMaterial=null
      outlineSize=0
      outlineColor=<rgba 0 0 0 1>
     }
    }
    shadowColorAndOpacity=<rgba 0 0 0 1>
   }
   slot {class=HorizontalBoxSlot, properties {horizontalAlignment=HAlign_Center, verticalAlignment=VAlign_Center, padding=<ltrb 8 0 12 0>}}
  }
  Border_RowSource {
   widgetClassPath=&Border
   parent=HB_TableHeaderRow
   widget=Border_RowSource
   properties {brushColor=<rgba 0.0232 0.0262 0.0331 1>, padding=<ltrb 0 0 12 0>}
   slot {class=HorizontalBoxSlot, properties {horizontalAlignment=HAlign_Center, verticalAlignment=VAlign_Center, size {sizeRule=Fill, value=1}}}
  }
  Text_HeaderSource {
   widgetClassPath=&UEFN_TextBlock
   parent=Border_RowSource
   widget=Text_HeaderSource
   properties {
    justification=Center
    text=Sequencer Sound
    colorAndOpacity {colorUseRule=UseColor_Specified, specifiedColor=<rgba 0.7913 0.7991 0.8228 1>}
    font {
     bForceMonospaced=false
     bMaterialIsStencil=false
     fontMaterial=null
     fontObject=&BrutalType
     letterSpacing=0
     monospacedWidth=1
     size=9
     skewAmount=0
     typefaceFontName=Bold
     outlineSettings {
      bApplyOutlineToDropShadows=false
      bMiteredCorners=false
      bSeparateFillAlpha=false
      outlineMaterial=null
      outlineSize=0
      outlineColor=<rgba 0 0 0 1>
     }
    }
    shadowColorAndOpacity=<rgba 0 0 0 1>
   }
   slot {class=BorderSlot, properties {padding=<ltrb 0 0 12 0>}}
  }
  Text_HeaderArrow {
   widgetClassPath=&UEFN_TextBlock
   parent=HB_TableHeaderRow
   widget=Text_HeaderArrow
   properties {
    justification=Center
    text="->"
    colorAndOpacity {colorUseRule=UseColor_Specified, specifiedColor=<rgba 0.7913 0.7991 0.8228 1>}
    font {
     bForceMonospaced=false
     bMaterialIsStencil=false
     fontMaterial=null
     fontObject=&BrutalType
     letterSpacing=0
     monospacedWidth=1
     size=9
     skewAmount=0
     typefaceFontName=Bold
     outlineSettings {
      bApplyOutlineToDropShadows=false
      bMiteredCorners=false
      bSeparateFillAlpha=false
      outlineMaterial=null
      outlineSize=0
      outlineColor=<rgba 0 0 0 1>
     }
    }
    shadowColorAndOpacity=<rgba 0 0 0 1>
   }
   slot {class=HorizontalBoxSlot, properties {horizontalAlignment=HAlign_Center, verticalAlignment=VAlign_Center, padding=<ltrb 0 0 16 0>}}
  }
  Border_RowReplacement {
   widgetClassPath=&Border
   parent=HB_TableHeaderRow
   widget=Border_RowReplacement
   properties {brushColor=<rgba 0.0232 0.0262 0.0331 1>, padding=<ltrb 0 0 12 0>}
   slot {class=HorizontalBoxSlot, properties {horizontalAlignment=HAlign_Center, verticalAlignment=VAlign_Center, size {sizeRule=Fill, value=1}}}
  }
  Text_HeaderReplacement {
   widgetClassPath=&UEFN_TextBlock
   parent=Border_RowReplacement
   widget=Text_HeaderReplacement
   properties {
    justification=Center
    text=Replacement Sound
    colorAndOpacity {colorUseRule=UseColor_Specified, specifiedColor=<rgba 0.7913 0.7991 0.8228 1>}
    font {
     bForceMonospaced=false
     bMaterialIsStencil=false
     fontMaterial=null
     fontObject=&BrutalType
     letterSpacing=0
     monospacedWidth=1
     size=9
     skewAmount=0
     typefaceFontName=Bold
     outlineSettings {
      bApplyOutlineToDropShadows=false
      bMiteredCorners=false
      bSeparateFillAlpha=false
      outlineMaterial=null
      outlineSize=0
      outlineColor=<rgba 0 0 0 1>
     }
    }
    shadowColorAndOpacity=<rgba 0 0 0 1>
   }
   slot {class=BorderSlot, properties {padding=<ltrb 0 0 12 0>}}
  }
  Text_HeaderStatus {
   widgetClassPath=&UEFN_TextBlock
   parent=HB_TableHeaderRow
   widget=Text_HeaderStatus
   properties {
    justification=Center
    text=Status
    colorAndOpacity {colorUseRule=UseColor_Specified, specifiedColor=<rgba 0.7913 0.7991 0.8228 1>}
    font {
     bForceMonospaced=false
     bMaterialIsStencil=false
     fontMaterial=null
     fontObject=&BrutalType
     letterSpacing=0
     monospacedWidth=1
     size=9
     skewAmount=0
     typefaceFontName=Bold
     outlineSettings {
      bApplyOutlineToDropShadows=false
      bMiteredCorners=false
      bSeparateFillAlpha=false
      outlineMaterial=null
      outlineSize=0
      outlineColor=<rgba 0 0 0 1>
     }
    }
    shadowColorAndOpacity=<rgba 0 0 0 1>
   }
   slot {class=HorizontalBoxSlot, properties {horizontalAlignment=HAlign_Center, verticalAlignment=VAlign_Center, padding=<ltrb 0 0 12 0>}}
  }
  ScrollBox_SoundRows {
   widgetClassPath=&EditorUtilityScrollBox
   bIsVariable=true
   parent=VB_Table
   widget=ScrollBox_SoundRows
   slot {class=VerticalBoxSlot, properties {size {sizeRule=Fill, value=1}}}
  }
  Border_Summary {
   widgetClassPath=&Border
   parent=VerticalBox_Root
   widget=Border_Summary
   properties {
    horizontalAlignment=HAlign_Center
    verticalAlignment=VAlign_Center
    brushColor=<rgba 0.0185 0.0203 0.0242 1>
    desiredSizeScale=<0 1>
    padding=<ltrb 0 10 0 10>
   }
   slot {class=VerticalBoxSlot}
  }
  Text_Summary {
   widgetClassPath=&UEFN_TextBlock
   bIsVariable=true
   parent=Border_Summary
   widget=Text_Summary
   properties {
    justification=Center
    text=0 total | 0 matched | 0 missing | 0 needs review
    colorAndOpacity {colorUseRule=UseColor_Specified, specifiedColor=<rgba 0.3968 0.4342 0.491 1>}
    font {
     bForceMonospaced=false
     bMaterialIsStencil=false
     fontMaterial=null
     fontObject=&BrutalType
     letterSpacing=0
     monospacedWidth=1
     size=9
     skewAmount=0
     typefaceFontName=Regular
     outlineSettings {
      bApplyOutlineToDropShadows=false
      bMiteredCorners=false
      bSeparateFillAlpha=false
      outlineMaterial=null
      outlineSize=0
      outlineColor=<rgba 0 0 0 1>
     }
    }
    shadowColorAndOpacity=<rgba 0 0 0 1>
   }
   slot {class=BorderSlot, properties {horizontalAlignment=HAlign_Center, verticalAlignment=VAlign_Center, padding=<ltrb 0 10 0 10>}}
  }
  Border_Footer {
   widgetClassPath=&Border
   parent=VerticalBox_Root
   widget=Border_Footer
   properties {horizontalAlignment=HAlign_Center, verticalAlignment=VAlign_Center, brushColor=<rgba 0.0423 0.0497 0.063 1>, padding=<ltrb 0 10 0 10>}
   slot {class=VerticalBoxSlot}
  }
  Button_ExecuteSwap {
   widgetClassPath=&EditorUtilityButton
   parent=Border_Footer
   widget=Button_ExecuteSwap
   properties {
    toolTipText=Apply all replacement sounds to the Level Sequence
    backgroundColor=<rgba 0.7011 0.0122 0 1>
    widgetStyle {
     clickedSlateSound {resourceObject=null}
     disabled {
      bIsDynamicallyLoaded=false
      drawAs=RoundedBox
      imageType=NoImage
      mirroring=NoMirror
      resourceName=None
      resourceObject=null
      tiling=NoTile
      imageSize=<0 0>
      margin=<ltrb 0 0 0 0>
      outlineSettings {
       bUseBrushTransparency=false
       roundingType=FixedRadius
       width=1
       color {colorUseRule=UseColor_ColorTable, specifiedColor=<rgba 0 0 0 0>}
       cornerRadii=<4 4 4 4>
      }
      tintColor {colorUseRule=UseColor_Specified, specifiedColor=<rgba 0.0395 0.0395 0.0395 1>}
      uVRegion {bIsValid=false, max=<0 0>, min=<0 0>}
     }
     disabledForeground {colorUseRule=UseColor_ColorTable, specifiedColor=<rgba 0 0 0 0>}
     hovered {
      bIsDynamicallyLoaded=false
      drawAs=RoundedBox
      imageType=NoImage
      mirroring=NoMirror
      resourceName=None
      resourceObject=null
      tiling=NoTile
      imageSize=<32 32>
      margin=<ltrb 0 0 0 0>
      outlineSettings {
       bUseBrushTransparency=false
       roundingType=FixedRadius
       width=1
       color {colorUseRule=UseColor_ColorTable, specifiedColor=<rgba 0 0 0 0>}
       cornerRadii=<4 4 4 4>
      }
      tintColor {colorUseRule=UseColor_Specified, specifiedColor=<rgba 0.0685 0.2542 0.807 1>}
      uVRegion {bIsValid=false, max=<0 0>, min=<0 0>}
     }
     hoveredForeground {colorUseRule=UseColor_ColorTable, specifiedColor=<rgba 0 0 0 0>}
     hoveredSlateSound {resourceObject=null}
     normal {
      bIsDynamicallyLoaded=false
      drawAs=RoundedBox
      imageType=NoImage
      mirroring=NoMirror
      resourceName=None
      resourceObject=null
      tiling=NoTile
      imageSize=<32 32>
      margin=<ltrb 0 0 0 0>
      outlineSettings {
       bUseBrushTransparency=false
       roundingType=FixedRadius
       width=1
       color {colorUseRule=UseColor_ColorTable, specifiedColor=<rgba 0 0 0 0>}
       cornerRadii=<4 4 4 4>
      }
      tintColor {colorUseRule=UseColor_Specified, specifiedColor=<rgba 0.0395 0.0395 0.0395 1>}
      uVRegion {bIsValid=false, max=<0 0>, min=<0 0>}
     }
     normalForeground {colorUseRule=UseColor_ColorTable, specifiedColor=<rgba 0 0 0 0>}
     normalPadding=<ltrb 12 1.5 12 1.5>
     pressed {
      bIsDynamicallyLoaded=false
      drawAs=RoundedBox
      imageType=NoImage
      mirroring=NoMirror
      resourceName=None
      resourceObject=null
      tiling=NoTile
      imageSize=<32 32>
      margin=<ltrb 0 0 0 0>
      outlineSettings {
       bUseBrushTransparency=false
       roundingType=FixedRadius
       width=1
       color {colorUseRule=UseColor_ColorTable, specifiedColor=<rgba 0 0 0 0>}
       cornerRadii=<4 4 4 4>
      }
      tintColor {colorUseRule=UseColor_Specified, specifiedColor=<rgba 0.0284 0.0284 0.0284 1>}
      uVRegion {bIsValid=false, max=<0 0>, min=<0 0>}
     }
     pressedForeground {colorUseRule=UseColor_ColorTable, specifiedColor=<rgba 0 0 0 0>}
     pressedPadding=<ltrb 12 2.5 12 0.5>
     pressedSlateSound {resourceObject=null}
    }
   }
   slot {class=BorderSlot, properties {horizontalAlignment=HAlign_Center, verticalAlignment=VAlign_Center, padding=<ltrb 0 10 0 10>}}
  }
  Text_ExecuteSwap {
   widgetClassPath=&UEFN_TextBlock
   parent=Button_ExecuteSwap
   widget=Text_ExecuteSwap
   properties {
    justification=Center
    text=Execute Swap
    colorAndOpacity {colorUseRule=UseColor_Specified, specifiedColor=<rgba 0.7913 0.7991 0.8228 1>}
    font {
     bForceMonospaced=false
     bMaterialIsStencil=false
     fontMaterial=null
     fontObject=&BurbankSmall
     letterSpacing=0
     monospacedWidth=1
     size=9
     skewAmount=0
     typefaceFontName=Regular
     outlineSettings {
      bApplyOutlineToDropShadows=false
      bMiteredCorners=false
      bSeparateFillAlpha=false
      outlineMaterial=null
      outlineSize=0
      outlineColor=<rgba 0 0 0 1>
     }
    }
    shadowColorAndOpacity=<rgba 0 0 0 1>
   }
   slot {class=ButtonSlot}
  }
 ]
}
